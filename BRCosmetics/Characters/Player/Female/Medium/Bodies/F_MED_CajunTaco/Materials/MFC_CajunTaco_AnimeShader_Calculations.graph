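# Material function for the CajunTaco character's anime shader, computing normals and masks for shading, specular, rim, metallic and line detail.
# Blends detail and vertex-color baked normals, then derives cel gradients from a cinematic-overridable light vector via curve atlases or smooth bands.
# Specular switches between Phong, Blinn, matcap, cubemap and planar highlights.
# Distance-field lines use derivative anti-aliasing and FOV-compensated distance thickness, while backlit shading overlays another gradient.
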MaterialFunction /BRCosmetics/Characters/Player/Female/Medium/Bodies/F_MED_CajunTaco/Materials/MFC_CajunTaco_AnimeShader_Calculations
graph {nodes [
  Composite!(In, RampWidth) {subgraphName=SmoothFloor}
  n1 = VectorParameter() {group=02_Anime Parameters, parameterName=ScreenSpaceLightVector, defaultValue=<rgba 1 1 -1 1>}
  Composite!(LightVector=@n1.RGB) {subgraphName=ProjectScreenAlighned}
  FunctionOutput(None=NamedRerouteUsage()) {outputName=DFL_Mask, sortPriority=8}
  FunctionOutput(None=NamedRerouteUsage()) {outputName=Specular_Gradient, sortPriority=5}
  FunctionOutput(None=NamedRerouteUsage()) {outputName=Sheen_Mask, sortPriority=6}
  FunctionOutput(None=NamedRerouteUsage()) {outputName=CINE_LightColor, sortPriority=32}
  FunctionOutput() {outputName=--------, sortPriority=30}
  FunctionOutput(None=NamedRerouteUsage()) {outputName=FX, sortPriority=9}
  FunctionOutput(None=NamedRerouteUsage()) {outputName=RimLight_Gradient, sortPriority=11}
  FunctionOutput(None=NamedRerouteUsage()) {outputName=Thickness, sortPriority=7}
  FunctionOutput(None=NamedRerouteUsage()) {outputName=Shading_Gradient, sortPriority=1}
  FunctionOutput(None=NamedRerouteUsage()) {outputName=Threshold_Gradient, sortPriority=2}
  n2 = StaticSwitchParameter(
   True=MaterialFunctionCall(BaseNormal=NamedRerouteUsage(), AdditionalNormal=NamedRerouteUsage()) {materialFunction=&BlendAngleCorrectedNormals}.Result
   False=NamedRerouteUsage()
  ) {group=02_Anime Parameters, parameterName=UseDetailNormalShading, sortPriority=0}
  n3 = VertexColor()
  n4 = Multiply(A=@n3) {constB=2}
  n5 = ScalarParameter() {desc="0.0 = surface normals, 1.0 baked vertex color normal", group=02_Anime Parameters, parameterName=VertexBakedNormal_Blend, sliderMax=1}
  n6 = LinearInterpolate(
   A=Transform(None=FunctionOutput(None=@n2) {outputName="Normal_TS(PreBakedBlend)", sortPriority=20})
   B=Transform(None=Subtract(A=@n4)) {transformSourceType=TRANSFORMSOURCE_Local}
   Alpha=Multiply(A=@n5, B=@n3.A)
  )
  n7 = FunctionOutput(None=FunctionInput(Preview=Normalize(VectorInput=@n6)) {bUsePreviewValueAsDefault=true, inputName=OverrideNormal_WS}) {desc=really needed??, outputName=Normal_WS, sortPriority=22}
  FunctionOutput(None=Transform(None=NamedRerouteDeclaration(None=@n7) {name=Normal_WS, nodeColor=<rgba 0.2083 0.2083 1 1>}) {transformSourceType=TRANSFORMSOURCE_World, transformType=TRANSFORM_Tangent}) {desc=really needed??, outputName=Normal_TS, sortPriority=21}
  FunctionOutput(None=NamedRerouteUsage()) {outputName=CINE_LightSources, sortPriority=31}
  FunctionOutput(None=NamedRerouteUsage()) {outputName=Metallic_Gradient, sortPriority=10}
  n8 = Saturate(None=Add(A=NamedRerouteUsage(), B=ScalarParameter() {group=02_Anime Parameters, parameterName=ShadingCurve_Bias, sliderMax=1, sliderMin=-1}))
  n9 = CurveAtlasRowParameter(CurveTime=@n8) {
   atlas=/Game/Characters/Shared/CurveAtlas/AthenaCharacterCurveAtlas
   curve=/Game/Characters/Shared/CurveAtlas/HB_TwoToneShadingCurve
   defaultValue=22
   group=02_Anime Parameters
   parameterName=ShadingCurve
   sortPriority=5
  }
  n10 = StaticSwitchParameter(
   True=ChannelMaskParameter(None=@n9) {group=02_Anime Parameters, parameterName=ShadingCurveChannel, sortPriority=10}.R
   False=MaterialFunctionCall!(
    NdL=@n8
    Width=ScalarParameter() {defaultValue=0.02, group=02_Anime Parameters, parameterName=ShadingCurve_Width, sliderMax=1}
    Bands Amount=ScalarParameter() {defaultValue=1, group=02_Anime Parameters, parameterName=ShadingCurve_Bands, sliderMax=4, sliderMin=1}
   ) {materialFunction=/Game/Global/Materials/Character_Parent_2023/Anime_Parent_2023/MaterialFunctions/MF_SmoothCelShadingWithBands}.Divide
  ) {group=02_Anime Parameters, parameterName=UseShadingCurve, sortPriority=-1}
  n11 = Saturate(None=Add(A=NamedRerouteUsage(), B=ScalarParameter() {group=02_Anime Parameters, parameterName=ShadingCurveAdditional_Bias, sliderMax=1, sliderMin=-1}))
  n12 = CurveAtlasRowParameter(CurveTime=@n11) {
   atlas=&AthenaCharacterCurveAtlas
   curve=&HB_TwoToneShadingCurve
   defaultValue=22
   group=02_Anime Parameters
   parameterName=ShadingCurveAdditional
   sortPriority=5
  }
  n13 = ChannelMaskParameter(None=@n12) {group=02_Anime Parameters, parameterName=ShadingCurveAdditionalChannel, sortPriority=10}
  n14 = MaterialFunctionCall(
   NdL=@n11
   Width=ScalarParameter() {defaultValue=0.02, group=02_Anime Parameters, parameterName=ShadingCurveAdditional_Width, sliderMax=1}
   Bands Amount=ScalarParameter() {defaultValue=1, group=02_Anime Parameters, parameterName=ShadingCurveAdditional_Bands, sliderMax=4, sliderMin=1}
  ) {materialFunction=&MF_SmoothCelShadingWithBands}
  n15 = ChannelMaskParameter(None=NamedRerouteUsage()) {
   desc=Select in which channel of the FX map the mask located
   group=02_Anime Parameters
   parameterName=AdditionalShadingCurve_FXChannel
   sortPriority=10
  }
  n16 = StaticSwitchParameter(
   True=LinearInterpolate(
    A=@n10
    B=StaticSwitchParameter(True=@n13.R, False=@n14.Divide) {group=02_Anime Parameters, parameterName=UseShadingCurveAdditional, sortPriority=-1}
    Alpha=@n15.R
   )
   False=@n10
  ) {group=02_Anime Parameters, parameterName=UseAdditionalShadingGradient, sortPriority=-2}
  n17 = ScalarParameter() {
   defaultValue=1
   desc="To convert the (mathmatical) linear greadient to linear perceived gradient"
   group=09_Anime Parameters_TechArt
   parameterName=ShadingCurve_Exp
   sliderMax=2.2
   sliderMin=0.4545
  }
  NamedRerouteDeclaration(None=Saturate(None=Power(Base=@n16, Exp=@n17) {constExponent=2.2})) {name=Shading_Gradient, nodeColor=<rgba 1 0.1875 0 1>}
  NamedRerouteDeclaration(None=FunctionInput() {inputName=FX, inputType=FunctionInput_Vector4, sortPriority=4}) {name=Map_FX, nodeColor=<rgba 0 0.326 1 1>}
  n18 = TextureCoordinate()
  n19 = TextureSampleParameter2D!(
   UVs=MaterialFunctionCall!(Vector=ReflectionVectorWS()) {materialFunction=/Engine/Functions/Engine_MaterialFunctions01/Coordinates/LongLatToUV}.Result
   "DDX(UVs)"=DDX(Value=@n18)
   "DDY(UVs)"=DDY(Value=@n18)
   Apply View MipBias
  ) {
   group=01_Anime Maps
   mipValueMode=TMVM_Derivative
   parameterName=SpecularMap_Cube
   samplerSource=SSM_Wrap_WorldGroupSettings
   sortPriority=20
   texture=/Game/Global/Materials/Character_Parent_2023/Anime_Parent_2023/Textures/T_CubeMap_Highlight
  }
  NamedRerouteDeclaration(None=@n19.RGB) {name=CubeSpec, nodeColor=<rgba 1 0.4327 0 1>}
  n20 = Transform(None=Constant3Vector() {constant=<rgba 0 0 1 0>}) {transformSourceType=TRANSFORMSOURCE_View, transformType=TRANSFORM_Local}
  n21 = MaterialFunctionCall(Float2=MaterialFunctionCall(Vector or UVs=ComponentMask(None=@n20) {g=true, r=true}) {materialFunction=&VectorToRadialValue}.Radial Coordinates) {materialFunction=&BreakOutFloat2Components}
  n22 = AppendVector(A=Add(A=Multiply(A=@n21.R) {constB=4}) {constB=-1}, B=Multiply(A=ComponentMask(None=@n20) {b=true}) {constB=-1})
  n23 = WorldPosition()
  n24 = MaterialFunctionCall(Float3=Divide(
    A=TransformPosition(None=@n23.XYZ) {transformSourceType=TRANSFORMPOSSOURCE_World}
    B=ScalarParameter() {defaultValue=30, group=02_Anime Parameters, parameterName=SpecularHighlight_Planar_Scale, sliderMax=200}
   ) {constB=30}) {materialFunction=&BreakOutFloat3Components}
  n25 = MaterialFunctionCall(X=@n24.R, Y=@n24.R) {materialFunction=&MakeFloat2}
  n26 = MaterialFunctionCall(X=@n24.G, Y=@n24.G) {materialFunction=&MakeFloat2}
  n27 = Add(A=Multiply(A=StaticSwitchParameter(True=@n25.Result, False=@n26.Result) {defaultValue=true, group=02_Anime Parameters, parameterName=UsePlanarReflectionYPlane, sortPriority=2}) {constB=-1}) {constB=0.5}
  n28 = TextureSampleParameter2D!(
   UVs=Rotator(
    Coordinate=Add(A=@n22, B=@n27)
    Time=ScalarParameter() {group=02_Anime Parameters, parameterName=SpecularHighlight_Planar_Rotate, sliderMax=1, sliderMin=-1}
   ) {speed=3.1415}
   Apply View MipBias
  ) {
   group=01_Anime Maps
   parameterName=SpecularMap_Planar
   samplerSource=SSM_Wrap_WorldGroupSettings
   sortPriority=20
   texture=/Game/Global/Materials/Character_Parent_2023/Anime_Parent_2023/Textures/T_Planar_Highlight
  }
  NamedRerouteDeclaration(None=@n28.RGB) {name=PlanarSpec, nodeColor=<rgba 1 0.4327 0 1>}
  n29 = TextureSampleParameter2D(UVs=Multiply(A=TextureCoordinate(), B=ScalarParameter() {defaultValue=128, group=02_Anime Parameters, parameterName=DetailNormal_Tiling, sliderMax=256})) {
   group=01_Anime Maps
   parameterName=DetailNormal_Map
   samplerType=SAMPLERTYPE_Normal
   sortPriority=19
   texture=/Game/Global/Materials/Character_Parent_2023/Anime_Parent_2023/Textures/T_SurfaceDetail_N
  }
  n30 = ScalarParameter() {
   desc="0 =original normal map intensity\r\n1= completly flat normal\r\n2= inverted normal direction"
   group=02_Anime Parameters
   parameterName=DetailNormal_Flattness
   sliderMax=2
  }
  n31 = ChannelMaskParameter(None=NamedRerouteUsage()) {group=01_Anime Maps, maskChannel=Alpha, parameterName=FX_Channel_DetailNormalMask, sortPriority=27, defaultValue=<rgba 0 0 0 1>}
  NamedRerouteDeclaration(None=LinearInterpolate(
    A=Constant3Vector() {constant=<rgba 0 0 1 0>}
    B=MaterialFunctionCall(Normal=@n29.RGB, Flatness=@n30) {materialFunction=&FlattenNormal}.Result
    Alpha=@n31.R
   )) {name=DetailNormal, nodeColor=<rgba 0 0.1253 1 1>}
  n32 = MaterialFunctionCall(Float3=FunctionInput() {inputName=STT, sortPriority=1, previewValue=<0 0.5 0 0>}) {materialFunction=&BreakOutFloat3Components}
  NamedRerouteDeclaration(None=@n32.B) {name=MapInput_Thickness, nodeColor=<rgba 0 1 0.4064 1>}
  n33 = LinearInterpolate(
   A=ScalarParameter() {defaultValue=0.4, group=02_Anime Parameters, parameterName=DF_LineThickness_Near, sliderMax=1}
   B=ScalarParameter() {defaultValue=0.6, group=02_Anime Parameters, parameterName=DF_LineThickness_Far, sliderMax=1}
   Alpha=MaterialFunctionCall!(
    Fade Offset=ScalarParameter() {defaultValue=50, group=09_Anime Parameters_TechArt, parameterName=DF_FadeStart, sliderMax=200}
    Fade Distance=ScalarParameter() {defaultValue=1000, group=09_Anime Parameters_TechArt, parameterName=DF_FadeDistance, sliderMax=1000}
   ) {materialFunction=/Game/Global/Materials/Character_Parent_2023/Anime_Parent_2023/MaterialFunctions/MFC_DistanceFade_FOV_compensation}.Result
  )
  n34 = Power(Base=NamedRerouteUsage()) {constExponent=0.4545, desc=gamma correction}
  n35 = Subtract(A=OneMinus(None=@n33), B=@n34)
  n36 = AppendVector(A=DDX(Value=@n35), B=DDY(Value=@n35))
  n37 = ScalarParameter() {defaultValue=0.8, group=02_Anime Parameters, parameterName=DF_LineHardness, sliderMax=1}
  n38 = Divide(A=@n35, B=Add(A=Length(None=@n36) {is=class MaterialExpressionLength}, B=LinearInterpolate(Alpha=@n37) {constA=0.5, constB=0}) {constB=0})
  NamedRerouteDeclaration(None=Saturate(None=Subtract(B=@n38) {constA=0.5})) {name=DFL_Mask, nodeColor=<rgba 0.1146 0.1142 0.1092 1>}
  NamedRerouteDeclaration(None=FunctionInput() {inputName=DFL, inputType=FunctionInput_Scalar, sortPriority=3}) {name=DFL, nodeColor=<rgba 0.4692 1 0 1>}
  NamedRerouteDeclaration(None=@n32.R) {name=MapInput_Sheen, nodeColor=<rgba 0 0.3016 1 1>}
  n39 = TransformPosition(None=WorldPosition().XYZ) {transformSourceType=TRANSFORMPOSSOURCE_World, transformType=TRANSFORMPOSSOURCE_Camera}
  n40 = Normalize(VectorInput=VertexNormalWS())
  n41 = MaterialFunctionCall(XY=CrossProduct(A=Normalize(VectorInput=@n39), B=Transform(None=@n40) {transformSourceType=TRANSFORMSOURCE_World, transformType=TRANSFORM_Camera})) {materialFunction=&Swizzle}
  n42 = TextureSampleParameter2D(UVs=Add(A=Multiply(A=@n41.YX) {constB=0.5}) {constB=0.5}) {
   group=01_Anime Maps
   parameterName=SpecularMap_MatCap
   samplerSource=SSM_Wrap_WorldGroupSettings
   sortPriority=20
   texture=/Game/Global/Materials/Character_Parent_2023/Anime_Parent_2023/Textures/T_MatCap_Highlight
  }
  NamedRerouteDeclaration(None=@n42.RGB) {name=MatCapSpec, nodeColor=<rgba 1 0.4327 0 1>}
  n43 = NamedRerouteUsage()
  n44 = CameraVectorWS()
  n45 = Multiply(A=@n44) {constB=0.1}
  n46 = Saturate(None=DotProduct(
    A=MaterialFunctionCall() {materialFunction=&CameraDirectionVector}.Result
    B=Constant3Vector() {constant=<rgba 0 0 1 0>}
   ))
  n47 = NamedRerouteUsage()
  n48 = ScalarParameter() {
   desc="LockedZ only for the specular highlight.\r\nfor a streched highlight in on the Z axis use the Blinn highlight"
   group=02_Anime Parameters
   parameterName=SpecularHighlight_LockedZVector
   sliderMax=1
  }
  n49 = AppendVector(A=Constant2Vector() {g=1, r=1}, B=Saturate(None=OneMinus(None=@n48)))
  n50 = Multiply(A=Add(A=LinearInterpolate(A=@n44, B=@n45, Alpha=@n46) {constB=0}, B=@n47), B=@n49)
  n51 = Multiply(A=@n47, B=@n49)
  n52 = NamedRerouteUsage()
  n53 = StaticSwitchParameter(
   True=DotProduct(A=Normalize(VectorInput=@n50), B=NamedRerouteUsage()) {desc=Blinn}
   False=DotProduct(A=Normalize(VectorInput=@n51), B=ReflectionVectorWS(CustomWorldNormal=@n52)) {desc=Phong}
  ) {
   desc="Default is the Phong highlight.\r\nUse Blinn for a vertical streched highlight"
   group=02_Anime Parameters
   parameterName=UseBlinnHighlight
   sortPriority=0
  }
  n54 = Saturate(None=LinearInterpolate(
    A=@n53
    B=NamedRerouteUsage()
    Alpha=ScalarParameter() {group=02_Anime Parameters, parameterName=SpecularHighlight_CameraFacing_Blend, sliderMax=1}
   ))
  n55 = Constant() {r=1}
  n56 = ScalarParameter() {defaultValue=0.1, group=02_Anime Parameters, parameterName=SpecularHighlight_Size, sliderMax=1, sliderMin=0.0001}
  n57 = ScalarParameter() {defaultValue=0.2, group=02_Anime Parameters, parameterName=SpecularHighlight_Hardness, sliderMax=1}
  n58 = StaticSwitchParameter(
   True=NamedRerouteUsage()
   False=StaticSwitchParameter(True=NamedRerouteUsage(), False=SphereMask(A=@n54, B=@n55, Radius=@n56, Hardness=@n57) {hardnessPercent=78}) {group=02_Anime Parameters, parameterName=UsePlanarSpecular, sortPriority=1}
  ) {group=02_Anime Parameters, parameterName=UseCubeMapSpecular, sortPriority=1}
  NamedRerouteDeclaration(None=Multiply(
    A=StaticSwitchParameter(True=@n43, False=@n58) {group=02_Anime Parameters, parameterName=UseMatCapSpecular, sortPriority=1}
    B=ScalarParameter() {defaultValue=1, group=02_Anime Parameters, parameterName=SpecularHighlight_Intensity, sliderMax=8}
   )) {name=Specular_Curve, nodeColor=<rgba 1 0.4453 0 1>}
  n59 = Saturate(None=DotProduct(A=NamedRerouteUsage(), B=PixelNormalWS()))
  n60 = ScalarParameter() {defaultValue=3, group=02_Anime Parameters, parameterName=RimLight_FresnelExponent, sliderMax=12}
  n61 = PixelNormalWS()
  n62 = Saturate(None=InverseLinearInterpolate(
    A=ScalarParameter() {defaultValue=0.1, group=02_Anime Parameters, parameterName=RimLight_Min, sliderMax=1}
    B=ScalarParameter() {defaultValue=0.2, group=02_Anime Parameters, parameterName=RimLight_Max, sliderMax=1}
    Value=Fresnel(ExponentIn=@n60, Normal=@n61) {baseReflectFraction=0}
   ) {constA=0.1, constB=0.2})
  NamedRerouteDeclaration(None=Saturate(None=Multiply(A=@n59, B=@n62))) {name=RimLight, nodeColor=<rgba 1 0 0.1172 1>}
  NamedRerouteDeclaration(None=DotProduct(A=NamedRerouteUsage(), B=CameraVectorWS())) {name=NdotV, nodeColor=<rgba 1 0.6242 0 1>}
  n63 = NamedRerouteUsage()
  n64 = Constant() {desc=Metallic Band Amount, r=3}
  n65 = Multiply(A=Saturate(None=Power(Base=@n63) {constExponent=2.2}), B=@n64)
  n66 = Multiply(A=Saturate(None=ScalarParameter() {defaultValue=0.1, group=02_Anime Parameters, parameterName=Metallic_Roughness, sliderMax=1})) {constB=2}
  n67 = LinearInterpolate(
   B=Divide(
    A=MaterialFunctionCall!(In=@n65, Ramp Width=@n66) {materialFunction=/Engine/Functions/Engine_MaterialFunctions01/Math/SmoothCeil}.Result
    B=@n64
   )
   Alpha=ScalarParameter() {defaultValue=1, group=02_Anime Parameters, parameterName=Metallic_Intensity, sliderMax=1}
  ) {constA=1}
  n68 = ScalarParameter() {
   desc="The Bias will adjust the Metallic gradient.\r\n0.0 = the center matches the Color\r\n1.0 = the inner ring matches the Color\r\n2.0 = the outer ring match the Color"
   group=02_Anime Parameters
   parameterName=Metallic_Bias
   sliderMax=2
   sliderMin=-1
  }
  NamedRerouteDeclaration(None=Add(A=Saturate(None=@n67), B=Multiply(A=@n68) {constB=0.3333})) {name=Metallic_Gradient, nodeColor=<rgba 0.1107 1 0.7992 1>}
  NamedRerouteDeclaration(None=@n32.G) {name=MapInput_Theshold, nodeColor=<rgba 1 0.4584 0 1>}
  n69 = MaterialFunctionCall!(In=DotProduct(A=NamedRerouteUsage(), B=NamedRerouteUsage())) {materialFunction=/Game/Packages/Fortress_SharedMaterials/Cinematics/Anime/MF_CINE_AnimeLightingFeatures}
  NamedRerouteDeclaration(None=@n69.CINE_LightSources) {name=CINE_LightSources, nodeColor=<rgba 1 0.9609 0 1>}
  NamedRerouteDeclaration(None=@n69.CINE_LightColor) {name=CINE_LightColor, nodeColor=<rgba 1 0.6803 0 1>}
  n70 = ScalarParameter() {group=02_Anime Parameters, parameterName=Threshold_Min, sliderMax=1}
  n71 = ScalarParameter() {defaultValue=1, group=02_Anime Parameters, parameterName=Threshold_Max, sliderMax=1}
  n72 = Power(Base=NamedRerouteUsage(), Exp=ScalarParameter() {defaultValue=1, group=02_Anime Parameters, parameterName=Threshold_Exp, sliderMax=8})
  NamedRerouteDeclaration(None=Saturate(None=LinearInterpolate(A=@n70, B=@n71, Alpha=@n72))) {name=Threshold, nodeColor=<rgba 1 0.2971 0.8877 1>}
  n73 = ChannelMaskParameter(None=NamedRerouteUsage()) {
   desc="Select in which channel of the FX map the mask camera light vector will be stored "
   group=02_Anime Parameters
   parameterName=CameraLightVector_FXChannel
   sortPriority=10
  }
  n74 = MaterialFunctionCall!(In=MaterialFunctionCall() {materialFunction=/Game/Global/Materials/Character_Parent_2023/Anime_Parent_2023/MaterialFunctions/MFC_AnimeShader_LightDirection}.Direction Vector) {materialFunction=/Game/Packages/Fortress_SharedMaterials/Cinematics/Anime/MF_CINE_AnimeLightDirOverride}
  n75 = MaterialFunctionCall(Float3=@n74.Result) {materialFunction=&BreakOutFloat3Components}
  n76 = ScalarParameter() {
   desc="0.0 = the light vector is locked on the XY plane.\r\n1.0 = the actual light vector is used  \r\n(for Faces use a blend value of 0,0)"
   group=02_Anime Parameters
   parameterName=LockedZLightVector_Blend
   sliderMax=1
  }
  n77 = MaterialFunctionCall(X=@n75.R, Y=@n75.R, Z=Multiply(A=@n75.B, B=@n76)) {materialFunction=&MakeFloat3}
  n78 = StaticSwitchParameter(True=LinearInterpolate(A=@n74.Result, B=Normalize(VectorInput=@n77.Result), Alpha=VertexColor().A), False=@n74.Result) {
   desc="Lock the Z axis of the Light Vector\r\nis driven by the VertexColor A channel (0 = Regular LightVector, 1= Locked LIghtVector)"
   group=02_Anime Parameters
   parameterName=UseLockedZLightVector
   sortPriority=0
  }
  n79 = StaticSwitchParameter(
   True=LinearInterpolate(
    A=@n78
    B=MaterialFunctionCall() {materialFunction=&CameraDirectionVector}.Result
    Alpha=@n73.R
   )
   False=@n78
  ) {
   desc="Use the Camera vector instead of the light vector which is defined by the FX map. Select the FX channel with this parameter \"CameraLightVector_FXChannel\"."
   group=02_Anime Parameters
   parameterName=UseCameraLightVector
   sortPriority=0
  }
  NamedRerouteDeclaration(None=FunctionInput(Preview=@n79) {bUsePreviewValueAsDefault=true, inputName=CustomLightVector, previewValue=<0 0 1 1>}) {name=LightVector, nodeColor=<rgba 1 0.934 0.5103 1>}
  NamedRerouteDeclaration(None=@n69.CINE_ShadingCurve) {name=NdotL, nodeColor=<rgba 0.9723 1 0.2825 1>}
  NamedRerouteDeclaration(None=FunctionInput() {inputName=Normal_Map, sortPriority=2, previewValue=<0 0 1 1>}) {name=Normal, nodeColor=<rgba 0 0.4109 1 1>}
  n80 = NamedRerouteUsage()
  n81 = NamedRerouteUsage()
  n82 = Add(A=ScalarParameter() {defaultValue=4, group=02_Anime Parameters, parameterName=BackLit_Exponent, sliderMax=12, sortPriority=31}) {constB=2.2}
  n83 = Saturate(None=SmoothStep(Value=Power(Base=@n81, Exp=@n82) {constExponent=4}) {constMax=0, constMin=0.02})
  n84 = Add(
   A=DotProduct(
    A=MaterialFunctionCall() {materialFunction=&CameraDirectionVector}.Result
    B=NamedRerouteUsage()
   )
   B=ScalarParameter() {defaultValue=0.25, group=02_Anime Parameters, parameterName=BackLit_Bias, sliderMax=1}
  )
  n85 = ScalarParameter() {defaultValue=0.75, group=02_Anime Parameters, parameterName=BackLit_MaxBias, sliderMax=1}
  n86 = LinearInterpolate(
   A=@n80
   B=Add(A=@n80, B=NamedRerouteDeclaration(None=@n83) {name=BackLit_Mask, nodeColor=<rgba 0 1 0.5078 1>})
   Alpha=NamedRerouteDeclaration(None=Clamp(None=@n84, Max=@n85)) {name=BackLit_Curve, nodeColor=<rgba 0.3207 0 1 1>}
  )
  n87 = Add(A=Multiply(A=NamedRerouteDeclaration(None=@n86) {name=BackLitShading_-1to1, nodeColor=<rgba 0.1307 1 0.334 1>}) {constB=0.5}) {constB=0.5}
  NamedRerouteDeclaration(None=MaterialFunctionCall!(Base=NamedRerouteUsage(), Blend=NamedRerouteDeclaration(None=@n87) {name=BackLitShading_0to1, nodeColor=<rgba 0 1 0.9234 1>}) {materialFunction=/Game/Global/Materials/Character_Parent_2023/Anime_Parent_2023/MaterialFunctions/MF_BlendOverlay}.Result) {name=BackLitShading+Threshold, nodeColor=<rgba 0 1 0.9234 1>}
  PinBase() {reroutePins [
    In {expression=&self:MaterialExpressionReroute_4}
    RampWidth {expression=&self:MaterialExpressionReroute_6}
   ]}
  MaterialExpressionReroute_4 = Reroute()
  n88 = Frac(None=@MaterialExpressionReroute_4.In)
  MaterialExpressionReroute_6 = Reroute()
  n89 = OneMinus(None=@MaterialExpressionReroute_6.RampWidth)
  n90 = Max(A=Divide(A=Subtract(A=@n88, B=@n89), B=@MaterialExpressionReroute_6.RampWidth), B=Constant() {r=0.5}) {desc=It is good to make functions look like a lancer}
  n91 = Add(A=Floor(None=@MaterialExpressionReroute_4.In), B=ConstantBiasScale(None=Cosine(None=@n90)))
  PinBase!(Output=@n91) {reroutePins [Output {expression=&self:MaterialExpressionReroute_7}]}
  PinBase!() {reroutePins [LightVector {expression=&self:MaterialExpressionReroute_0}]}
  MaterialExpressionReroute_0 = Reroute(None=@n1.RGB)
  n92 = Normalize(VectorInput=@MaterialExpressionReroute_0.LightVector)
  n93 = ComponentMask(None=@n92) {r=true}
  n94 = Constant3Vector() {constant=<rgba 0 1 0 0>}
  n95 = Normalize(VectorInput=WorldPosition() {worldPositionShaderOffset=WPT_CameraRelative}.XYZ)
  n96 = Normalize(VectorInput=CrossProduct(A=Transform(None=@n94) {transformSourceType=TRANSFORMSOURCE_View}, B=@n95))
  n97 = ComponentMask(None=@n92) {g=true}
  n98 = CrossProduct(A=@n95, B=@n96)
  n99 = Add(A=Add(A=Multiply(A=@n93, B=@n96), B=Multiply(A=@n97, B=@n98)), B=Multiply(A=ComponentMask(None=@n92) {b=true}, B=@n95))
  PinBase!(Result=@n99) {reroutePins [Result {expression=&self:MaterialExpressionReroute_1}]}
  MaterialExpressionReroute_7 = Reroute(None=@n91)
  SphereMask(A=@n34, B=Constant() {r=1}, Radius=@n33, Hardness=@n37) {attenuationRadius=0.1}
  MaterialExpressionReroute_1 = Reroute(None=@n99)
  StaticSwitchParameter(True=@MaterialExpressionReroute_1.Result) {
   desc="Use a Screen / View alight light vector. \r\nOverrides the UseDirectionalLightFunction!"
   group=02_Anime Parameters
   parameterName=UseScreenSpaceLightDirection
   sortPriority=0
  }
  n100 = Multiply(A=StaticSwitchParameter(
    True=TextureSampleParameter2D() {parameterName=BakedNormalMap, samplerType=SAMPLERTYPE_Normal, texture=&FlatNormalBlue}.RGB
    False=VertexColor()
   ) {group=Toon_Shading, parameterName=UseBakedNormalMap, sortPriority=0}) {constB=2}
  Transform(None=Normalize(VectorInput=Subtract(A=@n100))) {transformSourceType=TRANSFORMSOURCE_Local}
 ]}
