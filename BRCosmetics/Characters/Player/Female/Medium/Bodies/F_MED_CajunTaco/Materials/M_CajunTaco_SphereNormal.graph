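# Toon character material for CajunTaco that blends surface normals toward a spherical normal, using the 2023 anime shader functions and Cloth shading model.
# The proxy normal radiates from a parameterized point in pre-skinning local space, optionally flattened to cylindrical, and is computed through a vertex interpolator.
# The shared Pre function feeds a CajunTaco-specific Calculations function, which receives the blended world normal as an override and supplies gradients to the shared Post function.
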Material /BRCosmetics/Characters/Player/Female/Medium/Bodies/F_MED_CajunTaco/Materials/M_CajunTaco_SphereNormal
graph {
 properties {
  bAutomaticallySetUsageInEditor=false
  bUseMaterialAttributes=true
  bUsedWithMeshDeformer=true
  bUsedWithMorphTargets=true
  bUsedWithSkeletalMesh=true
  bUsedWithStaticLighting=true
  pixelDepthOffsetMode=PDOM_Legacy
  shadingModel=MSM_Cloth
  usedShadingModels=Cloth
 }
 nodes [
  n1 = MaterialFunctionCall!(CustomUV) {materialFunction=/Game/Global/Materials/Character_Parent_2023/Anime_Parent_2023/MaterialFunctions/MFC_AnimeShader_2023_Pre}
  n2 = Subtract(
   A=LocalPosition() {localOrigin=InstancePreSkinning}.XYZ
   B=VectorParameter() {group=Toon_Shading, parameterName=NormalSphere_Location, defaultValue=<rgba 0 0 162 1>}.RGB
  )
  n3 = Multiply(A=@n2, B=Constant3Vector() {constant=<rgba 1 1 0 0>})
  n4 = ScalarParameter() {group=Toon_Shading, parameterName=NormalSphereCylindrical_Blend, sliderMax=1}
  n5 = Transform(None=Normalize(VectorInput=LinearInterpolate(A=@n2, B=@n3, Alpha=@n4))) {transformSourceType=TRANSFORMSOURCE_Local}
  n6 = MaterialFunctionCall!(
   CustomLightVector
   OverrideNormal_WS=LinearInterpolate(
    A=Transform(None=@n1.Normal)
    B=VertexInterpolator(VS=@n5).PS
    Alpha=ScalarParameter() {group=Toon_Shading, parameterName=NormalSphere_Blend, sliderMax=1}
   ) {constAlpha=0}
   STT=@n1.STT
   Normal_Map=@n1.STT
   DFL=@n1.STT
   FX=@n1.STT
  ) {materialFunction=&/MFC_CajunTaco_AnimeShader_Calculations}
  MaterialFunctionCall!(
   Shading_Gradient=@n6.Shading_Gradient
   Threshold_Gradient=@n6.Shading_Gradient
   Specular_Gradient=@n6.Shading_Gradient
   Sheen_Mask=@n6.Shading_Gradient
   Thickness=@n6.Shading_Gradient
   DFL_Mask=@n6.Shading_Gradient
   FX=@n6.Shading_Gradient
   Metallic_Gradient=@n6.Shading_Gradient
   RimLight_Gradient=@n6.Shading_Gradient
   NormalBuffer=@n6.Shading_Gradient
   --------
   CINE_LightSources=@n6.Shading_Gradient
   CINE_LightColor=@n6.Shading_Gradient
   ---------
   Color_Lit=@n1.Color_Lit
   Color_Shaded=@n1.Color_Lit
   SRM=@n1.Color_Lit
   Emissive=@n1.Color_Lit
   DFL_Color=@n1.Color_Lit
   Specular_Color=@n1.Color_Lit
   RimLight_Color=@n1.Color_Lit
  ) {materialFunction=/Game/Global/Materials/Character_Parent_2023/Anime_Parent_2023/MaterialFunctions/MFC_AnimeShader_2023_Post}
  SetMaterialAttributes(Emissive Color=NamedRerouteUsage()) {attributeSetTypes [#1]}
 ]
}
