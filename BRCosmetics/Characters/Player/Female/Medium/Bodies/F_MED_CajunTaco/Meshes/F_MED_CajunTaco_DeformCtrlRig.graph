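# Post-animation deformation Control Rig for the F_MED_CajunTaco female body cosmetic, correcting twist bones and secondary costume bones.
# Weighted parent constraints anchor ropes, knee pad, calf belt and skirt to body bones.
# Swing/twist delta readers, spherical pose readers and deformRig anim attributes weight corrective offsets, with crouch, skydive and pickaxe-slide variables adjusting the skirt.
# Arm deform bones follow stretchy B-spline IK through constrained controls.
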ControlRigBlueprint /BRCosmetics/Characters/Player/Female/Medium/Bodies/F_MED_CajunTaco/Meshes/F_MED_CajunTaco_DeformCtrlRig
graphs [
 New Function_ContainedGraph {nodes [
   n1 = RigVMFunctionEntryNode()
   RigVMFunction_ControlFlowBranch = &RigVMFunction_ControlFlowBranch(ExecuteContext=@n1.ExecuteContext, Condition=@n1.debug)
   RigVMFunctionReturnNode(ExecuteContext=@RigVMFunction_ControlFlowBranch.Completed)
   RigVMFunction_Sequence = RigVMAggregateNode(ExecuteContext=@RigVMFunction_ControlFlowBranch.True)
   RigUnit_ItemArray = &RigUnit_ItemArray(Items="((Type=Bone,Name=\"spine_01\"),(Type=Bone,Name=\"spine_02\"),(Type=Bone,Name=\"spine_03\"),(Type=Bone,Name=\"spine_04\"),(Type=Bone,Name=\"spine_05\"),(Type=Bone,Name=\"clavicle_l\"),(Type=Bone,Name=\"upperarm_l\"),(Type=Bone,Name=\"lowerarm_l\"),(Type=Bone,Name=\"hand_l\") ...(truncated)")
   DISPATCH_RigVMDispatch_ArrayIterator = RigVMDispatchNode(Array=@RigUnit_ItemArray.Items, ExecuteContext=@RigVMFunction_Sequence.A) {template="DISPATCH_RigVMDispatch_ArrayIterator(in Array,out Element,out Index,out Count,out Ratio)"}
   Concat = /Script/RigVM.RigVMFunction_NameConcat(B=_ctrl, A=@DISPATCH_RigVMDispatch_ArrayIterator.Element.Name) {is=struct RigVMUnitNode, template="Concat::Execute(in A,in B,out Result)"}
   > &RigUnit_SetTransform(
    Item=(Type=Bone,Name="spine_01")
    Weight=1
    bInitial=False
    Item.Name=@DISPATCH_RigVMDispatch_ArrayIterator.Element.Name
    Value=&RigUnit_GetTransform(Item=(Type=Control,Name="spine_01_ctrl"), bInitial=False, Item.Name=@Concat.Result).Transform
   )
   RigUnit_GetRelativeTransformForItem = &RigUnit_GetRelativeTransformForItem(Child=(Type=Bone,Name="thigh_l"), Parent=(Type=Null,Name="thigh_l_null"), bChildInitial=False, bParentInitial=False)
   RigVMFunction_MathQuaternionSwingTwist = &RigVMFunction_MathQuaternionSwingTwist(TwistAxis=(X=1,Y=0,Z=0), Input=@RigUnit_GetRelativeTransformForItem.RelativeTransform.Rotation)
   Inverse = /Script/RigVM.RigVMFunction_MathQuaternionInverse(Value=@RigVMFunction_MathQuaternionSwingTwist.Twist) {is=struct RigVMUnitNode, template="Inverse::Execute(in Value,out Result)"}
   &RigUnit_ModifyTransforms(
    ItemToModify=((Item=(Type=Bone,Name="thigh_twist_01_l"),Transform=(Rotation=(X=0,Y=0,Z=0,W=1),Translation=(X=0,Y=0,Z=0),Scale3D=(X=1,Y=1,Z=1))))
    Weight=1
    WeightMinimum=0
    ItemToModify.0.Transform.Rotation=@Inverse.Result
    ExecutePin=@RigVMFunction_Sequence.B
   )
   n2 = &RigUnit_GetRelativeTransformForItem(Child=(Type=Bone,Name="thigh_r"), Parent=(Type=Null,Name="thigh_r_null"), bChildInitial=False, bParentInitial=False)
   > &RigUnit_ModifyTransforms(
    ItemToModify=((Item=(Type=Bone,Name="thigh_twist_01_r"),Transform=(Rotation=(X=0,Y=0,Z=0,W=1),Translation=(X=0,Y=0,Z=0),Scale3D=(X=1,Y=1,Z=1))))
    Weight=1
    WeightMinimum=0
    ItemToModify.0.Transform.Rotation=&RigVMFunction_MathQuaternionInverse(Value=&RigVMFunction_MathQuaternionSwingTwist(TwistAxis=(X=1,Y=0,Z=0), Input=@n2.RelativeTransform.Rotation).Twist) {template="Inverse::Execute(in Value,out Result)"}.Result
   )
   n3 = &RigVMFunction_MathQuaternionInverse(Value=&RigVMFunction_MathQuaternionSwingTwist(
     TwistAxis=(X=1,Y=0,Z=0)
     Input=&RigUnit_GetRelativeTransformForItem(Child=(Type=Bone,Name="upperarm_l"), Parent=(Type=Null,Name="upperarm_l_null"), bChildInitial=False, bParentInitial=False).RelativeTransform.Rotation
    ).Twist) {template="Inverse::Execute(in Value,out Result)"}
   &RigUnit_ModifyTransforms(
    ItemToModify="((Item=(Type=Bone,Name=\"upperarm_twist_01_l\"),Transform=(Rotation=(X=0,Y=0,Z=0,W=1),Translation=(X=0,Y=0,Z=0),Scale3D=(X=1,Y=1,Z=1))),(Item=(Type=Bone,Name=\"upperarm_twist_02_l\"),Transform=(Rotation=(X=0,Y=0,Z=0,W=1),Translation=(X=0,Y=0,Z=0),Scale3D=(X=1, ...(truncated)"
    Weight=1
    WeightMinimum=0
    ItemToModify.1.Transform.Rotation=/Script/RigVM.RigVMFunction_MathQuaternionSlerp(B=(X=0,Y=0,Z=0,W=1), T=0.5, A=@n3.Result) {is=struct RigVMUnitNode, template="Interpolate::Execute(in A,in B,in T,out Result)"}.Result
    ItemToModify.0.Transform.Rotation=@n3.Result
    ExecutePin=@RigVMFunction_Sequence.C
   )
   n4 = &RigVMFunction_MathQuaternionInverse(Value=&RigVMFunction_MathQuaternionSwingTwist(
     TwistAxis=(X=1,Y=0,Z=0)
     Input=&RigUnit_GetRelativeTransformForItem(Child=(Type=Bone,Name="upperarm_r"), Parent=(Type=Null,Name="upperarm_r_null"), bChildInitial=False, bParentInitial=False).RelativeTransform.Rotation
    ).Twist) {template="Inverse::Execute(in Value,out Result)"}
   Interpolate = &RigVMFunction_MathQuaternionSlerp(B=(X=0,Y=0,Z=0,W=1), T=0.5, A=@n4.Result) {template="Interpolate::Execute(in A,in B,in T,out Result)"}
   > &RigUnit_ModifyTransforms(
    ItemToModify="((Item=(Type=Bone,Name=\"upperarm_twist_01_r\"),Transform=(Rotation=(X=0,Y=0,Z=0,W=1),Translation=(X=0,Y=0,Z=0),Scale3D=(X=1,Y=1,Z=1))),(Item=(Type=Bone,Name=\"upperarm_twist_02_r\"),Transform=(Rotation=(X=0,Y=0,Z=0,W=1),Translation=(X=0,Y=0,Z=0),Scale3D=(X=1, ...(truncated)"
    Weight=1
    WeightMinimum=0
    ItemToModify.1.Transform.Rotation=@Interpolate.Result
    ItemToModify.0.Transform.Rotation=@n4.Result
   )
  ]}
 New Function_ContainedGraph {nodes [
   RigVMFunctionEntryNode()
   > DeltaRotation_1 = RigVMFunctionReferenceNode(Rotation Order=ZXY, child=hand_l, max=100, min=70, onlyTwist=true, parent=hand_l_null, parentType=Null, twistAxis=(X=1,Y=0,Z=0))
   > n1 = &RigUnit_ModifyTransforms(
    ItemToModify="((Item=(Type=Bone,Name=\"lowerarm_twist_01_l\"),Transform=(Rotation=(X=0,Y=-0.0262,Z=0,W=0.9997),Translation=(X=-0.7,Y=-0.17,Z=1),Scale3D=(X=1,Y=1,Z=1))),(Item=(Type=Bone,Name=\"lowerarm_twist_02_l\"),Transform=(Rotation=(X=0.342,Y=0,Z=0,W=0.9397),Translation= ...(truncated)"
    WeightMinimum=0
    Weight=@DeltaRotation_1.outputWeight
   )
   RigVMFunction_ControlFlowBranch = &RigVMFunction_ControlFlowBranch(Condition=False, ExecuteContext=@n1.ExecutePin)
   RigVMFunctionReturnNode(ExecuteContext=@RigVMFunction_ControlFlowBranch.Completed)
   DeltaRotation_1_1_1 = RigVMFunctionReferenceNode(
    Rotation Order=ZYX
    child=upperarm_l
    debugRotation=false
    max=90
    min=60
    onlyTwist=true
    parent=upperarm_l_null
    parentType=Null
    twistAxis=(X=1,Y=0,Z=0)
    ExecuteContext=@RigVMFunction_ControlFlowBranch.True
   )
   > &RigUnit_ModifyTransforms(
    ItemToModify=((Item=(Type=Bone,Name="upperarm_twist_02_l"),Transform=(Rotation=(X=0.1736,Y=0,Z=0,W=0.9848),Translation=(X=0,Y=0,Z=0),Scale3D=(X=1,Y=1,Z=1))))
    WeightMinimum=0
    Weight=@DeltaRotation_1_1_1.outputWeight
   )
   > DeltaRotation_1_1 = RigVMFunctionReferenceNode(Rotation Order=ZYX, child=upperarm_r, max=90, min=60, onlyTwist=true, parent=upperarm_r_null, parentType=Null, twistAxis=(X=1,Y=0,Z=0))
   > &RigUnit_ModifyTransforms(
    ItemToModify=((Item=(Type=Bone,Name="upperarm_twist_02_r"),Transform=(Rotation=(X=0.1736,Y=0,Z=0,W=0.9848),Translation=(X=0,Y=0,Z=0),Scale3D=(X=1,Y=1,Z=1))))
    WeightMinimum=0
    Weight=@DeltaRotation_1_1.outputWeight
   )
  ]}
 New Function_ContainedGraph {nodes [
   RigVMFunctionEntryNode()
   > RigVMFunction_Sequence = RigVMAggregateNode()
   &RigUnit_ParentConstraint(
    AdvancedSettings=(RotationOrderForFilter=XZY)
    Child=(Type=Bone,Name="base_chestRope_r")
    Parents=((Item=(Type=Bone,Name="spine_01"),Weight=0.0036),(Item=(Type=Bone,Name="spine_02"),Weight=0.0586),(Item=(Type=Bone,Name="spine_03"),Weight=0.1533),(Item=(Type=Bone,Name="spine_05"),Weight=0.2712),(Item=(Type=Bone,Name="spine_04"),Weight=0.5132))
    Weight=1
    ExecutePin=@RigVMFunction_Sequence.A
   )
   > &RigUnit_ParentConstraint(
    AdvancedSettings=(RotationOrderForFilter=XZY)
    Child=(Type=Bone,Name="base_chestRope_l")
    Parents=((Item=(Type=Bone,Name="spine_01"),Weight=0.0036),(Item=(Type=Bone,Name="spine_02"),Weight=0.0586),(Item=(Type=Bone,Name="spine_03"),Weight=0.1533),(Item=(Type=Bone,Name="spine_05"),Weight=0.2712),(Item=(Type=Bone,Name="spine_04"),Weight=0.5132))
    Weight=1
   )
   > &RigUnit_ParentConstraint(
    AdvancedSettings=(RotationOrderForFilter=XZY)
    Child=(Type=Bone,Name="base_backSyn")
    Parents="((Item=(Type=Bone,Name=\"spine_02\"),Weight=0.0112),(Item=(Type=Bone,Name=\"spine_03\"),Weight=0.0137),(Item=(Type=Bone,Name=\"clavicle_l\"),Weight=0.164),(Item=(Type=Bone,Name=\"clavicle_r\"),Weight=0.164),(Item=(Type=Bone,Name=\"spine_04\"),Weight=0.1827),(Item=(T ...(truncated)"
    Weight=1
   )
   > &RigUnit_ParentConstraint(
    AdvancedSettings=(RotationOrderForFilter=XZY)
    Child=(Type=Bone,Name="base_neckRope_l_B")
    Parents="((Item=(Type=Bone,Name=\"clavicle_r\"),Weight=0.0001),(Item=(Type=Bone,Name=\"neck_02\"),Weight=0.0146),(Item=(Type=Bone,Name=\"head\"),Weight=0.016),(Item=(Type=Bone,Name=\"clavicle_l\"),Weight=0.1613),(Item=(Type=Bone,Name=\"neck_01\"),Weight=0.2079),(Item=(Type=B ...(truncated)"
    Weight=1
   )
   > &RigUnit_ParentConstraint(
    AdvancedSettings=(RotationOrderForFilter=XZY)
    Child=(Type=Bone,Name="base_neckRope_l_A")
    Parents=((Item=(Type=Bone,Name="head"),Weight=0.0323),(Item=(Type=Bone,Name="neck_02"),Weight=0.0396),(Item=(Type=Bone,Name="clavicle_l"),Weight=0.0752),(Item=(Type=Bone,Name="neck_01"),Weight=0.3108),(Item=(Type=Bone,Name="spine_05"),Weight=0.5421))
    Weight=1
   )
   > &RigUnit_ParentConstraint(
    AdvancedSettings=(RotationOrderForFilter=XZY)
    Child=(Type=Bone,Name="base_neckRope_r_B")
    Parents="((Item=(Type=Bone,Name=\"clavicle_l\"),Weight=0.0001),(Item=(Type=Bone,Name=\"neck_02\"),Weight=0.0146),(Item=(Type=Bone,Name=\"head\"),Weight=0.016),(Item=(Type=Bone,Name=\"clavicle_r\"),Weight=0.1613),(Item=(Type=Bone,Name=\"neck_01\"),Weight=0.2079),(Item=(Type=B ...(truncated)"
    Weight=1
   )
   > &RigUnit_ParentConstraint(
    AdvancedSettings=(RotationOrderForFilter=XZY)
    Child=(Type=Bone,Name="base_neckRope_r_A")
    Parents=((Item=(Type=Bone,Name="head"),Weight=0.0324),(Item=(Type=Bone,Name="neck_02"),Weight=0.0396),(Item=(Type=Bone,Name="clavicle_r"),Weight=0.0752),(Item=(Type=Bone,Name="neck_01"),Weight=0.3108),(Item=(Type=Bone,Name="spine_05"),Weight=0.542))
    Weight=1
   )
   > &RigUnit_ParentConstraint(
    AdvancedSettings=(RotationOrderForFilter=XZY)
    Child=(Type=Bone,Name="base_breastRope_r_01")
    Parents="((Item=(Type=Bone,Name=\"spine_01\"),Weight=0.0002),(Item=(Type=Bone,Name=\"clavicle_r\"),Weight=0.0127),(Item=(Type=Bone,Name=\"spine_02\"),Weight=0.0273),(Item=(Type=Bone,Name=\"spine_03\"),Weight=0.0439),(Item=(Type=Bone,Name=\"spine_05\"),Weight=0.4476),(Item=(T ...(truncated)"
    Weight=1
   )
   > &RigUnit_ParentConstraint(
    AdvancedSettings=(RotationOrderForFilter=XZY)
    Child=(Type=Bone,Name="base_backKneeRope_r_01")
    Parents="((Item=(Type=Bone,Name=\"calf_twist_01_r\"),Weight=0.0001),(Item=(Type=Bone,Name=\"thigh_twist_01_r\"),Weight=0.0039),(Item=(Type=Bone,Name=\"calf_twist_02_r\"),Weight=0.0365),(Item=(Type=Bone,Name=\"deform_knee_out_r\"),Weight=0.1017),(Item=(Type=Bone,Name=\"thigh ...(truncated)"
    Weight=1
   )
   > &RigUnit_ParentConstraint(
    AdvancedSettings=(RotationOrderForFilter=XZY)
    Child=(Type=Bone,Name="base_kneePad")
    Parents="((Item=(Type=Bone,Name=\"calf_twist_01_r\"),Weight=0.0001),(Item=(Type=Bone,Name=\"deform_knee_in_r\"),Weight=0.0016),(Item=(Type=Bone,Name=\"thigh_twist_01_r\"),Weight=0.0055),(Item=(Type=Bone,Name=\"calf_twist_02_r\"),Weight=0.0294),(Item=(Type=Bone,Name=\"thigh_ ...(truncated)"
    Weight=1
   )
   > &RigUnit_ParentConstraint(
    AdvancedSettings=(RotationOrderForFilter=XZY)
    Child=(Type=Bone,Name="base_calfBelt")
    Parents="((Item=(Type=Bone,Name=\"thigh_r\"),Weight=0.0003),(Item=(Type=Bone,Name=\"deform_knee_out_r\"),Weight=0.0032),(Item=(Type=Bone,Name=\"calf_twist_01_r\"),Weight=0.122),(Item=(Type=Bone,Name=\"calf_r\"),Weight=0.1653),(Item=(Type=Bone,Name=\"calf_twist_02_r\"),Weight ...(truncated)"
    Weight=1
   )
   &RigUnit_ModifyTransforms(
    ItemToModify="((Item=(Type=Bone,Name=\"base_backKneeRope_r_01\"),Transform=(Rotation=(X=0,Y=0,Z=0,W=-1),Translation=(X=0,Y=-2.5378,Z=0.3436),Scale3D=(X=0.4235,Y=0.4783,Z=0.2602))),(Item=(Type=Bone,Name=\"base_kneePad\"),Transform=(Rotation=(X=0,Y=0,Z=0.0436,W=0.999),Transla ...(truncated)"
    WeightMinimum=0
    Weight=RigVMDispatchNode(BoneName=root, Default=0, Name=deformRig_r_knee_alpha) {template="DISPATCH_RigDispatch_GetAnimAttribute(in Name,in BoneName,in Default,out Value,out Found)"}.Value
    ExecutePin=@RigVMFunction_Sequence.B
   )
   DeltaRotation = RigVMFunctionReferenceNode(
    Rotation Order=ZYX
    child=clavicle_l
    max=-35
    parent=clavicle_l_null
    parentType=Null
    primAxis=2
    twistAxis=(X=1,Y=0,Z=0)
    ExecuteContext=@RigVMFunction_Sequence.C
   )
   > DeltaRotation_1 = RigVMFunctionReferenceNode(Rotation Order=ZYX, child=clavicle_r, max=-35, parent=clavicle_r_null, parentType=Null, primAxis=2, twistAxis=(X=1,Y=0,Z=0))
   Multiply = &RigVMFunction_MathDoubleMul(A=@DeltaRotation_1.outputWeight, B=@DeltaRotation.outputWeight)
   > &RigUnit_ModifyTransforms(
    ItemToModify=((Item=(Type=Bone,Name="base_backSyn"),Transform=(Rotation=(X=0,Y=0,Z=0,W=1),Translation=(X=0,Y=-2.5,Z=0),Scale3D=(X=1,Y=1,Z=1.1))))
    WeightMinimum=0
    Weight=@Multiply.Result
   )
   DeltaRotation_2 = RigVMFunctionReferenceNode(
    Rotation Order=ZYX
    child=clavicle_l
    max=-45
    min=-30
    parent=clavicle_l_null
    parentType=Null
    primAxis=1
    twistAxis=(X=1,Y=0,Z=0)
    ExecuteContext=@RigVMFunction_Sequence.D
   )
   > &RigUnit_ModifyTransforms(
    ItemToModify="((Item=(Type=Bone,Name=\"base_neckRope_l_B\"),Transform=(Rotation=(X=0,Y=0,Z=-0.3827,W=0.9239),Translation=(X=0,Y=0,Z=0),Scale3D=(X=1,Y=1,Z=1))),(Item=(Type=Bone,Name=\"base_neckRope_l_A\"),Transform=(Rotation=(X=0,Y=0,Z=-0.2588,W=0.9659),Translation=(X=0,Y=0, ...(truncated)"
    WeightMinimum=0
    Weight=@DeltaRotation_2.outputWeight
   )
   > DeltaRotation_1_1 = RigVMFunctionReferenceNode(Rotation Order=ZYX, child=clavicle_r, max=-45, min=-30, parent=clavicle_r_null, parentType=Null, primAxis=1, twistAxis=(X=1,Y=0,Z=0))
   > &RigUnit_ModifyTransforms(
    ItemToModify="((Item=(Type=Bone,Name=\"base_neckRope_r_B\"),Transform=(Rotation=(X=0,Y=0,Z=0.3827,W=0.9239),Translation=(X=0,Y=0,Z=0),Scale3D=(X=1,Y=1,Z=1))),(Item=(Type=Bone,Name=\"base_neckRope_r_A\"),Transform=(Rotation=(X=0,Y=0,Z=0.2588,W=0.9659),Translation=(X=0,Y=0,Z= ...(truncated)"
    WeightMinimum=0
    Weight=@DeltaRotation_1_1.outputWeight
   )
   > DeltaRotation_1_1_1 = RigVMFunctionReferenceNode(Rotation Order=ZYX, child=neck_01, debugRotation=false, max=-50, parent=neck_01_null, parentType=Null, primAxis=2, twistAxis=(X=1,Y=0,Z=0))
   > &RigUnit_ModifyTransforms(
    ItemToModify="((Item=(Type=Bone,Name=\"base_neckRope_l_B\"),Transform=(Rotation=(X=0,Y=0,Z=-0.1736,W=0.9848),Translation=(X=0,Y=0,Z=0),Scale3D=(X=1,Y=1,Z=1))),(Item=(Type=Bone,Name=\"base_neckRope_r_B\"),Transform=(Rotation=(X=0,Y=0,Z=0.1736,W=0.9848),Translation=(X=0,Y=0,Z ...(truncated)"
    WeightMinimum=0
    Weight=@DeltaRotation_1_1_1.outputWeight
   )
   > DeltaRotation_1_1_1_1 = RigVMFunctionReferenceNode(Rotation Order=ZYX, child=neck_01, debugRotation=false, max=-15, parent=neck_01_null, parentType=Null, primAxis=1, twistAxis=(X=1,Y=0,Z=0))
   > &RigUnit_ModifyTransforms(
    ItemToModify=((Item=(Type=Bone,Name="base_neckRope_r_B"),Transform=(Rotation=(X=0,Y=0,Z=0.1305,W=0.9914),Translation=(X=0,Y=0.2,Z=0),Scale3D=(X=1,Y=1,Z=1))))
    WeightMinimum=0
    Weight=@DeltaRotation_1_1_1_1.outputWeight
   )
   RigVMFunctionReturnNode(ExecuteContext=@RigVMFunction_Sequence.E)
  ]}
 New Function_ContainedGraph {nodes [
   n1 = RigVMFunctionEntryNode()
   RigVMFunction_MathBoolToInteger = &RigVMFunction_MathBoolToInteger(Value=@n1.onlyTwist)
   RigVMFunction_MathQuaternionSwingTwist = &RigVMFunction_MathQuaternionSwingTwist(
    TwistAxis=@n1.twistAxis
    Input=&RigUnit_GetRelativeTransformForItem(
     Child=(Type=Bone,Name="None")
     Parent=(Type=Bone,Name="None")
     bChildInitial=False
     bParentInitial=False
     Child.Name=@n1.child
     Parent.Name=@n1.parent
     Parent.Type=@n1.parentType
    ).RelativeTransform.Rotation
   )
   RigVMFunction_MathQuaternionToEuler = &RigVMFunction_MathQuaternionToEuler(
    Value=RigVMDispatchNode(
     Values=((),())
     Index=@RigVMFunction_MathBoolToInteger.Result
     Values.0=@RigVMFunction_MathQuaternionSwingTwist.Swing
     Values.1=@RigVMFunction_MathQuaternionSwingTwist.Twist
    ) {template="DISPATCH_RigVMDispatch_SelectInt32(in Index,in Values,out Result)"}.Result
    RotationOrder=@n1.Rotation Order
   )
   > DISPATCH_RigVMDispatch_Print = RigVMDispatchNode(ScreenColor=(R=1,G=1,B=1,A=1), ScreenDuration=1.5, Enabled=@n1.debugRotation, Value=@RigVMFunction_MathQuaternionToEuler.Result) {template="DISPATCH_RigVMDispatch_Print(in Prefix,in Value,in Enabled,in ScreenDuration,in ScreenColor)"}
   Clamp = &RigVMFunction_MathIntClamp(Value=@n1.primAxis)
   DISPATCH_RigVMDispatch_SelectInt32 = RigVMDispatchNode(
    Values=(0,0,0)
    Index=@Clamp.Result
    Values.0=@RigVMFunction_MathQuaternionToEuler.Result.X
    Values.1=@RigVMFunction_MathQuaternionToEuler.Result.Y
    Values.2=@RigVMFunction_MathQuaternionToEuler.Result.Z
   ) {template="DISPATCH_RigVMDispatch_SelectInt32(in Index,in Values,out Result)"}
   Remap = &RigVMFunction_MathDoubleRemap(
    TargetMaximum=1
    TargetMinimum=0
    bClamp=False
    Value=@DISPATCH_RigVMDispatch_SelectInt32.Result
    SourceMaximum=@n1.max
    SourceMinimum=@n1.min
   )
   > RigVMFunctionReturnNode(outputWeight=&RigVMFunction_MathDoubleClamp(Value=@Remap.Result).Result)
  ]}
 New Function_ContainedGraph {nodes [
   RigVMFunctionEntryNode()
   > RigVMFunction_Sequence = RigVMAggregateNode()
   n1 = &RigVMFunction_ControlFlowBranch(ExecuteContext=@RigVMFunction_Sequence.A, Condition=RigVMVariableNode() {variable=RBAN_on}.Value)
   &RigUnit_ParentConstraint(
    AdvancedSettings=(InterpolationType=Average,RotationOrderForFilter=XZY)
    Child=(Type=Bone,Name="base_skirt_side_frnt_l")
    Filter=(TranslationFilter=(bX=true,bY=false,bZ=true),RotationFilter=(bX=true,bY=true,bZ=true),ScaleFilter=(bX=True,bY=True,bZ=True))
    Parents=((Item=(Type=Bone,Name="thigh_twist_01_l"),Weight=1))
    Weight=1
    bMaintainOffset=true
    ExecutePin=@n1.True
   )
   > &RigUnit_ParentConstraint(
    AdvancedSettings=(InterpolationType=Average,RotationOrderForFilter=XZY)
    Child=(Type=Bone,Name="base_skirt_side_l")
    Filter=(TranslationFilter=(bX=false,bY=false,bZ=false),RotationFilter=(bX=True,bY=True,bZ=True),ScaleFilter=(bX=True,bY=True,bZ=True))
    Parents=((Item=(Type=Bone,Name="thigh_twist_01_l"),Weight=1))
    Weight=1
   )
   > &RigUnit_ParentConstraint(
    AdvancedSettings=(InterpolationType=Average,RotationOrderForFilter=XZY)
    Child=(Type=Bone,Name="base_skirt_side_bk_l")
    Filter=(TranslationFilter=(bX=false,bY=false,bZ=false),RotationFilter=(bX=True,bY=True,bZ=True),ScaleFilter=(bX=True,bY=True,bZ=True))
    Parents=((Item=(Type=Bone,Name="thigh_twist_01_l"),Weight=0.5),(Item=(Type=Bone,Name="pelvis"),Weight=0.5),(Item=(Type=Bone,Name="None"),Weight=0))
    Weight=1
   )
   > &RigUnit_ParentConstraint(
    AdvancedSettings=(InterpolationType=Average,RotationOrderForFilter=XZY)
    Child=(Type=Bone,Name="base_skirt_side_frnt_r")
    Filter=(TranslationFilter=(bX=true,bY=false,bZ=true),RotationFilter=(bX=True,bY=True,bZ=true),ScaleFilter=(bX=True,bY=True,bZ=True))
    Parents=((Item=(Type=Bone,Name="thigh_r"),Weight=1))
    Weight=1
   )
   > &RigUnit_ParentConstraint(
    AdvancedSettings=(InterpolationType=Average,RotationOrderForFilter=XZY)
    Child=(Type=Bone,Name="base_skirt_side_r")
    Filter=(TranslationFilter=(bX=false,bY=false,bZ=false),RotationFilter=(bX=True,bY=True,bZ=True),ScaleFilter=(bX=True,bY=True,bZ=True))
    Parents=((Item=(Type=Bone,Name="thigh_r"),Weight=1))
    Weight=1
   )
   > &RigUnit_ParentConstraint(
    AdvancedSettings=(InterpolationType=Average,RotationOrderForFilter=XZY)
    Child=(Type=Bone,Name="base_skirt_side_bk_r")
    Filter=(TranslationFilter=(bX=false,bY=false,bZ=false),RotationFilter=(bX=True,bY=True,bZ=True),ScaleFilter=(bX=True,bY=True,bZ=True))
    Parents=((Item=(Type=Bone,Name="thigh_twist_01_r"),Weight=0.5),(Item=(Type=Bone,Name="pelvis"),Weight=0.5))
    Weight=1
   )
   > &RigUnit_ParentConstraint(
    AdvancedSettings=(InterpolationType=Average,RotationOrderForFilter=XZY)
    Child=(Type=Bone,Name="base_skirt_frnt_mid")
    Filter=(TranslationFilter=(bX=true,bY=false,bZ=true),RotationFilter=(bX=true,bY=true,bZ=true),ScaleFilter=(bX=True,bY=True,bZ=True))
    Parents=((Item=(Type=Bone,Name="thigh_twist_01_l"),Weight=0.5),(Item=(Type=Bone,Name="thigh_twist_01_r"),Weight=0.5))
    Weight=1
   )
   > &RigUnit_ParentConstraint(
    AdvancedSettings=(InterpolationType=Average,RotationOrderForFilter=XZY)
    Child=(Type=Bone,Name="base_skirt_bk_mid")
    Filter=(TranslationFilter=(bX=false,bY=false,bZ=false),RotationFilter=(bX=True,bY=True,bZ=True),ScaleFilter=(bX=True,bY=True,bZ=True))
    Parents=((Item=(Type=Bone,Name="thigh_l"),Weight=0.4),(Item=(Type=Bone,Name="thigh_r"),Weight=0.4),(Item=(Type=Bone,Name="pelvis"),Weight=0.3),(Item=(Type=Bone,Name="deform_glute_r"),Weight=0))
    Weight=1
   )
   > n2 = &RigUnit_SphericalPoseReader(
    ActiveRegionScaleFactors=(PositiveWidth=3.5,NegativeWidth=0.5,PositiveHeight=3,NegativeHeight=2)
    ActiveRegionSize=0.1
    Debug=(bDrawDebug=false,bDraw2D=false,DebugScale=25,DebugSegments=20,DebugThickness=0.25)
    DriverAxis=(X=-1,Y=0,Z=0)
    DriverItem=(Type=Bone,Name="calf_l")
    FalloffRegionScaleFactors=(PositiveWidth=1,NegativeWidth=1,PositiveHeight=1,NegativeHeight=1)
    FalloffSize=0.5
    FlipHeightScaling=true
    FlipWidthScaling=false
    OptionalParentItem=(Type=Bone,Name="root")
    RotationOffset=(X=270,Y=0,Z=-5)
   )
   Interpolate = /Script/RigVM.RigVMFunction_MathDoubleLerp(B=0, A=@n2.OutputParam, T=RigVMVariableNode() {variable=is skydiving}.Value) {is=struct RigVMUnitNode, template="Interpolate::Execute(in A,in B,in T,out Result)"}
   > &RigUnit_ModifyTransforms(
    ItemToModify=((Item=(Type=Bone,Name="base_skirt_side_bk_l"),Transform=(Rotation=(X=-0.0904,Y=-0.2868,Z=-0.2868,W=0.9096),Translation=(X=0,Y=0,Z=0),Scale3D=(X=1,Y=1,Z=1))))
    WeightMinimum=0
    Weight=@Interpolate.Result
   )
   > n3 = &RigUnit_SphericalPoseReader(
    ActiveRegionScaleFactors=(PositiveWidth=3.5,NegativeWidth=0.5,PositiveHeight=3,NegativeHeight=2)
    ActiveRegionSize=0.1
    Debug=(bDrawDebug=false,bDraw2D=false,DebugScale=25,DebugSegments=20,DebugThickness=0.25)
    DriverAxis=(X=-1,Y=0,Z=0)
    DriverItem=(Type=Bone,Name="calf_r")
    FalloffRegionScaleFactors=(PositiveWidth=1,NegativeWidth=1,PositiveHeight=1,NegativeHeight=1)
    FalloffSize=0.5
    FlipHeightScaling=false
    FlipWidthScaling=false
    OptionalParentItem=(Type=Bone,Name="root")
    RotationOffset=(X=270,Y=0,Z=-5)
   )
   > &RigUnit_ModifyTransforms(
    ItemToModify=((Item=(Type=Bone,Name="base_skirt_side_bk_r"),Transform=(Rotation=(X=0.0904,Y=-0.2868,Z=0.2868,W=0.9096),Translation=(X=0,Y=0,Z=0),Scale3D=(X=1,Y=1,Z=1))))
    WeightMinimum=0
    Weight=&RigVMFunction_MathDoubleLerp(B=0, A=@n3.OutputParam, T=RigVMVariableNode() {variable=is skydiving}.Value) {template="Interpolate::Execute(in A,in B,in T,out Result)"}.Result
   )
   n4 = RigVMDispatchNode(BoneName=root, Default=0, Name=deformRig_l_thigh_up_alpha) {template="DISPATCH_RigDispatch_GetAnimAttribute(in Name,in BoneName,in Default,out Value,out Found)"}
   Clamp = &RigVMFunction_MathDoubleClamp(Maximum=0.15, Value=RigVMVariableNode() {variable=isCrouchingMoving}.Value)
   n5 = RigVMVariableNode() {variable=isCrouching}
   &RigUnit_ModifyTransforms(
    ItemToModify="((Item=(Type=Bone,Name=\"base_skirt_side_frnt_l\"),Transform=(Rotation=(X=0.0302,Y=-0.171,Z=0.171,W=0.9698),Translation=(X=0,Y=0,Z=0),Scale3D=(X=1,Y=1,Z=1))),(Item=(Type=Bone,Name=\"base_skirt_frnt_mid\"),Transform=(Rotation=(X=0,Y=-0.0872,Z=0,W=0.9962),Transl ...(truncated)"
    WeightMinimum=0
    Weight=&RigVMFunction_MathDoubleMul(
     A=@n4.Value
     B=RigVMDispatchNode(False=1, True=@Clamp.Result, Condition=@n5.Value) {template="DISPATCH_RigVMDispatch_If(in Condition,in True,in False,out Result)"}.Result
    ).Result
    ExecutePin=@RigVMFunction_Sequence.B
   )
   n6 = RigVMDispatchNode(BoneName=root, Default=0, Name=deformRig_r_thigh_up_alpha) {template="DISPATCH_RigDispatch_GetAnimAttribute(in Name,in BoneName,in Default,out Value,out Found)"}
   n7 = &RigVMFunction_MathDoubleClamp(Maximum=0.15, Value=RigVMVariableNode() {variable=isCrouchingMoving}.Value)
   n8 = RigVMVariableNode() {variable=isCrouching}
   > &RigUnit_ModifyTransforms(
    ItemToModify="((Item=(Type=Bone,Name=\"base_skirt_side_frnt_r\"),Transform=(Rotation=(X=-0.0302,Y=-0.171,Z=-0.171,W=0.9698),Translation=(X=0,Y=0,Z=0),Scale3D=(X=1,Y=1,Z=1))),(Item=(Type=Bone,Name=\"base_skirt_frnt_mid\"),Transform=(Rotation=(X=0,Y=-0.0872,Z=0,W=0.9962),Tran ...(truncated)"
    WeightMinimum=0
    Weight=&RigVMFunction_MathDoubleMul(
     A=@n6.Value
     B=RigVMDispatchNode(False=1, True=@n7.Result, Condition=@n8.Value) {template="DISPATCH_RigVMDispatch_If(in Condition,in True,in False,out Result)"}.Result
    ).Result
   )
   n9 = &RigVMFunction_MathDoubleClamp(Maximum=0.15, Value=RigVMVariableNode() {variable=isCrouchingMoving}.Value)
   n10 = RigVMVariableNode() {variable=isCrouching}
   Multiply = &RigVMFunction_MathDoubleMul(A=@n6.Value, B=@n4.Value)
   n11 = &RigVMFunction_AnimEvalRichCurve(
    Curve="(EditorCurveData=(Keys=((InterpMode=RCIM_Cubic,TangentMode=RCTM_User,ArriveTangent=1.6904,LeaveTangent=1.6904),(InterpMode=RCIM_Cubic,TangentMode=RCTM_User,Time=0.4643,Value=0.6384,ArriveTangent=1.0934,LeaveTangent=1.0934),(InterpMode=RCIM_Cubic,TangentMod ...(truncated)"
    SourceMinimum=0
    TargetMinimum=0
    Value=@Multiply.Result
   )
   > &RigUnit_ModifyTransforms(
    ItemToModify="((Item=(Type=Bone,Name=\"base_skirt_frnt_mid\"),Transform=(Rotation=(X=0,Y=-0.0872,Z=0,W=0.9962),Translation=(X=0,Y=0,Z=0),Scale3D=(X=1,Y=1,Z=1))),(Item=(Type=Bone,Name=\"base_skirt_side_frnt_l\"),Transform=(Rotation=(X=0,Y=-0.0872,Z=0,W=0.9962),Translation=(X ...(truncated)"
    WeightMinimum=0
    Weight=&RigVMFunction_MathDoubleMul(
     B=RigVMDispatchNode(False=1, True=@n9.Result, Condition=@n10.Value) {template="DISPATCH_RigVMDispatch_If(in Condition,in True,in False,out Result)"}.Result
     A=@n11.Result
    ).Result
   )
   RigVMFunctionReturnNode(ExecuteContext=@RigVMFunction_Sequence.C)
   n12 = RigVMDispatchNode(BoneName=root, Default=0, Name=deformRig_l_thigh_up_alpha) {template="DISPATCH_RigDispatch_GetAnimAttribute(in Name,in BoneName,in Default,out Value,out Found)"}
   n13 = &RigVMFunction_MathDoubleClamp(Maximum=0.15, Value=RigVMVariableNode() {variable=isCrouchingMoving}.Value)
   n14 = RigVMVariableNode() {variable=isCrouching}
   &RigUnit_ModifyTransforms(
    ExecutePin=()
    ItemToModify="((Item=(Type=Bone,Name=\"base_skirt_side_frnt_l\"),Transform=(Rotation=(X=0.329,Y=-0.4698,Z=0.4698,W=0.671),Translation=(X=0,Y=0,Z=0),Scale3D=(X=1,Y=1,Z=1))),(Item=(Type=Bone,Name=\"base_skirt_frnt_mid\"),Transform=(Rotation=(X=0,Y=-0.2588,Z=0,W=0.9659),Transl ...(truncated)"
    WeightMinimum=0
    Weight=&RigVMFunction_MathDoubleMul(
     A=@n12.Value
     B=RigVMDispatchNode(False=0.25, True=@n13.Result, Condition=@n14.Value) {template="DISPATCH_RigVMDispatch_If(in Condition,in True,in False,out Result)"}.Result
    ).Result
   )
   n15 = RigVMDispatchNode(BoneName=root, Default=0, Name=deformRig_r_thigh_up_alpha) {template="DISPATCH_RigDispatch_GetAnimAttribute(in Name,in BoneName,in Default,out Value,out Found)"}
   n16 = &RigVMFunction_MathDoubleClamp(Maximum=0.15, Value=RigVMVariableNode() {variable=isCrouchingMoving}.Value)
   n17 = RigVMVariableNode() {variable=isCrouching}
   > &RigUnit_ModifyTransforms(
    ItemToModify="((Item=(Type=Bone,Name=\"base_skirt_side_frnt_r\"),Transform=(Rotation=(X=-0.329,Y=-0.4698,Z=-0.4698,W=0.671),Translation=(X=0,Y=0,Z=0),Scale3D=(X=1,Y=1,Z=1))),(Item=(Type=Bone,Name=\"base_skirt_frnt_mid\"),Transform=(Rotation=(X=0,Y=-0.2588,Z=0,W=0.9659),Tran ...(truncated)"
    WeightMinimum=0
    Weight=&RigVMFunction_MathDoubleMul(
     A=@n15.Value
     B=RigVMDispatchNode(False=0.5, True=@n16.Result, Condition=@n17.Value) {template="DISPATCH_RigVMDispatch_If(in Condition,in True,in False,out Result)"}.Result
    ).Result
   )
   n18 = &RigVMFunction_MathDoubleClamp(Maximum=0.15, Value=RigVMVariableNode() {variable=isCrouchingMoving}.Value)
   n19 = RigVMVariableNode() {variable=isCrouching}
   n20 = &RigVMFunction_AnimEvalRichCurve(
    Curve="(EditorCurveData=(Keys=((InterpMode=RCIM_Cubic,TangentMode=RCTM_User,ArriveTangent=1.6904,LeaveTangent=1.6904),(InterpMode=RCIM_Cubic,TangentMode=RCTM_User,Time=0.4643,Value=0.6384,ArriveTangent=1.0934,LeaveTangent=1.0934),(InterpMode=RCIM_Cubic,TangentMod ...(truncated)"
    SourceMinimum=0
    TargetMinimum=0
    Value=&RigVMFunction_MathDoubleMul(A=@n15.Value, B=@n12.Value).Result
   )
   > &RigUnit_ModifyTransforms(
    ItemToModify="((Item=(Type=Bone,Name=\"base_skirt_frnt_mid\"),Transform=(Rotation=(X=0,Y=-0.2588,Z=0,W=0.9659),Translation=(X=0,Y=0,Z=0),Scale3D=(X=1,Y=1,Z=1))),(Item=(Type=Bone,Name=\"base_skirt_side_frnt_l\"),Transform=(Rotation=(X=0,Y=-0.2164,Z=0,W=0.9763),Translation=(X ...(truncated)"
    WeightMinimum=0
    Weight=&RigVMFunction_MathDoubleMul(
     B=RigVMDispatchNode(False=0.5, True=@n18.Result, Condition=@n19.Value) {template="DISPATCH_RigVMDispatch_If(in Condition,in True,in False,out Result)"}.Result
     A=@n20.Result
    ).Result
   )
   n21 = &RigUnit_SphericalPoseReader(
    ActiveRegionScaleFactors=(PositiveWidth=3.5,NegativeWidth=0.5,PositiveHeight=3,NegativeHeight=3)
    ActiveRegionSize=0.1
    Debug=(bDrawDebug=true,bDraw2D=false,DebugScale=25,DebugSegments=20,DebugThickness=0.25)
    DriverAxis=(X=-1,Y=0,Z=0)
    DriverItem=(Type=Bone,Name="thigh_l")
    ExecutePin=()
    FalloffRegionScaleFactors=(PositiveWidth=1,NegativeWidth=1,PositiveHeight=1,NegativeHeight=1)
    FalloffSize=0.5
    OptionalParentItem=(Type=Bone,Name="spine_04")
    RotationOffset=(X=90,Y=0,Z=0)
   )
   > &RigUnit_ModifyTransforms(
    ItemToModify="((Item=(Type=Bone,Name=\"base_skirt_side_frnt_l\"),Transform=(Rotation=(X=0.1464,Y=-0.3536,Z=0.3536,W=0.8536),Translation=(X=0,Y=0,Z=0),Scale3D=(X=1,Y=1,Z=1))),(Item=(Type=Bone,Name=\"base_skirt_frnt_mid\"),Transform=(Rotation=(X=0,Y=0,Z=0.3827,W=0.9239),Trans ...(truncated)"
    WeightMinimum=0
    Weight=@n21.OutputParam
   )
   > n22 = &RigUnit_SphericalPoseReader(
    ActiveRegionScaleFactors=(PositiveWidth=3.5,NegativeWidth=0.5,PositiveHeight=3,NegativeHeight=3)
    ActiveRegionSize=0.1
    Debug=(bDrawDebug=true,DebugScale=25,DebugSegments=20,DebugThickness=0.25)
    DriverAxis=(X=1,Y=0,Z=0)
    DriverItem=(Type=Bone,Name="thigh_r")
    FalloffRegionScaleFactors=(PositiveWidth=1,NegativeWidth=1,PositiveHeight=1,NegativeHeight=1)
    FalloffSize=0.5
    FlipHeightScaling=false
    FlipWidthScaling=true
    OptionalParentItem=(Type=Bone,Name="spine_02")
    RotationOffset=(X=-90,Y=0,Z=0)
   )
   > &RigUnit_ModifyTransforms(
    ItemToModify="((Item=(Type=Bone,Name=\"base_skirt_side_frnt_r\"),Transform=(Rotation=(X=-0.1464,Y=-0.3536,Z=-0.3536,W=0.8536),Translation=(X=0,Y=0,Z=0),Scale3D=(X=1,Y=1,Z=1))),(Item=(Type=Bone,Name=\"base_skirt_frnt_mid\"),Transform=(Rotation=(X=0,Y=0,Z=-0.3827,W=0.9239),Tr ...(truncated)"
    WeightMinimum=0
    Weight=@n22.OutputParam
   )
  ]}
 New Function_ContainedGraph {nodes [
   n1 = RigVMFunctionEntryNode()
   n2 = &RigUnit_GetRelativeTransformForItem(
    Child=(Type=Bone,Name="None")
    Parent=(Type=Bone,Name="None")
    bChildInitial=False
    bParentInitial=False
    Child.Name=@n1.Child
    Parent.Name=@n1.Parent
   )
   > RigVMFunctionReturnNode(Delta=&RigVMFunction_MathDoubleRemap(
     TargetMaximum=1
     TargetMinimum=0
     Value=&RigVMFunction_MathQuaternionToEuler(RotationOrder=ZYX, Value=@n2.RelativeTransform.Rotation).Result.Z
     SourceMaximum=@n1.Source Max
     SourceMinimum=@n1.Source Min
    ).Result)
  ]}
 New Function_ContainedGraph {nodes [
   RigVMFunctionEntryNode()
   > DeltaRotationZ = RigVMFunctionReferenceNode(Child=calf_l, Parent=thigh_l, Source Max=120, Source Min=80)
   > RigVMVariableNode(Value=@DeltaRotationZ.Delta) {variable=thingh_Rot_L}
   > DeltaRotationZ_1 = RigVMFunctionReferenceNode(Child=calf_r, Parent=thigh_r, Source Max=120, Source Min=80)
   > RigVMVariableNode(Value=@DeltaRotationZ_1.Delta) {variable=thingh_Rot_R}
   > RigVMFunctionReturnNode()
  ]}
 New Function_ContainedGraph {nodes [
   n1 = RigVMFunctionEntryNode()
   RigVMFunction_ControlFlowBranch = &RigVMFunction_ControlFlowBranch(ExecuteContext=@n1.ExecuteContext, Condition=RigVMVariableNode() {variable=isSlidingwithPickAxe}.Value)
   RigVMFunctionReturnNode(ExecuteContext=@RigVMFunction_ControlFlowBranch.Completed)
   &RigUnit_ModifyTransforms(
    ItemToModify=((Item=(Type=Bone,Name="base_skirt_bk_mid"),Transform=(Rotation=(X=0,Y=-0.8192,Z=0,W=0.5736),Translation=(X=0,Y=0,Z=0),Scale3D=(X=1,Y=1,Z=1))))
    Weight=1
    WeightMinimum=0
    ExecutePin=@RigVMFunction_ControlFlowBranch.True
   )
   > &RigUnit_ModifyTransforms(
    ItemToModify=((Item=(Type=Bone,Name="base_skirt_side_bk_l"),Transform=(Rotation=(X=0,Y=-0.8192,Z=0,W=0.5736),Translation=(X=0,Y=0,Z=0),Scale3D=(X=1,Y=1,Z=1))))
    Weight=1
    WeightMinimum=0
   )
   > &RigUnit_ModifyTransforms(
    ItemToModify=((Item=(Type=Bone,Name="base_skirt_side_bk_r"),Transform=(Rotation=(X=0,Y=-0.5736,Z=0,W=0.8192),Translation=(X=0,Y=0,Z=0),Scale3D=(X=1,Y=1,Z=1))))
    Weight=1
    WeightMinimum=0
   )
   RigVMFunction_MathVectorDistance = &RigVMFunction_MathVectorDistance(
    A=&RigUnit_GetTransform(Item=(Type=Bone,Name="dyn_skirt_bk_mid_01"), bInitial=False).Transform.Translation
    B=&RigUnit_GetTransform(Item=(Type=Bone,Name="foot_r"), bInitial=False).Transform.Translation
   )
   Remap = &RigVMFunction_MathDoubleRemap(SourceMaximum=45, SourceMinimum=50, TargetMaximum=1, TargetMinimum=0, Value=@RigVMFunction_MathVectorDistance.Result)
  ]}
 New Function_ContainedGraph {nodes [
   RigVMFunctionEntryNode()
   > RigVMFunction_Sequence = RigVMAggregateNode()
   &RigUnit_ParentConstraint(
    AdvancedSettings=(InterpolationType=Average,RotationOrderForFilter=XZY)
    Child=(Type=Control,Name="dfrm_arm_spline_01_l_ctrl")
    Parents=((Item=(Type=Bone,Name="upperarm_twist_01_l"),Weight=0.5),(Item=(Type=Bone,Name="upperarm_l"),Weight=0.5))
    Weight=1
    ExecutePin=@RigVMFunction_Sequence.A
   )
   > &RigUnit_ParentConstraint(
    AdvancedSettings=(InterpolationType=Average,RotationOrderForFilter=XZY)
    Child=(Type=Control,Name="dfrm_arm_spline_03_l_ctrl")
    Parents=((Item=(Type=Bone,Name="upperarm_twist_01_l"),Weight=0.5),(Item=(Type=Bone,Name="upperarm_twist_02_l"),Weight=0.5))
    Weight=1
   )
   > &RigUnit_ParentConstraint(
    AdvancedSettings=(InterpolationType=Average,RotationOrderForFilter=XZY)
    Child=(Type=Control,Name="dfrm_arm_spline_05_l_ctrl")
    Parents=((Item=(Type=Bone,Name="lowerarm_l"),Weight=1))
    Weight=1
   )
   > &RigUnit_ParentConstraint(
    AdvancedSettings=(InterpolationType=Average,RotationOrderForFilter=XZY)
    Child=(Type=Control,Name="dfrm_arm_spline_07_l_ctrl")
    Parents=((Item=(Type=Bone,Name="lowerarm_twist_02_l"),Weight=1),(Item=(Type=None,Name="hand_l"),Weight=0.5))
    Weight=1
   )
   > &RigUnit_ParentConstraint(
    AdvancedSettings=(InterpolationType=Average,RotationOrderForFilter=XZY)
    Child=(Type=Control,Name="dfrm_arm_spline_09_l_ctrl")
    Filter=(TranslationFilter=(bX=True,bY=True,bZ=True),RotationFilter=(bX=false,bY=false,bZ=false),ScaleFilter=(bX=True,bY=True,bZ=True))
    Parents=((Item=(Type=Bone,Name="hand_l"),Weight=1))
    Weight=1
   )
   > &RigUnit_ModifyTransforms(
    ItemToModify="((Item=(Type=Control,Name=\"dfrm_arm_spline_03_l_ctrl\"),Transform=(Rotation=(X=0,Y=0,Z=0,W=1),Translation=(X=0,Y=-5,Z=0),Scale3D=(X=1,Y=1,Z=1))),(Item=(Type=Control,Name=\"dfrm_arm_spline_07_l_ctrl\"),Transform=(Rotation=(X=0,Y=0,Z=0,W=1),Translation=(X=0,Y=- ...(truncated)"
    WeightMinimum=0
    Weight=RigVMDispatchNode(BoneName=root, Default=0, Name=deformRig_l_elbow_alpha) {template="DISPATCH_RigDispatch_GetAnimAttribute(in Name,in BoneName,in Default,out Value,out Found)"}.Value
   )
   RerouteNode_3 = RigVMDispatchNode(Values="((Type=Bone,Name=\"dfrm_arm_spline_01_l\"),((),Type=Bone,Name=\"dfrm_arm_spline_02_l\"),((),Type=Bone,Name=\"dfrm_arm_spline_03_l\"),((),Type=Bone,Name=\"dfrm_arm_spline_04_l\"),((),Type=Bone,Name=\"dfrm_arm_spline_05_l\"),((),Type=Bone,Name=\"dfrm_arm_spline_06_l\"), ...(truncated)") {template="DISPATCH_RigVMDispatch_ArrayMake(in Values,out Array)"}
   > SplineIK_3 = RigVMFunctionReferenceNode(
    Primary Axis=(X=1,Y=0,Z=0)
    Samples Per Segment=16
    Secondary Spline Direction=(X=0,Y=0,Z=1)
    Spline Mode=BSpline
    Stretch=true
    Up Axis=(X=0,Y=0,Z=1)
    Bones=@RerouteNode_3.Array
    Controls=RigVMVariableNode() {variable=arm_l_ctrls}.Value
   )
   &RigUnit_ParentConstraint(
    AdvancedSettings=(InterpolationType=Average,RotationOrderForFilter=XZY)
    Child=(Type=Control,Name="dfrm_arm_spline_01_r_ctrl")
    Parents=((Item=(Type=Bone,Name="upperarm_twist_01_r"),Weight=0.5),(Item=(Type=Bone,Name="upperarm_r"),Weight=0.5))
    Weight=1
    ExecutePin=@RigVMFunction_Sequence.B
   )
   > &RigUnit_ParentConstraint(
    AdvancedSettings=(InterpolationType=Average,RotationOrderForFilter=XZY)
    Child=(Type=Control,Name="dfrm_arm_spline_03_r_ctrl")
    Parents=((Item=(Type=Bone,Name="upperarm_twist_01_r"),Weight=0.5),(Item=(Type=Bone,Name="upperarm_twist_02_r"),Weight=0.5))
    Weight=1
    bMaintainOffset=true
   )
   > &RigUnit_ParentConstraint(
    AdvancedSettings=(InterpolationType=Average,RotationOrderForFilter=XZY)
    Child=(Type=Control,Name="dfrm_arm_spline_05_r_ctrl")
    Parents=((Item=(Type=Bone,Name="lowerarm_r"),Weight=0.5),(Item=(Type=Bone,Name="upperarm_r"),Weight=0.5))
    Weight=1
    bMaintainOffset=true
   )
   > &RigUnit_ParentConstraint(
    AdvancedSettings=(InterpolationType=Average,RotationOrderForFilter=XZY)
    Child=(Type=Control,Name="dfrm_arm_spline_07_r_ctrl")
    Parents=((Item=(Type=Bone,Name="lowerarm_twist_02_r"),Weight=1))
    Weight=1
    bMaintainOffset=true
   )
   > &RigUnit_ParentConstraint(
    AdvancedSettings=(InterpolationType=Average,RotationOrderForFilter=XZY)
    Child=(Type=Control,Name="dfrm_arm_spline_09_r_ctrl")
    Filter=(TranslationFilter=(bX=True,bY=True,bZ=True),RotationFilter=(bX=false,bY=false,bZ=false),ScaleFilter=(bX=True,bY=True,bZ=True))
    Parents=((Item=(Type=Bone,Name="hand_r"),Weight=1))
    Weight=1
   )
   > &RigUnit_ModifyTransforms(
    ItemToModify="((Item=(Type=Control,Name=\"dfrm_arm_spline_03_r_ctrl\"),Transform=(Rotation=(X=0,Y=0,Z=0,W=1),Translation=(X=0,Y=5,Z=0),Scale3D=(X=1,Y=1,Z=1))),(Item=(Type=Control,Name=\"dfrm_arm_spline_07_r_ctrl\"),Transform=(Rotation=(X=0,Y=0,Z=0,W=1),Translation=(X=0,Y=5, ...(truncated)"
    WeightMinimum=0
    Weight=RigVMDispatchNode(BoneName=root, Default=0, Name=deformRig_r_elbow_alpha) {template="DISPATCH_RigDispatch_GetAnimAttribute(in Name,in BoneName,in Default,out Value,out Found)"}.Value
   )
   RerouteNode_2 = RigVMDispatchNode(Values="((Type=Bone,Name=\"dfrm_arm_spline_01_r\"),((),Type=Bone,Name=\"dfrm_arm_spline_02_r\"),((),Type=Bone,Name=\"dfrm_arm_spline_03_r\"),((),Type=Bone,Name=\"dfrm_arm_spline_04_r\"),((),Type=Bone,Name=\"dfrm_arm_spline_05_r\"),((),Type=Bone,Name=\"dfrm_arm_spline_06_r\"), ...(truncated)") {template="DISPATCH_RigVMDispatch_ArrayMake(in Values,out Array)"}
   > SplineIK_1 = RigVMFunctionReferenceNode(
    Primary Axis=(X=-1,Y=0,Z=0)
    Samples Per Segment=16
    Secondary Spline Direction=(X=0,Y=0,Z=1)
    Spline Mode=BSpline
    Stretch=true
    Up Axis=(X=0,Y=0,Z=1)
    Bones=@RerouteNode_2.Array
    Controls=RigVMVariableNode() {variable=arm_r_ctrls}.Value
   )
   &RigUnit_ParentConstraint(
    AdvancedSettings=(InterpolationType=Average,RotationOrderForFilter=XZY)
    Child=(Type=Control,Name="dfrm_leg_spline_01_l_ctrl")
    Parents=((Item=(Type=Bone,Name="thigh_twist_01_l"),Weight=0.75),(Item=(Type=Bone,Name="thigh_l"),Weight=0.25))
    Weight=1
    ExecutePin=@RigVMFunction_Sequence.C
   )
   > &RigUnit_ParentConstraint(
    AdvancedSettings=(InterpolationType=Average,RotationOrderForFilter=XZY)
    Child=(Type=Control,Name="dfrm_leg_spline_03_l_ctrl")
    Parents=((Item=(Type=Bone,Name="thigh_twist_01_l"),Weight=1),(Item=(Type=None,Name="upperarm_twist_02_l"),Weight=0))
    Weight=1
   )
   > &RigUnit_ParentConstraint(
    AdvancedSettings=(InterpolationType=Average,RotationOrderForFilter=XZY)
    Child=(Type=Control,Name="dfrm_leg_spline_05_l_ctrl")
    Parents=((Item=(Type=Bone,Name="calf_l"),Weight=1))
    Weight=1
   )
   > &RigUnit_ParentConstraint(
    AdvancedSettings=(InterpolationType=Average,RotationOrderForFilter=XZY)
    Child=(Type=Control,Name="dfrm_leg_spline_07_l_ctrl")
    Parents=((Item=(Type=Bone,Name="calf_twist_02_l"),Weight=1))
    Weight=1
   )
   > &RigUnit_ParentConstraint(
    AdvancedSettings=(InterpolationType=Average,RotationOrderForFilter=XZY)
    Child=(Type=Control,Name="dfrm_leg_spline_09_l_ctrl")
    Filter=(TranslationFilter=(bX=True,bY=True,bZ=True),RotationFilter=(bX=false,bY=false,bZ=false),ScaleFilter=(bX=True,bY=True,bZ=True))
    Parents=((Item=(Type=Bone,Name="foot_l"),Weight=1))
    Weight=1
   )
   > &RigUnit_ModifyTransforms(
    ItemToModify="((Item=(Type=Control,Name=\"dfrm_leg_spline_03_l_ctrl\"),Transform=(Rotation=(X=0,Y=0,Z=0,W=1),Translation=(X=0,Y=5,Z=0),Scale3D=(X=1,Y=1,Z=1))),(Item=(Type=Control,Name=\"dfrm_leg_spline_07_l_ctrl\"),Transform=(Rotation=(X=0,Y=0,Z=0,W=1),Translation=(X=0,Y=5, ...(truncated)"
    WeightMinimum=0
    Weight=RigVMDispatchNode(BoneName=root, Default=0, Name=deformRig_l_knee_alpha) {template="DISPATCH_RigDispatch_GetAnimAttribute(in Name,in BoneName,in Default,out Value,out Found)"}.Value
   )
   RerouteNode_1 = RigVMDispatchNode(Values="((Type=Bone,Name=\"dfrm_leg_spline_01_l\"),((),Type=Bone,Name=\"dfrm_leg_spline_02_l\"),((),Type=Bone,Name=\"dfrm_leg_spline_03_l\"),((),Type=Bone,Name=\"dfrm_leg_spline_04_l\"),((),Type=Bone,Name=\"dfrm_leg_spline_05_l\"),((),Type=Bone,Name=\"dfrm_leg_spline_06_l\"), ...(truncated)") {template="DISPATCH_RigVMDispatch_ArrayMake(in Values,out Array)"}
   > SplineIK_2 = RigVMFunctionReferenceNode(
    Primary Axis=(X=-1,Y=0,Z=0)
    Samples Per Segment=16
    Secondary Spline Direction=(X=0,Y=0,Z=1)
    Spline Mode=BSpline
    Stretch=true
    Up Axis=(X=0,Y=0,Z=1)
    Bones=@RerouteNode_1.Array
    Controls=RigVMVariableNode() {variable=leg_l_ctrls}.Value
   )
   &RigUnit_ParentConstraint(
    AdvancedSettings=(InterpolationType=Average,RotationOrderForFilter=XZY)
    Child=(Type=Control,Name="dfrm_leg_spline_01_r_ctrl")
    Parents=((Item=(Type=Bone,Name="thigh_twist_01_r"),Weight=0.75),(Item=(Type=Bone,Name="thigh_r"),Weight=0.25))
    Weight=1
    ExecutePin=@RigVMFunction_Sequence.D
   )
   > &RigUnit_ParentConstraint(
    AdvancedSettings=(InterpolationType=Average,RotationOrderForFilter=XZY)
    Child=(Type=Control,Name="dfrm_leg_spline_03_r_ctrl")
    Parents=((Item=(Type=Bone,Name="thigh_twist_01_r"),Weight=1),(Item=(Type=None,Name="upperarm_twist_02_l"),Weight=0))
    Weight=1
   )
   > &RigUnit_ParentConstraint(
    AdvancedSettings=(InterpolationType=Average,RotationOrderForFilter=XZY)
    Child=(Type=Control,Name="dfrm_leg_spline_05_r_ctrl")
    Parents=((Item=(Type=Bone,Name="calf_r"),Weight=1))
    Weight=1
   )
   > &RigUnit_ParentConstraint(
    AdvancedSettings=(InterpolationType=Average,RotationOrderForFilter=XZY)
    Child=(Type=Control,Name="dfrm_leg_spline_07_r_ctrl")
    Parents=((Item=(Type=Bone,Name="calf_twist_02_r"),Weight=1))
    Weight=1
   )
   > &RigUnit_ParentConstraint(
    AdvancedSettings=(InterpolationType=Average,RotationOrderForFilter=XZY)
    Child=(Type=Control,Name="dfrm_leg_spline_09_r_ctrl")
    Filter=(TranslationFilter=(bX=True,bY=True,bZ=True),RotationFilter=(bX=false,bY=false,bZ=false),ScaleFilter=(bX=True,bY=True,bZ=True))
    Parents=((Item=(Type=Bone,Name="foot_r"),Weight=1))
    Weight=1
   )
   > &RigUnit_ModifyTransforms(
    ItemToModify="((Item=(Type=Control,Name=\"dfrm_leg_spline_03_r_ctrl\"),Transform=(Rotation=(X=0,Y=0,Z=0,W=1),Translation=(X=0,Y=-5,Z=0),Scale3D=(X=1,Y=1,Z=1))),(Item=(Type=Control,Name=\"dfrm_leg_spline_07_r_ctrl\"),Transform=(Rotation=(X=0,Y=0,Z=0,W=1),Translation=(X=0,Y=- ...(truncated)"
    WeightMinimum=0
    Weight=RigVMDispatchNode(BoneName=root, Default=0, Name=deformRig_r_knee_alpha) {template="DISPATCH_RigDispatch_GetAnimAttribute(in Name,in BoneName,in Default,out Value,out Found)"}.Value
   )
   RerouteNode_4 = RigVMDispatchNode(Values="((Type=Bone,Name=\"dfrm_leg_spline_01_r\"),((),Type=Bone,Name=\"dfrm_leg_spline_02_r\"),((),Type=Bone,Name=\"dfrm_leg_spline_03_r\"),((),Type=Bone,Name=\"dfrm_leg_spline_04_r\"),((),Type=Bone,Name=\"dfrm_leg_spline_05_r\"),((),Type=Bone,Name=\"dfrm_leg_spline_06_r\"), ...(truncated)") {template="DISPATCH_RigVMDispatch_ArrayMake(in Values,out Array)"}
   > SplineIK_4 = RigVMFunctionReferenceNode(
    Primary Axis=(X=1,Y=0,Z=0)
    Samples Per Segment=16
    Secondary Spline Direction=(X=0,Y=0,Z=1)
    Spline Mode=BSpline
    Stretch=true
    Up Axis=(X=0,Y=0,Z=1)
    Bones=@RerouteNode_4.Array
    Controls=RigVMVariableNode() {variable=leg_r_ctrls}.Value
   )
   RigVMFunctionReturnNode(ExecuteContext=@RigVMFunction_Sequence.E)
   TwistBones = RigVMFunctionReferenceNode(
    Axis to Twist=(bX=True,bY=False,bZ=False)
    Bones to Twist=((Type=Bone,Name="dfrm_leg_spline_01_l"),(Type=Bone,Name="dfrm_leg_spline_02_l"),(Type=Bone,Name="dfrm_leg_spline_03_l"),(Type=Bone,Name="dfrm_leg_spline_04_l"))
    End=(Type=Bone,Name="calf_l")
    ExecuteContext=()
    Start=(Type=Bone,Name="thigh_twist_01_l")
   )
   > TwistBones_5 = RigVMFunctionReferenceNode(
    Axis to Twist=(bX=True,bY=False,bZ=False)
    Bones to Twist=((Type=Bone,Name="dfrm_leg_spline_05_l"),(Type=Bone,Name="dfrm_leg_spline_06_l"),(Type=Bone,Name="dfrm_leg_spline_07_l"),(Type=Bone,Name="dfrm_leg_spline_08_l"),(Type=Bone,Name="dfrm_leg_spline_09_l"))
    End=(Type=Bone,Name="foot_l")
    Start=(Type=Bone,Name="calf_l")
   )
   TwistBones_1 = RigVMFunctionReferenceNode(
    Axis to Twist=(bX=True,bY=False,bZ=False)
    Bones to Twist=((Type=Bone,Name="dfrm_arm_spline_01_r"),(Type=Bone,Name="dfrm_arm_spline_02_r"),(Type=Bone,Name="dfrm_arm_spline_03_r"),(Type=Bone,Name="dfrm_arm_spline_04_r"),(Type=None,Name="dfrm_arm_spline_03_r"))
    End=(Type=Bone,Name="lowerarm_r")
    ExecuteContext=()
    Start=(Type=Bone,Name="upperarm_twist_01_r")
   )
   > TwistBones_2 = RigVMFunctionReferenceNode(
    Axis to Twist=(bX=True,bY=False,bZ=False)
    Bones to Twist=((Type=Bone,Name="dfrm_arm_spline_05_r"),(Type=Bone,Name="dfrm_arm_spline_06_r"),(Type=Bone,Name="dfrm_arm_spline_07_r"),(Type=Bone,Name="dfrm_arm_spline_08_r"),(Type=Bone,Name="dfrm_arm_spline_09_r"))
    End=(Type=Bone,Name="hand_r")
    Start=(Type=Bone,Name="lowerarm_r")
   )
   TwistBones_3 = RigVMFunctionReferenceNode(
    Axis to Twist=(bX=True,bY=False,bZ=False)
    Bones to Twist=((Type=Bone,Name="dfrm_arm_spline_01_l"),(Type=Bone,Name="dfrm_arm_spline_02_l"),(Type=Bone,Name="dfrm_arm_spline_03_l"),(Type=Bone,Name="dfrm_arm_spline_04_l"),(Type=None,Name="dfrm_arm_spline_03_l"))
    End=(Type=Bone,Name="lowerarm_l")
    ExecuteContext=()
    Start=(Type=Bone,Name="upperarm_twist_01_l")
   )
   > TwistBones_4 = RigVMFunctionReferenceNode(
    Axis to Twist=(bX=True,bY=False,bZ=False)
    Bones to Twist=((Type=Bone,Name="dfrm_arm_spline_05_l"),(Type=Bone,Name="dfrm_arm_spline_06_l"),(Type=Bone,Name="dfrm_arm_spline_07_l"),(Type=Bone,Name="dfrm_arm_spline_08_l"),(Type=Bone,Name="dfrm_arm_spline_09_l"))
    End=(Type=Bone,Name="hand_l")
    Start=(Type=Bone,Name="lowerarm_l")
   )
   TwistBones_7 = RigVMFunctionReferenceNode(
    Axis to Twist=(bX=True,bY=False,bZ=False)
    Bones to Twist=((Type=Bone,Name="dfrm_leg_spline_01_r"),(Type=Bone,Name="dfrm_leg_spline_02_r"),(Type=Bone,Name="dfrm_leg_spline_03_r"),(Type=Bone,Name="dfrm_leg_spline_04_r"))
    End=(Type=Bone,Name="calf_r")
    ExecuteContext=()
    Start=(Type=Bone,Name="thigh_twist_01_r")
   )
   > TwistBones_6 = RigVMFunctionReferenceNode(
    Axis to Twist=(bX=True,bY=False,bZ=False)
    Bones to Twist=((Type=Bone,Name="dfrm_leg_spline_05_r"),(Type=Bone,Name="dfrm_leg_spline_06_r"),(Type=Bone,Name="dfrm_leg_spline_07_r"),(Type=Bone,Name="dfrm_leg_spline_08_r"),(Type=Bone,Name="dfrm_leg_spline_09_r"))
    End=(Type=Bone,Name="foot_r")
    Start=(Type=Bone,Name="calf_r")
   )
   &RigUnit_ModifyTransforms(
    ExecutePin=()
    ItemToModify="((Item=(Type=Control,Name=\"dfrm_arm_spline_01_l_ctrl\"),Transform=(Rotation=(X=0.7071,Y=0,Z=0,W=0.7071),Translation=(X=0,Y=0,Z=0),Scale3D=(X=1,Y=1,Z=1))),(Item=(Type=Control,Name=\"dfrm_arm_spline_09_l_ctrl\"),Transform=(Rotation=(X=-0.3827,Y=0,Z=0,W=0.9239), ...(truncated)"
    Weight=1
    WeightMinimum=0
   )
   &RigUnit_ModifyTransforms(
    ExecutePin=()
    ItemToModify="((Item=(Type=Control,Name=\"dfrm_arm_spline_01_l_ctrl\"),Transform=(Rotation=(X=0.7071,Y=0,Z=0,W=0.7071),Translation=(X=0,Y=0,Z=0),Scale3D=(X=1,Y=1,Z=1))),(Item=(Type=Bone,Name=\"None\"),Transform=(Rotation=(X=0,Y=0,Z=0,W=1),Translation=(X=0,Y=0,Z=0),Scale3D=( ...(truncated)"
    Weight=1
    WeightMinimum=0
   )
   SplineIK_5 = RigVMFunctionReferenceNode(
    Bones=()
    Controls=()
    ExecuteContext=()
    Primary Axis=(X=1,Y=0,Z=0)
    Samples Per Segment=16
    Secondary Spline Direction=(X=0,Y=-1,Z=0)
    Spline Mode=BSpline
    Stretch=true
    Up Axis=(X=0,Y=0,Z=1)
   )
   SplineIK_6 = RigVMFunctionReferenceNode(
    Bones=()
    Controls=()
    ExecuteContext=()
    Primary Axis=(X=-1,Y=0,Z=0)
    Samples Per Segment=16
    Secondary Spline Direction=(X=0,Y=0,Z=1)
    Spline Mode=BSpline
    Stretch=true
    Up Axis=(X=0,Y=0,Z=1)
   )
  ]}
 New Function_ContainedGraph {nodes [
   RigVMFunctionEntryNode()
   > &RigUnit_ModifyTransforms(
    ItemToModify="((Item=(Type=Bone,Name=\"crv_calf_l_top\"),Transform=(Rotation=(X=0,Y=0,Z=-0.6088,W=0.7934),Translation=(X=5,Y=0,Z=0),Scale3D=(X=1,Y=4,Z=1))),(Item=(Type=Bone,Name=\"crv_calf_l_bottom\"),Transform=(Rotation=(X=0,Y=0,Z=-0.1736,W=0.9848),Translation=(X=-5,Y=10,Z ...(truncated)"
    WeightMinimum=0
    Weight=RigVMDispatchNode(BoneName=root, Default=0, Name=deformRig_l_knee_alpha) {template="DISPATCH_RigDispatch_GetAnimAttribute(in Name,in BoneName,in Default,out Value,out Found)"}.Value
   )
   > &RigUnit_ModifyTransforms(
    ItemToModify="((Item=(Type=Bone,Name=\"crv_calf_r_top\"),Transform=(Rotation=(X=0,Y=0,Z=-0.6088,W=0.7934),Translation=(X=-5,Y=0,Z=0),Scale3D=(X=1,Y=4,Z=1))),(Item=(Type=Bone,Name=\"crv_calf_r_bottom\"),Transform=(Rotation=(X=0,Y=0,Z=-0.1736,W=0.9848),Translation=(X=5,Y=-10, ...(truncated)"
    WeightMinimum=0
    Weight=RigVMDispatchNode(BoneName=root, Default=0, Name=deformRig_r_knee_alpha) {template="DISPATCH_RigDispatch_GetAnimAttribute(in Name,in BoneName,in Default,out Value,out Found)"}.Value
   )
   > &RigUnit_ModifyTransforms(
    ItemToModify="((Item=(Type=Bone,Name=\"crv_lowerarm_l_top\"),Transform=(Rotation=(X=0,Y=0,Z=-0.5,W=0.866),Translation=(X=0,Y=0,Z=0),Scale3D=(X=1,Y=1,Z=1))),(Item=(Type=Bone,Name=\"crv_lowerarm_l_bottom\"),Transform=(Rotation=(X=0,Y=0,Z=0,W=1),Translation=(X=0,Y=-20,Z=0),Sca ...(truncated)"
    WeightMinimum=0
    Weight=RigVMDispatchNode(BoneName=root, Default=0, Name=deformRig_l_elbow_alpha) {template="DISPATCH_RigDispatch_GetAnimAttribute(in Name,in BoneName,in Default,out Value,out Found)"}.Value
   )
   > &RigUnit_ModifyTransforms(
    ItemToModify="((Item=(Type=Bone,Name=\"crv_lowerarm_r_top\"),Transform=(Rotation=(X=0,Y=0,Z=-0.5,W=0.866),Translation=(X=0,Y=0,Z=0),Scale3D=(X=1,Y=1,Z=1))),(Item=(Type=Bone,Name=\"crv_lowerarm_r_bottom\"),Transform=(Rotation=(X=0,Y=0,Z=0,W=1),Translation=(X=0,Y=20,Z=0),Scal ...(truncated)"
    WeightMinimum=0
    Weight=RigVMDispatchNode(BoneName=root, Default=0, Name=deformRig_r_elbow_alpha) {template="DISPATCH_RigDispatch_GetAnimAttribute(in Name,in BoneName,in Default,out Value,out Found)"}.Value
   )
   > &RigUnit_ModifyTransforms(
    ItemToModify="((Item=(Type=Bone,Name=\"crv_lowerarm_l_top\"),Transform=(Rotation=(X=0,Y=0,Z=0.1736,W=0.9848),Translation=(X=0,Y=0,Z=0),Scale3D=(X=1,Y=1,Z=1))),(Item=(Type=Bone,Name=\"crv_lowerarm_l_bottom\"),Transform=(Rotation=(X=0,Y=0,Z=0,W=1),Translation=(X=0,Y=5,Z=0),Sc ...(truncated)"
    WeightMinimum=0
    Weight=RigVMVariableNode() {variable=Is Sky Dive Float}.Value
   )
   > RigVMFunctionReturnNode()
  ]}
 New Function_ContainedGraph {nodes [
   RigVMFunctionEntryNode()
   > RigVMFunction_Sequence = RigVMAggregateNode()
   n1 = &RigUnit_SphericalPoseReader(
    ActiveRegionScaleFactors=(PositiveWidth=1,NegativeWidth=1,PositiveHeight=1,NegativeHeight=1)
    ActiveRegionSize=0.2
    Debug=(bDrawDebug=false,bDraw2D=False,bDrawLocalAxes=False,DebugScale=25,DebugSegments=20,DebugThickness=0.25)
    DriverAxis=(X=-1,Y=0,Z=0)
    DriverItem=(Type=Bone,Name="upperarm_r")
    FalloffRegionScaleFactors=(PositiveWidth=1,NegativeWidth=1,PositiveHeight=1,NegativeHeight=1)
    FalloffSize=0.3
    FlipHeightScaling=False
    FlipWidthScaling=False
    OptionalParentItem=(Type=Bone,Name="None")
    RotationOffset=(X=-190,Y=-10,Z=100)
    ExecutePin=@RigVMFunction_Sequence.A
   )
   n2 = &RigVMFunction_MathDoubleMul(
    B=RigVMDispatchNode(False=0, True=1, Condition=RigVMVariableNode() {variable=IsPuzzleshedEmotePlay}.Value) {template="DISPATCH_RigVMDispatch_If(in Condition,in True,in False,out Result)"}.Result
    A=@n1.OutputParam
   )
   > &RigUnit_ModifyTransforms(
    ItemToModify=((Item=(Type=Bone,Name="clavicle_r"),Transform=(Rotation=(X=-0.0483,Y=0.0892,Z=-0.0388,W=0.9941),Translation=(X=-4,Y=0,Z=0),Scale3D=(X=1,Y=1,Z=1))))
    WeightMinimum=0
    Weight=@n2.Result
   )
   > &RigUnit_ModifyTransforms(
    ItemToModify=((Item=(Type=Bone,Name="upperarm_r"),Transform=(Rotation=(X=0,Y=0,Z=0,W=1),Translation=(X=0,Y=3,Z=0),Scale3D=(X=1,Y=1,Z=1))))
    WeightMinimum=0
    Weight=@n2.Result
   )
   n3 = &RigUnit_SphericalPoseReader(
    ActiveRegionScaleFactors=(PositiveWidth=1,NegativeWidth=1,PositiveHeight=1,NegativeHeight=1)
    ActiveRegionSize=0.2
    Debug=(bDrawDebug=false,bDraw2D=False,bDrawLocalAxes=False,DebugScale=25,DebugSegments=20,DebugThickness=0.25)
    DriverAxis=(X=-1,Y=0,Z=0)
    DriverItem=(Type=Bone,Name="upperarm_l")
    FalloffRegionScaleFactors=(PositiveWidth=1,NegativeWidth=1,PositiveHeight=1,NegativeHeight=1)
    FalloffSize=0.3
    FlipHeightScaling=False
    FlipWidthScaling=False
    OptionalParentItem=(Type=Bone,Name="None")
    RotationOffset=(X=-190,Y=-10,Z=100)
    ExecutePin=@RigVMFunction_Sequence.B
   )
   Multiply = &RigVMFunction_MathDoubleMul(
    B=RigVMDispatchNode(False=0, True=1, Condition=RigVMVariableNode() {variable=IsPuzzleshedEmotePlay}.Value) {template="DISPATCH_RigVMDispatch_If(in Condition,in True,in False,out Result)"}.Result
    A=@n3.OutputParam
   )
   > &RigUnit_ModifyTransforms(
    ItemToModify=((Item=(Type=Bone,Name="clavicle_l"),Transform=(Rotation=(X=-0.0516,Y=0.0998,Z=-0.0251,W=0.9933),Translation=(X=2,Y=0,Z=0),Scale3D=(X=1,Y=1,Z=1))))
    WeightMinimum=0
    Weight=@Multiply.Result
   )
   RigVMFunctionReturnNode(ExecuteContext=@RigVMFunction_Sequence.C)
  ]}
 RigVMFunctionLibrary {nodes [
   DeltaRotation = RigVMCollapseNode(ExecuteContext=(), Rotation Order=ZYX, twistAxis=(X=1,Y=0,Z=0))
   DeltaRotationZ = RigVMCollapseNode(ExecuteContext=())
   Emotes_pose_fix = RigVMCollapseNode(ExecuteContext=())
   accessories = RigVMCollapseNode(ExecuteContext=())
   bendSetupV2 = RigVMCollapseNode(ExecuteContext=())
   ikSplinesArm_Setup = RigVMCollapseNode(ExecuteContext=())
   is sliding setup = RigVMCollapseNode(ExecuteContext=())
   skirt setup = RigVMCollapseNode(ExecuteContext=())
   twistFix = RigVMCollapseNode(ExecuteContext=())
   variables = RigVMCollapseNode(ExecuteContext=())
  ]}
 RigVMFunction_Sequence_1_ContainedGraph {nodes [
   n1 = RigVMFunctionEntryNode()
   RigVMFunction_Sequence = &RigVMFunction_Sequence(ExecuteContext=@n1.ExecuteContext)
   n2 = &RigVMFunction_Sequence(ExecuteContext=@RigVMFunction_Sequence.B)
   RigVMFunctionReturnNode(A=@RigVMFunction_Sequence.A, B=@n2.A, C=@n2.B)
  ]}
 RigVMFunction_Sequence_1_ContainedGraph {nodes [
   n1 = RigVMFunctionEntryNode()
   RigVMFunction_Sequence = &RigVMFunction_Sequence(ExecuteContext=@n1.ExecuteContext)
   n2 = &RigVMFunction_Sequence(ExecuteContext=@RigVMFunction_Sequence.B)
   n3 = &RigVMFunction_Sequence(ExecuteContext=@n2.B)
   RigVMFunctionReturnNode(A=@RigVMFunction_Sequence.A, B=@n2.A, C=@n3.A, D=@n3.B)
  ]}
 RigVMFunction_Sequence_1_ContainedGraph {nodes [
   n1 = RigVMFunctionEntryNode()
   RigVMFunction_Sequence = &RigVMFunction_Sequence(ExecuteContext=@n1.ExecuteContext)
   n2 = &RigVMFunction_Sequence(ExecuteContext=@RigVMFunction_Sequence.B)
   n3 = &RigVMFunction_Sequence(ExecuteContext=@n2.B)
   n4 = &RigVMFunction_Sequence(ExecuteContext=@n3.B)
   RigVMFunctionReturnNode(A=@RigVMFunction_Sequence.A, B=@n2.A, C=@n3.A, D=@n4.A, E=@n4.B)
  ]}
 RigVMFunction_Sequence_1_ContainedGraph {nodes [
   n1 = RigVMFunctionEntryNode()
   RigVMFunction_Sequence = &RigVMFunction_Sequence(ExecuteContext=@n1.ExecuteContext)
   n2 = &RigVMFunction_Sequence(ExecuteContext=@RigVMFunction_Sequence.B)
   RigVMFunctionReturnNode(A=@RigVMFunction_Sequence.A, B=@n2.A, C=@n2.B)
  ]}
 RigVMFunction_Sequence_1_ContainedGraph {nodes [
   n1 = RigVMFunctionEntryNode()
   RigVMFunction_Sequence = &RigVMFunction_Sequence(ExecuteContext=@n1.ExecuteContext)
   n2 = &RigVMFunction_Sequence(ExecuteContext=@RigVMFunction_Sequence.B)
   n3 = &RigVMFunction_Sequence(ExecuteContext=@n2.B)
   n4 = &RigVMFunction_Sequence(ExecuteContext=@n3.B)
   RigVMFunctionReturnNode(A=@RigVMFunction_Sequence.A, B=@n2.A, C=@n3.A, D=@n4.A, E=@n4.B)
  ]}
 RigVMFunction_Sequence_1_ContainedGraph {nodes [
   n1 = RigVMFunctionEntryNode()
   RigVMFunction_Sequence = &RigVMFunction_Sequence(ExecuteContext=@n1.ExecuteContext)
   n2 = &RigVMFunction_Sequence(ExecuteContext=@RigVMFunction_Sequence.B)
   RigVMFunctionReturnNode(A=@RigVMFunction_Sequence.A, B=@n2.A, C=@n2.B)
  ]}
 RigVMFunction_Sequence_2_ContainedGraph {nodes [
   n1 = RigVMFunctionEntryNode()
   RigVMFunction_Sequence = &RigVMFunction_Sequence(ExecuteContext=@n1.ExecuteContext)
   n2 = &RigVMFunction_Sequence(ExecuteContext=@RigVMFunction_Sequence.B)
   RigVMFunctionReturnNode(A=@RigVMFunction_Sequence.A, B=@n2.A, C=@n2.B)
  ]}
 RigVMModel {nodes [
   RigUnit_BeginExecution = &RigUnit_BeginExecution()
   RigVMFunction_Sequence = RigVMAggregateNode(ExecuteContext=@RigUnit_BeginExecution.ExecutePin)
   FN_DeformRig_FNC = RigVMFunctionReferenceNode(
    DeformRig_ Curve="(deformRig_l_upperarm_up_alpha_72_#1=False,deformRig_r_upperarm_up_alpha_73_#2=False,deformRig_l_upperarm_fr_alpha_74_#3=False,deformRig_r_upperarm_fr_alpha_75_CB51B8 ...(truncated)"
    DeformRig_Bool="(upperarm_1_#4=True,elbow_3_#5=True,wrist_7_#6=True,groin_8_#7=True,glute_17_#8=True,knee_9_40C408324CA261 ...(truncated)"
    DeformRig_Debug=(upperarm_debug_18_#4=false,wrist_debug_20_#6=false,lat_debug_26_#9=false,groinGlute_debug_27_#7=False)
    DeformRig_Value="(upperarm_up_18_#10=(Rotation=(X=0,Y=0,Z=0,W=1),Translation=(X=0,Y=0,Z=4),Scale3D=(X=1,Y=1,Z=1)),upperarm_fwd_fr_19_#11=(Rotation=(X=0,Y=0,Z=0,W=1),Translation=(X=0,Y=10,Z=0),Scale3D=(X=0.5,Y=1,Z=1) ...(truncated)"
    ExecuteContext=@RigVMFunction_Sequence.B
   )
   > variables = RigVMFunctionReferenceNode()
   > Emotes_pose_fix = RigVMFunctionReferenceNode()
   bendSetupV2 = RigVMFunctionReferenceNode(ExecuteContext=@RigVMFunction_Sequence.C)
   controllers = RigVMCollapseNode(ExecuteContext=(), debug=true)
   ikSplinesArm_Setup = RigVMFunctionReferenceNode(ExecuteContext=())
  ]}
 RigVMModel Rig_2 {nodes [
   n1 = &RigUnit_PrepareForExecution()
   RigVMFunction_Sequence = RigVMAggregateNode(ExecuteContext=@n1.ExecutePin)
   RerouteNode_8 = RigVMDispatchNode(Values=((Type=Control,Name="dfrm_arm_spline_01_r_ctrl"),((),Type=Control,Name="dfrm_arm_spline_03_r_ctrl"),((),Type=Control,Name="dfrm_arm_spline_05_r_ctrl"),((),Type=Control,Name="dfrm_arm_spline_07_r_ctrl"),((),Type=Control,Name="dfrm_arm_spline_09_r_ctrl"))) {template="DISPATCH_RigVMDispatch_ArrayMake(in Values,out Array)"}
   RigVMVariableNode(Value=@RerouteNode_8.Array, ExecuteContext=@RigVMFunction_Sequence.B) {variable=arm_r_ctrls}
   > RigVMVariableNode(Value=/Script/ControlRig.RigUnit_HierarchyGetPoseItemArray(ElementType=All, Initial=False, ItemsToGet=@RerouteNode_8.Array) {is=struct RigVMUnitNode}.Pose) {variable=arm_r_bind_pose}
   RerouteNode_6 = RigVMDispatchNode(Values=((Type=Control,Name="dfrm_arm_spline_01_l_ctrl"),((),Type=Control,Name="dfrm_arm_spline_03_l_ctrl"),((),Type=Control,Name="dfrm_arm_spline_05_l_ctrl"),((),Type=Control,Name="dfrm_arm_spline_07_l_ctrl"),((),Type=Control,Name="dfrm_arm_spline_09_l_ctrl"))) {template="DISPATCH_RigVMDispatch_ArrayMake(in Values,out Array)"}
   RigVMVariableNode(Value=@RerouteNode_6.Array, ExecuteContext=@RigVMFunction_Sequence.A) {variable=arm_l_ctrls}
   > RigVMVariableNode(Value=&RigUnit_HierarchyGetPoseItemArray(ElementType=All, Initial=False, ItemsToGet=@RerouteNode_6.Array).Pose) {variable=arm_l_bind_pose}
   RerouteNode_1 = RigVMDispatchNode(Values=((Type=Control,Name="dfrm_leg_spline_01_l_ctrl"),((),Type=Control,Name="dfrm_leg_spline_03_l_ctrl"),((),Type=Control,Name="dfrm_leg_spline_05_l_ctrl"),((),Type=Control,Name="dfrm_leg_spline_07_l_ctrl"),((),Type=Control,Name="dfrm_leg_spline_09_l_ctrl"))) {template="DISPATCH_RigVMDispatch_ArrayMake(in Values,out Array)"}
   RigVMVariableNode(Value=@RerouteNode_1.Array, ExecuteContext=@RigVMFunction_Sequence.C) {variable=leg_l_ctrls}
   > RigVMVariableNode(Value=&RigUnit_HierarchyGetPoseItemArray(ElementType=All, Initial=False, ItemsToGet=@RerouteNode_1.Array).Pose) {variable=leg_l_bind_pose}
   RerouteNode_2 = RigVMDispatchNode(Values=((Type=Control,Name="dfrm_leg_spline_01_r_ctrl"),((),Type=Control,Name="dfrm_leg_spline_03_r_ctrl"),((),Type=Control,Name="dfrm_leg_spline_05_r_ctrl"),((),Type=Control,Name="dfrm_leg_spline_07_r_ctrl"),((),Type=Control,Name="dfrm_leg_spline_09_r_ctrl"))) {template="DISPATCH_RigVMDispatch_ArrayMake(in Values,out Array)"}
   RigVMVariableNode(Value=@RerouteNode_2.Array, ExecuteContext=@RigVMFunction_Sequence.D) {variable=leg_r_ctrls}
   > RigVMVariableNode(Value=&RigUnit_HierarchyGetPoseItemArray(ElementType=All, Initial=False, ItemsToGet=@RerouteNode_2.Array).Pose) {variable=leg_r_bind_pose}
   RigUnit_GetTransform = &RigUnit_GetTransform(Item=(Type=Bone,Name="hand_l"), bInitial=true)
   RigUnit_SetTransform = &RigUnit_SetTransform(ExecutePin=(), Item=(Type=Control,Name="dfrm_arm_spline_09_l_ctrl"), Weight=1, bInitial=False, Value=@RigUnit_GetTransform.Transform)
   > &RigUnit_SetTransform(
    Item=(Type=Control,Name="dfrm_arm_spline_01_l_ctrl")
    Weight=1
    bInitial=False
    Value=&RigUnit_GetTransform(Item=(Type=Bone,Name="upperarm_l"), bInitial=true).Transform
   )
   &RigUnit_SetTransform(
    ExecutePin=()
    Item=(Type=Control,Name="dfrm_arm_spline_09_r_ctrl")
    Weight=1
    bInitial=False
    Value=&RigUnit_GetTransform(Item=(Type=Bone,Name="hand_r"), bInitial=true).Transform
   )
   > &RigUnit_SetTransform(
    Item=(Type=Control,Name="dfrm_arm_spline_01_r_ctrl")
    Weight=1
    bInitial=False
    Value=&RigUnit_GetTransform(Item=(Type=Bone,Name="upperarm_r"), bInitial=true).Transform
   )
  ]}
]
hierarchy [
 root {loc=<0 0 0>, rot=<0 0 0 1>}
 pelvis {parent=root, loc=<0 -0.5301 98.0571>, rot=<0.0127 0.707 -0.0127 -0.707>}
 spine_01 {parent=pelvis, loc=<6.1208 -0.9328 0>, rot=<0 0 0.034 -0.9994>}
 spine_02 {parent=spine_01, loc=<6.5997 -0.2095 0>, rot=<0 0 0.0238 -0.9997>}
 spine_03 {parent=spine_02, loc=<6.4914 -0.1237 0>, rot=<0 0 0.0149 -0.9999>}
 spine_04 {parent=spine_03, loc=<6.9598 0.0794 0>, rot=<0 0 -0.0132 -0.9999>}
 spine_05 {parent=spine_04, loc=<6.9987 0 0>, rot=<0 0 -0.0015 -1>}
 neck_01 {parent=spine_05, loc=<11.3138 0.1529 0>, rot=<0 0 -0.2003 -0.9797>}
 neck_02 {parent=neck_01, loc=<4.5627 0 0>, rot=<0 0 0.0504 -0.9987>}
 head {parent=neck_02, loc=<5.3914 0 0>, rot=<0 0 0.115 -0.9934>}
 clavicle_l {parent=spine_05, loc=<5.4831 -0.1053 -3.2946>, rot=<0.0465 -0.7289 -0.009 -0.6829>}
 upperarm_l {parent=clavicle_l, loc=<9.8658 -1.2136 -0.4105>, rot=<0.0164 -0.3245 -0.0269 -0.9454>}
 lowerarm_l {parent=upperarm_l, loc=<26.465 -0.8345 0.259>, rot=<0.0036 0.0161 -0.2298 -0.9731>}
 hand_l {parent=lowerarm_l, loc=<22.7355 -0.3203 -0.6105>, rot=<-0.6784 -0.0223 0.0735 -0.7307>}
 thumb_01_l {parent=hand_l, loc=<1.3118 -0.3191 -2.459>, rot=<0.6546 -0.4042 0.0664 -0.6354>}
 thumb_02_l {parent=thumb_01_l, loc=<5.9336 -0.0448 0.0401>, rot=<-0.0037 0.0213 0.298 -0.9543>}
 thumb_03_l {parent=thumb_02_l, loc=<2.2223 0.3682 -0.0907>, rot=<0.0118 -0.013 0.0066 -0.9998>}
 index_metacarpal_l {parent=hand_l, loc=<1.6535 0.0183 -2.96>, rot=<0.0158 -0.0124 -0.0173 -0.9996>}
 index_01_l {parent=index_metacarpal_l, loc=<8.3008 0.0357 -0.0524>, rot=<-0.0044 0.0485 0.0924 -0.9945>}
 index_02_l {parent=index_01_l, loc=<3.3587 -0.3004 0.0452>, rot=<-0.002 0.0027 0.3225 -0.9466>}
 index_03_l {parent=index_02_l, loc=<2.5578 -0.123 -0.0252>, rot=<-0.002 0.0067 0.2558 -0.9667>}
 middle_metacarpal_l {parent=hand_l, loc=<2.1856 -0.1725 -0.5984>, rot=<0.006 0.0168 -0.0329 -0.9993>}
 middle_01_l {parent=middle_metacarpal_l, loc=<7.3275 -0.2447 0.0129>, rot=<-0.0208 0.0042 0.0858 -0.9961>}
 middle_02_l {parent=middle_01_l, loc=<4.0123 0.1401 0.1175>, rot=<-0.0218 0.0258 0.3512 -0.9357>}
 middle_03_l {parent=middle_02_l, loc=<2.6077 0.0534 0.1373>, rot=<0.0082 0.0188 0.1708 -0.9851>}
 ring_metacarpal_l {parent=hand_l, loc=<2.2074 -0.3625 1.3638>, rot=<-0.0284 0.0722 -0.0458 -0.9959>}
 ring_01_l {parent=ring_metacarpal_l, loc=<7.0295 -0.6197 -0.1446>, rot=<-0.0489 -0.0252 0.1892 -0.9804>}
 ring_02_l {parent=ring_01_l, loc=<3.6174 0.4647 -0.1027>, rot=<0 0 0.2806 -0.9598>}
 ring_03_l {parent=ring_02_l, loc=<2.4152 0.114 0.0199>, rot=<0 0 0.1189 -0.9929>}
 pinky_metacarpal_l {parent=hand_l, loc=<1.7616 -0.9753 3.3814>, rot=<-0.0397 0.0814 -0.069 -0.9935>}
 pinky_01_l {parent=pinky_metacarpal_l, loc=<7.1709 -0.4831 -0.1094>, rot=<-0.0654 -0.0035 0.2242 -0.9723>}
 pinky_02_l {parent=pinky_01_l, loc=<2.8274 0.0893 -0.1043>, rot=<0.0037 0.0011 0.2854 -0.9584>}
 pinky_03_l {parent=pinky_02_l, loc=<2.2046 0.0307 -0.0078>, rot=<-0.0002 0.0098 0.1445 -0.9895>}
 deform_wrist_dn_base_l {parent=hand_l, loc=<0 -1.5326 0>, rot=<0 0 0 -1>}
 deform_wrist_dn_l {parent=deform_wrist_dn_base_l, loc=<0 0 0>, rot=<0 0 0 1>}
 deform_wrist_up_base_l {parent=hand_l, loc=<0 2.3009 0>, rot=<0 0 0 -1>}
 deform_wrist_up_l {parent=deform_wrist_up_base_l, loc=<0 0 0>, rot=<0 0 0 1>}
 lowerarm_twist_02_l {parent=lowerarm_l, loc=<10.6707 0 0>, rot=<0 0 0 1>}
 lowerarm_twist_01_l {parent=lowerarm_l, loc=<18.316 0 0>, rot=<0 0 0 1>}
 deform_elbow_in_base_l {parent=lowerarm_l, loc=<0.1834 3.6062 -0.2572>, rot=<0 0 0 -1>}
 deform_elbow_in_l {parent=deform_elbow_in_base_l, loc=<0 0 0>, rot=<0 0 0 1>}
 deform_elbow_out_base_l {parent=lowerarm_l, loc=<0 -4 0>, rot=<0 0 0 1>}
 deform_elbow_out_l {parent=deform_elbow_out_base_l, loc=<0 0 0>, rot=<0 0 0 1>}
 crv_lowerarm_l_bottom {parent=lowerarm_l, loc=<0 0 0>, rot=<0 0 0 1>}
 crv_lowerarm_l_top {parent=lowerarm_l, loc=<0 0 0>, rot=<0 0 0 1>}
 upperarm_twist_01_l {parent=upperarm_l, loc=<7.2135 -0.1134 0>, rot=<0 0 0 1>}
 deform_upperarm_bk_base_l {parent=upperarm_twist_01_l, loc=<-6.4053 -4.3746 0.314>, rot=<0 0 0 -1>}
 deform_upperarm_bk_l {parent=deform_upperarm_bk_base_l, loc=<0 0 0>, rot=<0 0 0 1>}
 deform_upperarm_fr_base_l {parent=upperarm_twist_01_l, loc=<-5.6312 5.092 0.3354>, rot=<0 0 0 -1>}
 deform_upperarm_fr_l {parent=deform_upperarm_fr_base_l, loc=<0 0 0>, rot=<0 0 0 1>}
 deform_upperarm_up_base_l {parent=upperarm_twist_01_l, loc=<-8.891 0.4825 3.9517>, rot=<0 0 0 -1>}
 deform_upperarm_up_l {parent=deform_upperarm_up_base_l, loc=<0 0 0>, rot=<0 0 0 1>}
 upperarm_twist_02_l {parent=upperarm_l, loc=<19.7998 -0.1134 0>, rot=<0 0 0 1>}
 dfrm_arm_spline_mstr_l {parent=upperarm_l, loc=<-0.121 -0.0811 0.1944>, rot=<0 0 0 -1>}
 dfrm_arm_spline_01_l {parent=dfrm_arm_spline_mstr_l, loc=<0 0 0>, rot=<0 0 0 1>}
 dfrm_arm_spline_02_l {parent=dfrm_arm_spline_01_l, loc=<6.9266 0.0011 -0.2633>, rot=<0 0 0 -1>}
 dfrm_arm_spline_03_l {parent=dfrm_arm_spline_02_l, loc=<5.9472 -0.3982 0.2823>, rot=<0 0 0 -1>}
 dfrm_arm_spline_04_l {parent=dfrm_arm_spline_03_l, loc=<5.819 -0.2729 0.148>, rot=<0 0 0 -1>}
 dfrm_arm_spline_05_l {parent=dfrm_arm_spline_04_l, loc=<6.2912 0.0768 -0.1231>, rot=<-0.0005 0.0165 0.0186 -0.9997>}
 dfrm_arm_spline_06_l {parent=dfrm_arm_spline_05_l, loc=<5.3135 2.0991 0.0438>, rot=<0 0 -0.1903 -0.9817>}
 dfrm_arm_spline_07_l {parent=dfrm_arm_spline_06_l, loc=<6.0386 0.342 -0.459>, rot=<0 0 -0.059 -0.9983>}
 dfrm_arm_spline_08_l {parent=dfrm_arm_spline_07_l, loc=<5.8874 -0.1878 -0.3618>, rot=<0 0 0 1>}
 dfrm_arm_spline_09_l {parent=dfrm_arm_spline_08_l, loc=<6.5036 0.0895 0.1341>, rot=<0.0781 -0.0407 -0.0246 -0.9958>}
 upperarm_l_null {parent=clavicle_l, type=Null, loc=<9.8658 -1.2136 -0.4105>, rot=<0.0164 -0.3245 -0.0269 -0.9454>}
 clavicle_r {parent=spine_05, loc=<5.4831 -0.1053 3.2946>, rot=<0.7289 0.0465 -0.6829 0.009>}
 upperarm_r {parent=clavicle_r, loc=<-9.8658 1.2136 0.4105>, rot=<0.0164 -0.3245 -0.0269 -0.9454>}
 lowerarm_r {parent=upperarm_r, loc=<-26.465 0.8345 -0.259>, rot=<0.0036 0.0161 -0.2298 -0.9731>}
 hand_r {parent=lowerarm_r, loc=<-22.7355 0.3203 0.6106>, rot=<-0.6784 -0.0223 0.0735 -0.7307>}
 thumb_01_r {parent=hand_r, loc=<-1.3118 0.3191 2.459>, rot=<0.6546 -0.4042 0.0664 -0.6354>}
 thumb_02_r {parent=thumb_01_r, loc=<-5.9336 0.0448 -0.0401>, rot=<-0.0037 0.0213 0.298 -0.9543>}
 thumb_03_r {parent=thumb_02_r, loc=<-2.2223 -0.3682 0.0907>, rot=<0.0118 -0.013 0.0066 -0.9998>}
 index_metacarpal_r {parent=hand_r, loc=<-1.6535 -0.0183 2.96>, rot=<0.0158 -0.0124 -0.0173 -0.9996>}
 index_01_r {parent=index_metacarpal_r, loc=<-8.3008 -0.0357 0.0524>, rot=<-0.0044 0.0485 0.0924 -0.9945>}
 index_02_r {parent=index_01_r, loc=<-3.3587 0.3004 -0.0452>, rot=<-0.002 0.0027 0.3225 -0.9466>}
 index_03_r {parent=index_02_r, loc=<-2.5578 0.123 0.0252>, rot=<-0.002 0.0067 0.2558 -0.9667>}
 middle_metacarpal_r {parent=hand_r, loc=<-2.1856 0.1725 0.5984>, rot=<0.006 0.0168 -0.0329 -0.9993>}
 middle_01_r {parent=middle_metacarpal_r, loc=<-7.3275 0.2447 -0.0129>, rot=<-0.0208 0.0042 0.0858 -0.9961>}
 middle_02_r {parent=middle_01_r, loc=<-4.0123 -0.1401 -0.1175>, rot=<-0.0218 0.0258 0.3512 -0.9357>}
 middle_03_r {parent=middle_02_r, loc=<-2.6077 -0.0534 -0.1373>, rot=<0.0082 0.0188 0.1708 -0.9851>}
 ring_metacarpal_r {parent=hand_r, loc=<-2.2074 0.3625 -1.3638>, rot=<-0.0284 0.0722 -0.0458 -0.9959>}
 ring_01_r {parent=ring_metacarpal_r, loc=<-7.0295 0.6197 0.1446>, rot=<-0.0489 -0.0252 0.1892 -0.9804>}
 ring_02_r {parent=ring_01_r, loc=<-3.6174 -0.4647 0.1027>, rot=<0 0 0.2806 -0.9598>}
 ring_03_r {parent=ring_02_r, loc=<-2.4152 -0.114 -0.0199>, rot=<0 0 0.1189 -0.9929>}
 pinky_metacarpal_r {parent=hand_r, loc=<-1.7616 0.9753 -3.3814>, rot=<-0.0397 0.0814 -0.069 -0.9935>}
 pinky_01_r {parent=pinky_metacarpal_r, loc=<-7.1709 0.4831 0.1094>, rot=<-0.0654 -0.0035 0.2242 -0.9723>}
 pinky_02_r {parent=pinky_01_r, loc=<-2.8274 -0.0893 0.1043>, rot=<0.0037 0.0011 0.2854 -0.9584>}
 pinky_03_r {parent=pinky_02_r, loc=<-2.2046 -0.0307 0.0078>, rot=<-0.0002 0.0098 0.1445 -0.9895>}
 deform_wrist_dn_base_r {parent=hand_r, loc=<-0.0001 1.5328 0>, rot=<0 0 0 -1>}
 deform_wrist_dn_r {parent=deform_wrist_dn_base_r, loc=<0 0 0>, rot=<0 0 0 1>}
 deform_wrist_up_base_r {parent=hand_r, loc=<-0.0002 -2.3007 -0.0001>, rot=<0 0 0 -1>}
 deform_wrist_up_r {parent=deform_wrist_up_base_r, loc=<0 0 0>, rot=<0 0 0 1>}
 lowerarm_twist_02_r {parent=lowerarm_r, loc=<-10.6707 0 0>, rot=<0 0 0 -1>}
 lowerarm_twist_01_r {parent=lowerarm_r, loc=<-18.316 0 0>, rot=<0 0 0 -1>}
 deform_elbow_in_base_r {parent=lowerarm_r, loc=<-0.1834 -3.6062 0.2572>, rot=<0 0 0 -1>}
 deform_elbow_in_r {parent=deform_elbow_in_base_r, loc=<0 0 0>, rot=<0 0 0 1>}
 deform_elbow_out_base_r {parent=lowerarm_r, loc=<0 4 -0.0001>, rot=<0 0 0 -1>}
 deform_elbow_out_r {parent=deform_elbow_out_base_r, loc=<0 0 0>, rot=<0 0 0 1>}
 crv_lowerarm_r_bottom {parent=lowerarm_r, loc=<0 0 0>, rot=<0 0 0 -1>}
 crv_lowerarm_r_top {parent=lowerarm_r, loc=<0 0 0>, rot=<0 0 0 -1>}
 upperarm_twist_01_r {parent=upperarm_r, loc=<-7.2135 0.1134 0>, rot=<0 0 0 -1>}
 deform_upperarm_bk_base_r {parent=upperarm_twist_01_r, loc=<6.4053 4.3746 -0.314>, rot=<0 0 0 -1>}
 deform_upperarm_bk_r {parent=deform_upperarm_bk_base_r, loc=<0 0 0>, rot=<0 0 0 1>}
 deform_upperarm_fr_base_r {parent=upperarm_twist_01_r, loc=<5.6312 -5.092 -0.3354>, rot=<0 0 0 -1>}
 deform_upperarm_fr_r {parent=deform_upperarm_fr_base_r, loc=<0 0 0>, rot=<0 0 0 1>}
 deform_upperarm_up_base_r {parent=upperarm_twist_01_r, loc=<8.891 -0.4825 -3.9517>, rot=<0 0 0 -1>}
 deform_upperarm_up_r {parent=deform_upperarm_up_base_r, loc=<0 0 0>, rot=<0 0 0 1>}
 upperarm_twist_02_r {parent=upperarm_r, loc=<-19.7998 0.1134 0>, rot=<0 0 0 -1>}
 dfrm_arm_spline_mstr_r {parent=upperarm_r, loc=<0.1213 0.0811 -0.1948>, rot=<0 0 0 1>}
 dfrm_arm_spline_01_r {parent=dfrm_arm_spline_mstr_r, loc=<0 0 0>, rot=<0 0 0 1>}
 dfrm_arm_spline_02_r {parent=dfrm_arm_spline_01_r, loc=<-6.9267 -0.0011 0.2634>, rot=<0 0 0 -1>}
 dfrm_arm_spline_03_r {parent=dfrm_arm_spline_02_r, loc=<-5.9471 0.3982 -0.2824>, rot=<0 0 0 1>}
 dfrm_arm_spline_04_r {parent=dfrm_arm_spline_03_r, loc=<-5.8192 0.2729 -0.1479>, rot=<0 0 0 1>}
 dfrm_arm_spline_05_r {parent=dfrm_arm_spline_04_r, loc=<-6.2914 -0.0768 0.1232>, rot=<-0.0005 0.0165 0.0186 -0.9997>}
 dfrm_arm_spline_06_r {parent=dfrm_arm_spline_05_r, loc=<-5.3137 -2.0991 -0.0435>, rot=<0 0 -0.1903 -0.9817>}
 dfrm_arm_spline_07_r {parent=dfrm_arm_spline_06_r, loc=<-6.0382 -0.3421 0.4585>, rot=<0 0 -0.059 -0.9983>}
 dfrm_arm_spline_08_r {parent=dfrm_arm_spline_07_r, loc=<-5.8876 0.188 0.3621>, rot=<0 0 0 1>}
 dfrm_arm_spline_09_r {parent=dfrm_arm_spline_08_r, loc=<-6.5034 -0.0896 -0.1344>, rot=<0.0781 -0.0407 -0.0246 -0.9958>}
 upperarm_r_null {parent=clavicle_r, type=Null, loc=<-9.8658 1.2136 0.4105>, rot=<0.0164 -0.3245 -0.0269 -0.9454>}
 clavicle_r_null {parent=spine_05, type=Null, loc=<5.4831 -0.1053 3.2946>, rot=<0.7289 0.0465 -0.6829 0.009>}
 clavicle_l_null {parent=spine_05, type=Null, loc=<5.4831 -0.1053 -3.2946>, rot=<0.0465 -0.7289 -0.009 -0.6829>}
 neck_01_null {parent=spine_05, type=Null, loc=<11.3138 0.1529 0>, rot=<0 0 -0.2003 -0.9797>}
 attach_cape {parent=spine_05, loc=<9.3216 -8.4501 0>, rot=<0.0616 -0.7044 -0.0616 -0.7044>}
 attach_backpack {parent=spine_05, loc=<0 -8.35 0>, rot=<-0.0093 -0.707 0.0093 -0.707>}
 deform_pec_aim_base_l {parent=spine_04, loc=<3.2456 9.3309 -1.2237>, rot=<0.0603 -0.4033 0.3318 -0.8507>}
 deform_pec_aim_l {parent=deform_pec_aim_base_l, loc=<0 0 0>, rot=<0 0 0 1>}
 deform_pec_base_l {parent=deform_pec_aim_l, loc=<9.3665 0 0>, rot=<0 0 0 1>}
 deform_pec_l {parent=deform_pec_base_l, loc=<0 0 0>, rot=<0 0 0 1>}
 deform_pec_aim_base_r {parent=spine_04, loc=<3.2452 9.3309 1.2237>, rot=<-0.4033 -0.0603 0.8507 0.3318>}
 deform_pec_aim_r {parent=deform_pec_aim_base_r, loc=<0 0 0>, rot=<0 0 0 1>}
 deform_pec_base_r {parent=deform_pec_aim_r, loc=<-9.3667 -0.0003 -0.0004>, rot=<0 0 0 1>}
 deform_pec_r {parent=deform_pec_base_r, loc=<0 0 0>, rot=<0 0 0 1>}
 thigh_l {parent=pelvis, loc=<-1.0855 -0.2109 -8.7053>, rot=<0.0009 0.0355 0.0155 -0.9993>}
 calf_l {parent=thigh_l, loc=<-42.5284 -0.0019 -0.7297>, rot=<0.0006 -0.0161 -0.0391 -0.9991>}
 foot_l {parent=calf_l, loc=<-46.7401 -2.5014 -0.592>, rot=<0.0427 -0.0238 0.0353 -0.9982>}
 ball_l {parent=foot_l, loc=<-3.5658 10.8066 -0.1544>, rot=<0.0001 -0.0001 -0.7063 -0.7079>}
 calf_twist_02_l {parent=calf_l, loc=<-19.1546 -1.7943 -0.2638>, rot=<0.0237 0.0095 -0.0035 -0.9997>}
 calf_twist_01_l {parent=calf_l, loc=<-35.8483 -1.9197 -0.5794>, rot=<0.0237 0.0095 -0.0035 -0.9997>}
 foot_l_null {parent=calf_l, type=Null, loc=<-46.7401 -2.5014 -0.592>, rot=<0.0427 -0.0238 0.0353 -0.9982>}
 deform_knee_in_base_l {parent=calf_l, loc=<0.8031 -5.0231 0.2456>, rot=<0 0 0 -1>}
 deform_knee_in_l {parent=deform_knee_in_base_l, loc=<0 0 0>, rot=<0 0 0 1>}
 deform_knee_out_base_l {parent=calf_l, loc=<0 4 0>, rot=<0 0 0 -1>}
 deform_knee_out_l {parent=deform_knee_out_base_l, loc=<0 0 0>, rot=<0 0 0 1>}
 crv_calf_l_top {parent=calf_l, loc=<0 0 0>, rot=<0 0 0 -1>}
 crv_calf_l_bottom {parent=calf_l, loc=<0 0 0>, rot=<0 0 0 -1>}
 thigh_twist_01_l {parent=thigh_l, loc=<-27.952 0 0>, rot=<0.0013 0 0 -1>}
 deform_groin_base_l {parent=thigh_l, loc=<0.4801 5.6776 -0.2486>, rot=<0 0 0 -1>}
 deform_groin_l {parent=deform_groin_base_l, loc=<0 0 0>, rot=<0 0 0 1>}
 dfrm_leg_spline_mstr_l {parent=thigh_l, loc=<0 0 0>, rot=<0 0 0 1>}
 dfrm_leg_spline_01_l {parent=dfrm_leg_spline_mstr_l, loc=<0 0 0>, rot=<0 0 0 1>}
 dfrm_leg_spline_02_l {parent=dfrm_leg_spline_01_l, loc=<-12.026 -0.2025 -0.2394>, rot=<0.0013 0 0 -1>}
 dfrm_leg_spline_03_l {parent=dfrm_leg_spline_02_l, loc=<-10.8195 -0.0832 -0.2621>, rot=<0 0 0 1>}
 dfrm_leg_spline_04_l {parent=dfrm_leg_spline_03_l, loc=<-10.8288 -0.1514 -0.2684>, rot=<-0.0013 0 0 -1>}
 dfrm_leg_spline_05_l {parent=dfrm_leg_spline_04_l, loc=<-10.7714 -0.1933 -0.1547>, rot=<0.0006 -0.0161 -0.0391 -0.9991>}
 dfrm_leg_spline_06_l {parent=dfrm_leg_spline_05_l, loc=<-10.7897 -0.5063 -0.1368>, rot=<0.0237 0.0095 -0.0035 -0.9997>}
 dfrm_leg_spline_07_l {parent=dfrm_leg_spline_06_l, loc=<-10.7999 -0.3271 0.1242>, rot=<0 0 0 1>}
 dfrm_leg_spline_08_l {parent=dfrm_leg_spline_07_l, loc=<-10.7997 -0.6299 0.078>, rot=<0 0 0 -1>}
 dfrm_leg_spline_09_l {parent=dfrm_leg_spline_08_l, loc=<-12.4243 -0.6401 0.2819>, rot=<0.0146 0.0045 0.0069 -0.9999>}
 thigh_r {parent=pelvis, loc=<-1.0855 -0.2109 8.7053>, rot=<0.0355 -0.0009 0.9993 0.0155>}
 calf_r {parent=thigh_r, loc=<42.5284 0.0019 0.7297>, rot=<0.0006 -0.0161 -0.0391 -0.9991>}
 foot_r {parent=calf_r, loc=<46.7401 2.5014 0.592>, rot=<0.0427 -0.0238 0.0353 -0.9982>}
 ball_r {parent=foot_r, loc=<3.5658 -10.8066 0.1544>, rot=<0.0001 -0.0001 -0.7063 -0.7079>}
 calf_twist_02_r {parent=calf_r, loc=<19.1546 1.7943 0.2638>, rot=<0.0237 0.0095 -0.0035 -0.9997>}
 calf_twist_01_r {parent=calf_r, loc=<35.8483 1.9197 0.5794>, rot=<0.0237 0.0095 -0.0035 -0.9997>}
 foot_r_null {parent=calf_r, type=Null, loc=<46.7401 2.5014 0.592>, rot=<0.0427 -0.0238 0.0353 -0.9982>}
 deform_knee_in_base_r {parent=calf_r, loc=<-0.8031 5.0231 -0.2457>, rot=<0 0 0 -1>}
 deform_knee_in_r {parent=deform_knee_in_base_r, loc=<0 0 0>, rot=<0 0 0 1>}
 deform_knee_out_base_r {parent=calf_r, loc=<0 -4 0>, rot=<0 0 0 -1>}
 deform_knee_out_r {parent=deform_knee_out_base_r, loc=<0 0 0>, rot=<0 0 0 1>}
 crv_calf_r_top {parent=calf_r, loc=<0 0 0>, rot=<0 0 0 -1>}
 crv_calf_r_bottom {parent=calf_r, loc=<0 0 0>, rot=<0 0 0 -1>}
 thigh_twist_01_r {parent=thigh_r, loc=<27.952 0 0>, rot=<0.0013 0 0 -1>}
 deform_groin_base_r {parent=thigh_r, loc=<-0.4801 -5.6776 0.2486>, rot=<0 0 0 -1>}
 deform_groin_r {parent=deform_groin_base_r, loc=<0 0 0>, rot=<0 0 0 1>}
 dfrm_leg_spline_mstr_r {parent=thigh_r, loc=<0 0 0>, rot=<0 0 0 -1>}
 dfrm_leg_spline_01_r {parent=dfrm_leg_spline_mstr_r, loc=<0 0 0>, rot=<0 0 0 1>}
 dfrm_leg_spline_02_r {parent=dfrm_leg_spline_01_r, loc=<12.026 0.2025 0.2394>, rot=<0.0013 0 0 -1>}
 dfrm_leg_spline_03_r {parent=dfrm_leg_spline_02_r, loc=<10.8194 0.0832 0.2621>, rot=<0 0 0 1>}
 dfrm_leg_spline_04_r {parent=dfrm_leg_spline_03_r, loc=<10.8288 0.1514 0.2684>, rot=<-0.0013 0 0 -1>}
 dfrm_leg_spline_05_r {parent=dfrm_leg_spline_04_r, loc=<10.7714 0.1933 0.1547>, rot=<0.0006 -0.0161 -0.0391 -0.9991>}
 dfrm_leg_spline_06_r {parent=dfrm_leg_spline_05_r, loc=<10.7897 0.5063 0.1368>, rot=<0.0237 0.0095 -0.0035 -0.9997>}
 dfrm_leg_spline_07_r {parent=dfrm_leg_spline_06_r, loc=<10.7999 0.3271 -0.1242>, rot=<0 0 0 1>}
 dfrm_leg_spline_08_r {parent=dfrm_leg_spline_07_r, loc=<10.7997 0.6299 -0.078>, rot=<0 0 0 -1>}
 dfrm_leg_spline_09_r {parent=dfrm_leg_spline_08_r, loc=<12.4244 0.6401 -0.2819>, rot=<0.0146 0.0045 0.0069 -0.9999>}
 thigh_r_null {parent=pelvis, type=Null, loc=<-1.0855 -0.2109 8.7053>, rot=<0.0355 -0.0009 0.9993 0.0155>}
 thigh_l_null {parent=pelvis, type=Null, loc=<-1.0855 -0.2109 -8.7053>, rot=<0.0009 0.0355 0.0155 -0.9993>}
 deform_glute_base_l {parent=pelvis, loc=<-7.3466 -9.0213 -9.1245>, rot=<0.0172 0.031 -0.4477 -0.8935>}
 deform_glute_l {parent=deform_glute_base_l, loc=<0 0 0>, rot=<0 0 0 1>}
 deform_glute_base_r {parent=pelvis, loc=<-7.3464 -9.0213 9.1245>, rot=<0.031 -0.0172 0.8935 -0.4477>}
 deform_glute_r {parent=deform_glute_base_r, loc=<0 0 0>, rot=<0 0 0 1>}
 pelvis_ctrl {controlType=EULER_TRANSFORM, type=Control, loc=<0 0 0>, rot=<0 0 0 1>}
 spine_01_ctrl {controlType=EULER_TRANSFORM, parent=pelvis_ctrl, type=Control, loc=<-6.0855 -0.7259 104.2074>, rot=<-0.0355 0.7062 -0.0127 -0.707>}
 spine_02_ctrl {controlType=EULER_TRANSFORM, parent=spine_01_ctrl, type=Control, loc=<0 0 0>, rot=<0 0 0 1>}
 spine_03_ctrl {controlType=EULER_TRANSFORM, parent=spine_02_ctrl, type=Control, loc=<0 0 0>, rot=<0 0 0 1>}
 spine_04_ctrl {controlType=EULER_TRANSFORM, parent=spine_03_ctrl, type=Control, loc=<0 0 0>, rot=<0 0 0 1>}
 spine_05_ctrl {controlType=EULER_TRANSFORM, parent=spine_04_ctrl, type=Control, loc=<0 0 0>, rot=<0 0 0 1>}
 neck_01_ctrl {controlType=EULER_TRANSFORM, parent=spine_05_ctrl, type=Control, loc=<0 0 0>, rot=<0 0 0 1>}
 neck_02_ctrl {controlType=EULER_TRANSFORM, parent=neck_01_ctrl, type=Control, loc=<0 0 0>, rot=<0 0 0 1>}
 head_ctrl {controlType=EULER_TRANSFORM, parent=neck_02_ctrl, type=Control, loc=<0 0 0>, rot=<0 0 0 1>}
 clavicle_r_ctrl {controlType=EULER_TRANSFORM, parent=spine_05_ctrl, type=Control, loc=<0 0 0>, rot=<0 0 0 1>}
 upperarm_r_ctrl {controlType=EULER_TRANSFORM, parent=clavicle_r_ctrl, type=Control, loc=<0 0 0>, rot=<0 0 0 1>}
 lowerarm_r_ctrl {controlType=EULER_TRANSFORM, parent=upperarm_r_ctrl, type=Control, loc=<0 0 0>, rot=<0 0 0 1>}
 hand_r_ctrl {controlType=EULER_TRANSFORM, parent=lowerarm_r_ctrl, type=Control, loc=<0 0 0>, rot=<0 0 0 1>}
 clavicle_l_ctrl {controlType=EULER_TRANSFORM, parent=spine_05_ctrl, type=Control, loc=<0 0 0>, rot=<0 0 0 1>}
 upperarm_l_ctrl {controlType=EULER_TRANSFORM, parent=clavicle_l_ctrl, type=Control, loc=<0 0 0>, rot=<0 0 0 1>}
 lowerarm_l_ctrl {controlType=EULER_TRANSFORM, parent=upperarm_l_ctrl, type=Control, loc=<0 0 0>, rot=<0 0 0 1>}
 hand_l_ctrl {controlType=EULER_TRANSFORM, parent=lowerarm_l_ctrl, type=Control, loc=<0 0 0>, rot=<0 0 0 1>}
 thigh_r_ctrl {controlType=EULER_TRANSFORM, parent=pelvis_ctrl, type=Control, loc=<0 0 0>, rot=<0 0 0 1>}
 calf_r_ctrl {controlType=EULER_TRANSFORM, parent=thigh_r_ctrl, type=Control, loc=<0 0 0>, rot=<0 0 0 1>}
 foot_r_ctrl {controlType=EULER_TRANSFORM, parent=calf_r_ctrl, type=Control, loc=<0 0 0>, rot=<0 0 0 1>}
 thigh_l_ctrl {controlType=EULER_TRANSFORM, parent=pelvis_ctrl, type=Control, loc=<0 0 0>, rot=<0 0 0 1>}
 calf_l_ctrl {controlType=EULER_TRANSFORM, parent=thigh_l_ctrl, type=Control, loc=<0 0 0>, rot=<0 0 0 1>}
 foot_l_ctrl {controlType=EULER_TRANSFORM, parent=calf_l_ctrl, type=Control, loc=<0 0 0>, rot=<0 0 0 1>}
 r_knee_corrective {type=Curve, loc=<0 0 0>, rot=<0 0 0 1>}
 l_knee_corrective {type=Curve, loc=<0 0 0>, rot=<0 0 0 1>}
 mask_geo {type=Curve, loc=<0 0 0>, rot=<0 0 0 1>}
 dfrm_arm_spline_01_r_null {type=Null, loc=<-13.0772 -5.8989 135.7715>, rot=<-0.9986 -0.0044 -0.0484 0.0216>}
 dfrm_arm_spline_01_r_ctrl {controlType=EULER_TRANSFORM, parent=dfrm_arm_spline_01_r_null, type=Control, loc=<0 0 0>, rot=<0 0 0 1>}
 dfrm_arm_spline_03_r_null {type=Null, loc=<-22.6987 -6.2623 127.2164>, rot=<-0.9343 0.0039 -0.3563 -0.0065>}
 dfrm_arm_spline_03_r_ctrl {controlType=EULER_TRANSFORM, parent=dfrm_arm_spline_03_r_null, type=Control, loc=<0 0 0>, rot=<0 0 0 1>}
 dfrm_arm_spline_05_r_null {type=Null, loc=<-31.7521 -6.4267 119.172>, rot=<-0.94 -0.0135 -0.3407 -0.0126>}
 dfrm_arm_spline_05_r_ctrl {controlType=EULER_TRANSFORM, parent=dfrm_arm_spline_05_r_null, type=Control, loc=<0 0 0>, rot=<0 0 0 1>}
 dfrm_arm_spline_07_r_null {type=Null, loc=<-39.8436 -2.1307 111.7942>, rot=<-0.914 0.2199 -0.3332 0.0722>}
 dfrm_arm_spline_07_r_ctrl {controlType=EULER_TRANSFORM, parent=dfrm_arm_spline_07_r_null, type=Control, loc=<0 0 0>, rot=<0 0 0 1>}
 dfrm_arm_spline_09_r_null {type=Null, loc=<-48.1743 3.3545 104.4381>, rot=<-0.8968 0.2704 -0.3501 -0.0003>}
 dfrm_arm_spline_09_r_ctrl {controlType=EULER_TRANSFORM, parent=dfrm_arm_spline_09_r_null, type=Control, loc=<0 0 0>, rot=<0 0 0 1>}
 dfrm_arm_spline_01_l_null {type=Null, loc=<13.0772 -5.8989 135.771>, rot=<-0.0969 0.9938 0.0298 -0.0453>}
 dfrm_arm_spline_01_l_ctrl {controlType=EULER_TRANSFORM, parent=dfrm_arm_spline_01_l_null, type=Control, loc=<0 0 0>, rot=<0 0 0 1>}
 dfrm_arm_spline_03_l_null {type=Null, loc=<22.6987 -6.2623 127.216>, rot=<-0.0065 0.3563 0.0039 0.9343>}
 dfrm_arm_spline_03_l_ctrl {controlType=EULER_TRANSFORM, parent=dfrm_arm_spline_03_l_null, type=Control, loc=<0 0 0>, rot=<0 0 0 1>}
 dfrm_arm_spline_05_l_null {type=Null, loc=<31.7521 -6.4267 119.172>, rot=<-0.0126 0.3407 -0.0135 0.94>}
 dfrm_arm_spline_05_l_ctrl {controlType=EULER_TRANSFORM, parent=dfrm_arm_spline_05_l_null, type=Control, loc=<0 0 0>, rot=<0 0 0 1>}
 dfrm_arm_spline_07_l_null {type=Null, loc=<39.8436 -2.1307 111.794>, rot=<0.0722 0.3332 0.2199 0.914>}
 dfrm_arm_spline_07_l_ctrl {controlType=EULER_TRANSFORM, parent=dfrm_arm_spline_07_l_null, type=Control, loc=<0 0 0>, rot=<0 0 0 1>}
 dfrm_arm_spline_09_l_null {type=Null, loc=<48.1743 3.3545 104.438>, rot=<-0.0003 0.3501 0.2704 0.8968>}
 dfrm_arm_spline_09_l_ctrl {controlType=EULER_TRANSFORM, parent=dfrm_arm_spline_09_l_null, type=Control, loc=<0 0 0>, rot=<0 0 0 1>}
 dfrm_leg_spline_01_l_null {type=Null, loc=<8.7053 -0.7798 96.9799>, rot=<-0.0019 -0.7317 0.0015 0.6816>}
 dfrm_leg_spline_01_l_ctrl {controlType=EULER_TRANSFORM, parent=dfrm_leg_spline_01_l_null, type=Control, loc=<0 0 0>, rot=<0 0 0 1>}
 dfrm_leg_spline_03_l_null {type=Null, loc=<10.8252 -1.1771 74.2292>, rot=<-0.0028 -0.7317 0.0006 0.6816>}
 dfrm_leg_spline_03_l_ctrl {controlType=EULER_TRANSFORM, parent=dfrm_leg_spline_03_l_null, type=Control, loc=<0 0 0>, rot=<0 0 0 1>}
 dfrm_leg_spline_05_l_null {type=Null, loc=<12.7784 -1.6275 52.7152>, rot=<-0.0309 -0.7201 0.0276 0.6927>}
 dfrm_leg_spline_05_l_ctrl {controlType=EULER_TRANSFORM, parent=dfrm_leg_spline_05_l_null, type=Control, loc=<0 0 0>, rot=<0 0 0 1>}
 dfrm_leg_spline_07_l_null {type=Null, loc=<13.8058 -4.3214 31.3008>, rot=<-0.0496 -0.727 0.0133 0.6847>}
 dfrm_leg_spline_07_l_ctrl {controlType=EULER_TRANSFORM, parent=dfrm_leg_spline_07_l_null, type=Control, loc=<0 0 0>, rot=<0 0 0 1>}
 dfrm_leg_spline_09_l_null {type=Null, loc=<14.7108 -7.6657 8.2987>, rot=<-0.0546 -0.7305 -0.0018 0.6807>}
 dfrm_leg_spline_09_l_ctrl {controlType=EULER_TRANSFORM, parent=dfrm_leg_spline_09_l_null, type=Control, loc=<0 0 0>, rot=<0 0 0 1>}
 dfrm_leg_spline_01_r_null {type=Null, loc=<-8.7053 -0.7798 96.9799>, rot=<0.6816 -0.0015 -0.7317 0.0019>}
 dfrm_leg_spline_01_r_ctrl {controlType=EULER_TRANSFORM, parent=dfrm_leg_spline_01_r_null, type=Control, loc=<0 0 0>, rot=<0 0 0 1>}
 dfrm_leg_spline_03_r_null {type=Null, loc=<-10.8252 -1.1772 74.2292>, rot=<0.6816 -0.0006 -0.7317 0.0028>}
 dfrm_leg_spline_03_r_ctrl {controlType=EULER_TRANSFORM, parent=dfrm_leg_spline_03_r_null, type=Control, loc=<0 0 0>, rot=<0 0 0 1>}
 dfrm_leg_spline_05_r_null {type=Null, loc=<-12.7784 -1.6275 52.7152>, rot=<0.6927 -0.0276 -0.7201 0.0309>}
 dfrm_leg_spline_05_r_ctrl {controlType=EULER_TRANSFORM, parent=dfrm_leg_spline_05_r_null, type=Control, loc=<0 0 0>, rot=<0 0 0 1>}
 dfrm_leg_spline_07_r_null {type=Null, loc=<-13.8058 -4.3214 31.3008>, rot=<0.6847 -0.0133 -0.727 0.0496>}
 dfrm_leg_spline_07_r_ctrl {controlType=EULER_TRANSFORM, parent=dfrm_leg_spline_07_r_null, type=Control, loc=<0 0 0>, rot=<0 0 0 1>}
 dfrm_leg_spline_09_r_null {type=Null, loc=<-14.7108 -7.6657 8.2987>, rot=<0.6807 0.0018 -0.7305 0.0546>}
 dfrm_leg_spline_09_r_ctrl {controlType=EULER_TRANSFORM, parent=dfrm_leg_spline_09_r_null, type=Control, loc=<0 0 0>, rot=<0 0 0 1>}
]
variables [
 Is Sky Dive Float {type=double, default=0}
 IsPuzzleshedEmotePlay {type=bool, default=False}
 RBAN_on {type=bool, default=False}
 arm_l_bind_pose {type=FRigPose, default=(Elements=,HierarchyTopologyVersion=-1,PoseHash=-1)}
 arm_l_ctrls {type="TArray<FRigElementKey>"}
 arm_r_bind_pose {type=FRigPose, default=(Elements=,HierarchyTopologyVersion=-1,PoseHash=-1)}
 arm_r_ctrls {type="TArray<FRigElementKey>"}
 is skydiving {type=double, default=0}
 isCrouching {type=bool, default=False}
 isCrouchingMoving {type=double, default=0}
 isSlidingwithPickAxe {type=bool, default=False}
 leg_l_bind_pose {type=FRigPose, default=(Elements=,HierarchyTopologyVersion=-1,PoseHash=-1)}
 leg_l_ctrls {type="TArray<FRigElementKey>"}
 leg_r_bind_pose {type=FRigPose, default=(Elements=,HierarchyTopologyVersion=-1,PoseHash=-1)}
 leg_r_ctrls {type="TArray<FRigElementKey>"}
 thingh_Rot_L {type=double, default=0}
 thingh_Rot_R {type=double, default=0}
]
