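# Anim Blueprint for the F_MED_CajunTaco female body cosmetic, layering corrective motion onto a pose copied from its attached parent mesh.
# The copied pose runs through the body's deform Control Rig, fed skydive-floating and PuzzleShed emote-playing states.
# An additive offset then shifts the attach_backpack bone whenever rigid body simulation is disabled.
# The update event refreshes these states, querying the owner mesh's anim instance for the emote montage.
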AnimBlueprint /BRCosmetics/Characters/Player/Female/Medium/Bodies/F_MED_CajunTaco/Meshes/F_MED_CajunTaco_AnimBP
target_skeleton=&/F_MED_CajunTaco_Skeleton
graphs [
 AnimGraph {nodes [
   n1 = Control Rig!(
    Source=Copy Pose From Mesh() {bUseAttachedParent=true}
    Alpha
    Is Sky Dive Float:double=Get is skydiving float alpha() -> (is skydiving float alpha:double)
    IsPuzzleshedEmotePlay:bool=Get isPuzzleshedEmotePlaying() -> (isPuzzleshedEmotePlaying:bool)
   ) -> (Pose) {
    alphaBoolBlend {blendOption=Linear}
    controlRigAssetReference {blueprintRigClass=&/F_MED_CajunTaco_DeformCtrlRig.F_MED_CajunTaco_DeformCtrlRig_C}
   }
   n2 = "Transform (Modify) Bone - Bone: attach_backpack"!(
    ComponentPose=Local To Component(LocalPose=@n1)
    bAlphaBoolEnabled:bool=Get RBANoffAlpha() -> (RBANoffAlpha:bool)
    Translation="0,-6,0"
    Rotation="0,0,0"
    Scale="1,1,1"
   ) -> (Pose) {
    alphaInputType=Bool
    rotationMode=BMM_Additive
    rotationSpace=BCS_BoneSpace
    translationMode=BMM_Additive
    translationSpace=BCS_BoneSpace
    alphaBoolBlend {blendOption=Linear}
    boneToModify {boneName=attach_backpack}
    translation {y=-6}
   }
   Output Pose(Result=Component To Local(ComponentPose=@n2)) {layerGroup=None}
  ]}
 EventGraph {nodes [
   Event BlueprintUpdateAnimation()
   > n1 = Sequence()
   > Set is skydiving float alpha(is skydiving float alpha:double="To Float (Boolean)"(InBool=Get bIsSkydiveFloating())) -> (Output_Get:double)
   > Set RBANoffAlpha(RBANoffAlpha:bool=NOT Boolean(A=Get bRigidBodyEnabled())) -> (Output_Get:bool)
   n2 = GetAnimInstance(self=GetOwnerSkeletalMeshComponent())
   Is Valid(exec=@n1.then_1, InputObject=@n2)
   > Set isPuzzleshedEmotePlaying(isPuzzleshedEmotePlaying:bool=Montage_IsPlaying(self=@n2, Montage=/BRCosmetics/Animation/Game/MainPlayer/Emotes/PuzzleShed/CMF/Emote_PuzzleShed_CMF_M)) -> (Output_Get:bool)
  ]}
]
variables [
 is skydiving float alpha {type=number}
 isPuzzleshedEmotePlaying {type=boolean}
 RBANoffAlpha {type=boolean}
]
