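# Anim Blueprint for the CajunTaco female face accessory part, driving its three hair strands (left, center, right).
# A Control Rig adjusts the copied mesh pose using gameplay states like aiming, skydiving and emotes.
# Its cached pose feeds two dynamics branches: AnimDynamics, bone offsets and RigidBody, plus a backpack variant, blended by bools before hair-tip offsets.
# The event graph updates gravity overrides from movement, DBNO alpha and emote montage flags.
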AnimBlueprint /BRCosmetics/Characters/Player/Female/Medium/Bodies/F_MED_CajunTaco/Meshes/Parts/F_MED_CajunTaco_FaceAcc_AnimBP
target_skeleton=&/F_MED_CajunTaco_FaceAcc_Skeleton
graphs [
 AnimGraph {nodes [
   Component To Local()
   Copy Pose From Mesh() {bUseAttachedParent=true}
   n1 = Blend Poses by bool(
    BlendPose_0=Use cached pose 'CRCache_HairDynBackPackPose'()
    BlendPose_1=Blend Poses by bool(BlendPose_0=Use cached pose 'CRCache_HairDynPose'(), BlendPose_1=Use cached pose 'CRCache_HairPose'()) {blendType=Linear}
   ) {blendType=Linear}
   n2 = Use cached pose 'CRCache_HairPose'()
   n3 = "Transform (Modify) Bone - Bone: dyn_C_hair_5"!(
    ComponentPose=Local To Component(LocalPose=Blend Poses by bool(BlendPose_0=@n2, BlendPose_1=@n1) {blendType=Linear})
    bAlphaBoolEnabled:bool=True
    Translation="-3,0,0"
    Rotation="-10,0,0"
   ) -> (Pose) {
    alphaInputType=Bool
    rotationMode=BMM_Additive
    rotationSpace=BCS_BoneSpace
    translationMode=BMM_Additive
    translationSpace=BCS_BoneSpace
    alphaBoolBlend {blendInTime=0.5, blendOption=Linear}
    boneToModify {boneName=dyn_C_hair_5}
    rotation {pitch=-10}
    translation {x=-3}
   }
   n4 = "Transform (Modify) Bone - Bone: dyn_L_hair_5"!(ComponentPose=@n3, bAlphaBoolEnabled:bool=True, Translation="-3,0,0", Rotation="-10,0,0") -> (Pose) {
    alphaInputType=Bool
    rotationMode=BMM_Additive
    rotationSpace=BCS_BoneSpace
    translationMode=BMM_Additive
    translationSpace=BCS_BoneSpace
    alphaBoolBlend {blendInTime=0.5, blendOption=Linear}
    boneToModify {boneName=dyn_L_hair_5}
    rotation {pitch=-10}
    translation {x=-3}
   }
   n5 = "Transform (Modify) Bone - Bone: dyn_R_hair_5"!(ComponentPose=@n4, bAlphaBoolEnabled:bool=True, Translation="-3,0,0", Rotation="-10,0,0") -> (Pose) {
    alphaInputType=Bool
    rotationMode=BMM_Additive
    rotationSpace=BCS_BoneSpace
    translationMode=BMM_Additive
    translationSpace=BCS_BoneSpace
    alphaBoolBlend {blendInTime=0.5, blendOption=Linear}
    boneToModify {boneName=dyn_R_hair_5}
    rotation {pitch=-10}
    translation {x=-3}
   }
   n6 = "Transform (Modify) Bone - Bone: dyn_C_hair_6"!(ComponentPose=@n5, bAlphaBoolEnabled:bool=True, Translation="-3,0,1", Rotation="-20,0,0") -> (Pose) {
    alphaInputType=Bool
    rotationMode=BMM_Additive
    rotationSpace=BCS_BoneSpace
    translationMode=BMM_Additive
    translationSpace=BCS_BoneSpace
    alphaBoolBlend {blendInTime=0.5, blendOption=Linear}
    boneToModify {boneName=dyn_C_hair_6}
    rotation {pitch=-20}
    translation {x=-3, z=1}
   }
   n7 = "Transform (Modify) Bone - Bone: dyn_L_hair_6"!(ComponentPose=@n6, bAlphaBoolEnabled:bool=True, Translation="-3,0,1", Rotation="-20,0,0") -> (Pose) {
    alphaInputType=Bool
    rotationMode=BMM_Additive
    rotationSpace=BCS_BoneSpace
    translationMode=BMM_Additive
    translationSpace=BCS_BoneSpace
    alphaBoolBlend {blendInTime=0.5, blendOption=Linear}
    boneToModify {boneName=dyn_L_hair_6}
    rotation {pitch=-20}
    translation {x=-3, z=1}
   }
   n8 = "Transform (Modify) Bone - Bone: dyn_R_hair_6"!(ComponentPose=@n7, bAlphaBoolEnabled:bool=True, Translation="-3,0,1", Rotation="-20,0,0") -> (Pose) {
    alphaInputType=Bool
    rotationMode=BMM_Additive
    rotationSpace=BCS_BoneSpace
    translationMode=BMM_Additive
    translationSpace=BCS_BoneSpace
    alphaBoolBlend {blendInTime=0.5, blendOption=Linear}
    boneToModify {boneName=dyn_R_hair_6}
    rotation {pitch=-20}
    translation {x=-3, z=1}
   }
   n9 = "Transform (Modify) Bone - Bone: dyn_C_hair_7"!(ComponentPose=@n8, bAlphaBoolEnabled:bool=True, Translation="-3,0,2", Rotation="-20,0,0") -> (Pose) {
    alphaInputType=Bool
    rotationMode=BMM_Additive
    rotationSpace=BCS_BoneSpace
    translationMode=BMM_Additive
    translationSpace=BCS_BoneSpace
    alphaBoolBlend {blendInTime=0.5, blendOption=Linear}
    boneToModify {boneName=dyn_C_hair_7}
    rotation {pitch=-20}
    translation {x=-3, z=2}
   }
   n10 = "Transform (Modify) Bone - Bone: dyn_L_hair_7"!(ComponentPose=@n9, bAlphaBoolEnabled:bool=True, Translation="-3,0,2", Rotation="-20,0,0") -> (Pose) {
    alphaInputType=Bool
    rotationMode=BMM_Additive
    rotationSpace=BCS_BoneSpace
    translationMode=BMM_Additive
    translationSpace=BCS_BoneSpace
    alphaBoolBlend {blendInTime=0.5, blendOption=Linear}
    boneToModify {boneName=dyn_L_hair_7}
    rotation {pitch=-20}
    translation {x=-3, z=2}
   }
   n11 = "Transform (Modify) Bone - Bone: dyn_R_hair_7"!(ComponentPose=@n10, bAlphaBoolEnabled:bool=True, Translation="-3,0,2", Rotation="-20,0,0") -> (Pose) {
    alphaInputType=Bool
    rotationMode=BMM_Additive
    rotationSpace=BCS_BoneSpace
    translationMode=BMM_Additive
    translationSpace=BCS_BoneSpace
    alphaBoolBlend {blendInTime=0.5, blendOption=Linear}
    boneToModify {boneName=dyn_R_hair_7}
    rotation {pitch=-20}
    translation {x=-3, z=2}
   }
   Output Pose(Result=Component To Local(ComponentPose=@n11)) {layerGroup=None}
   n12 = Control Rig!(
    Source=Copy Pose From Mesh(SourceMeshComponent=Get Skeletal_Mesh() -> (Skeletal_Mesh:Skeletal Mesh Component Object Reference)) {bUseAttachedParent=true}
    Alpha
    pub disablecontrolRig:bool=Get disableControlRig() -> (disableControlRig:bool)
    pub disableUpperLimbsCorrection:bool=Get disableUpperLimbsCorrection() -> (disableUpperLimbsCorrection:bool)
    pub diveFloating:bool=Get bIsSkydiveFloating()
    pub isTargeting:bool=Get bIsTargeting()
    pub aimPitch:double=Get AimPitch()
    pub isAssaultRifleEquipped:double=Get isAssaultRifleEquipped float() -> (isAssaultRifleEquipped float:double)
    Is Skydiving:bool=Get bIsSkyDiving()
    Skydive Alpha:double=0
    Is Hang Gliding:double=Get Hang Gliding Alpha() -> (Hang Gliding Alpha:double)
    Not DBNO Alpha:double=Get Not DBNO Alpha() -> (Not DBNO Alpha:double)
    IsPlayingCajuntacoEmote:bool=Get isCajunTacoEmotePlaying() -> (isCajunTacoEmotePlaying:bool)
    IsPlayingCelebrationEmote:bool=Get IsPlayingCelebration() -> (IsPlayingCelebration:bool)
   ) -> (Pose) {
    alphaBoolBlend {blendOption=Linear}
    controlRigAssetReference {blueprintRigClass=&/F_MED_CajunTaco_FaceAcc_CtrlRig.F_MED_CajunTaco_FaceAcc_CtrlRig_C}
   }
   Save cached pose 'CRCache_HairPose'(Pose=@n12) {cacheName=CRCache_HairPose}
   n13 = Anim Dynamics - Bone: dyn_L_hair_2!(
    ComponentPose=Local To Component(LocalPose=Use cached pose 'CRCache_HairPose'())
    bAlphaBoolEnabled:bool=True
    GravityOverride:Vector="0,0,-980"
    AngularSpringConstant:float=20000
    ExternalForce="0,0,0"
   ) -> (Pose) {
    alphaInputType=Bool
    angularDampingOverride=0.9999
    angularSpringConstant=20000
    bAngularSpring=true
    bOverrideAngularDamping=true
    bUseGravityOverride=true
    gravityScale=0.35
    simulationSpace=RootRelative
    alphaBoolBlend {blendInTime=2, blendOption=Cubic}
    boundBone {boneName=dyn_L_hair_2}
    componentAppliedLinearAccClamp=<3000 3000 3000>
    componentLinearAccScale=<0.3 0.4 0.3>
    componentLinearVelScale=<0.3 0.4 0.3>
    gravityOverride {z=-980}
    physicsBodyDefinitions [{
      boundBone {boneName=dyn_L_hair_2}
      boxExtents=<25 25 25>
      constraintSetup {angularLimitsMax=<0 40 0>, angularLimitsMin=<0 0 0>, angularTarget=<1 0 0>}
      localJointOffset=<2 0 0>
     }]
    relativeSpaceBone {boneName=root}
   }
   n14 = Anim Dynamics - Bone: dyn_C_hair_2!(ComponentPose=@n13, bAlphaBoolEnabled:bool=True, GravityOverride:Vector="0,0,-980", AngularSpringConstant:float=20000, ExternalForce="0,0,0") -> (Pose) {
    alphaInputType=Bool
    angularDampingOverride=0.9999
    angularSpringConstant=20000
    bAngularSpring=true
    bOverrideAngularDamping=true
    bUseGravityOverride=true
    gravityScale=0.35
    simulationSpace=RootRelative
    alphaBoolBlend {blendInTime=2, blendOption=Cubic}
    boundBone {boneName=dyn_C_hair_2}
    componentAppliedLinearAccClamp=<3000 3000 3000>
    componentLinearAccScale=<0.3 0.4 0.3>
    componentLinearVelScale=<0.3 0.4 0.3>
    gravityOverride {z=-980}
    physicsBodyDefinitions [{
      boundBone {boneName=dyn_C_hair_2}
      boxExtents=<25 25 25>
      constraintSetup {angularLimitsMax=<0 40 0>, angularLimitsMin=<0 0 0>, angularTarget=<1 0 0>}
      localJointOffset=<2 0 0>
     }]
    relativeSpaceBone {boneName=root}
   }
   n15 = Anim Dynamics - Bone: dyn_R_hair_2!(ComponentPose=@n14, bAlphaBoolEnabled:bool=True, GravityOverride:Vector="0,0,-980", AngularSpringConstant:float=20000, ExternalForce="0,0,0") -> (Pose) {
    alphaInputType=Bool
    angularDampingOverride=0.9999
    angularSpringConstant=20000
    bAngularSpring=true
    bOverrideAngularDamping=true
    bUseGravityOverride=true
    gravityScale=0.35
    simulationSpace=RootRelative
    alphaBoolBlend {blendInTime=2, blendOption=Cubic}
    boundBone {boneName=dyn_R_hair_2}
    componentAppliedLinearAccClamp=<3000 3000 3000>
    componentLinearAccScale=<0.3 0.4 0.3>
    componentLinearVelScale=<0.3 0.4 0.3>
    gravityOverride {z=-980}
    physicsBodyDefinitions [{
      boundBone {boneName=dyn_R_hair_2}
      boxExtents=<25 25 25>
      constraintSetup {angularLimitsMax=<0 40 0>, angularLimitsMin=<0 0 0>, angularTarget=<1 0 0>}
      localJointOffset=<2 0 0>
     }]
    relativeSpaceBone {boneName=root}
   }
   n16 = Anim Dynamics - Bone: dyn_L_hair_3!(ComponentPose=@n15, bAlphaBoolEnabled:bool=True, GravityOverride:Vector="0,0,-980", AngularSpringConstant:float=20000, ExternalForce="0,0,0") -> (Pose) {
    alphaInputType=Bool
    angularDampingOverride=0.9999
    angularSpringConstant=20000
    bAngularSpring=true
    bOverrideAngularDamping=true
    bUseGravityOverride=true
    gravityScale=0.35
    simulationSpace=RootRelative
    alphaBoolBlend {blendInTime=2, blendOption=Cubic}
    boundBone {boneName=dyn_L_hair_3}
    componentAppliedLinearAccClamp=<3000 3000 3000>
    componentLinearAccScale=<0.3 0.4 0.3>
    componentLinearVelScale=<0.3 0.4 0.3>
    gravityOverride {z=-980}
    physicsBodyDefinitions [{
      boundBone {boneName=dyn_L_hair_3}
      boxExtents=<25 25 25>
      constraintSetup {angularLimitsMax=<0 35 2>, angularLimitsMin=<0 0 -2>, angularTarget=<1 0 0>}
      localJointOffset=<2 0 0>
     }]
    relativeSpaceBone {boneName=root}
   }
   n17 = Anim Dynamics - Bone: dyn_C_hair_3!(ComponentPose=@n16, bAlphaBoolEnabled:bool=True, GravityOverride:Vector="0,0,-980", AngularSpringConstant:float=20000, ExternalForce="0,0,0") -> (Pose) {
    alphaInputType=Bool
    angularDampingOverride=0.9999
    angularSpringConstant=20000
    bAngularSpring=true
    bOverrideAngularDamping=true
    bUseGravityOverride=true
    gravityScale=0.35
    simulationSpace=RootRelative
    alphaBoolBlend {blendInTime=2, blendOption=Cubic}
    boundBone {boneName=dyn_C_hair_3}
    componentAppliedLinearAccClamp=<3000 3000 3000>
    componentLinearAccScale=<0.3 0.4 0.3>
    componentLinearVelScale=<0.3 0.4 0.3>
    gravityOverride {z=-980}
    physicsBodyDefinitions [{
      boundBone {boneName=dyn_C_hair_3}
      boxExtents=<25 25 25>
      constraintSetup {angularLimitsMax=<0 35 2>, angularLimitsMin=<0 0 -2>, angularTarget=<1 0 0>}
      localJointOffset=<2 0 0>
     }]
    relativeSpaceBone {boneName=root}
   }
   n18 = Anim Dynamics - Bone: dyn_R_hair_3!(ComponentPose=@n17, bAlphaBoolEnabled:bool=True, GravityOverride:Vector="0,0,-980", AngularSpringConstant:float=20000, ExternalForce="0,0,0") -> (Pose) {
    alphaInputType=Bool
    angularDampingOverride=0.9999
    angularSpringConstant=20000
    bAngularSpring=true
    bOverrideAngularDamping=true
    bUseGravityOverride=true
    gravityScale=0.35
    simulationSpace=RootRelative
    alphaBoolBlend {blendInTime=2, blendOption=Cubic}
    boundBone {boneName=dyn_R_hair_3}
    componentAppliedLinearAccClamp=<3000 3000 3000>
    componentLinearAccScale=<0.3 0.4 0.3>
    componentLinearVelScale=<0.3 0.4 0.3>
    gravityOverride {z=-980}
    physicsBodyDefinitions [{
      boundBone {boneName=dyn_R_hair_3}
      boxExtents=<25 25 25>
      constraintSetup {angularLimitsMax=<0 35 2>, angularLimitsMin=<0 0 -2>, angularTarget=<1 0 0>}
      localJointOffset=<2 0 0>
     }]
    relativeSpaceBone {boneName=root}
   }
   n19 = Anim Dynamics - Bone: dyn_L_hair_4!(ComponentPose=@n18, bAlphaBoolEnabled:bool=True, GravityOverride:Vector="0,0,-980", AngularSpringConstant:float=20000, ExternalForce="0,0,0") -> (Pose) {
    alphaInputType=Bool
    angularDampingOverride=0.9999
    angularSpringConstant=20000
    bAngularSpring=true
    bOverrideAngularDamping=true
    bUseGravityOverride=true
    gravityScale=0.35
    simulationSpace=RootRelative
    alphaBoolBlend {blendInTime=2, blendOption=Cubic}
    boundBone {boneName=dyn_L_hair_4}
    componentAppliedLinearAccClamp=<3000 3000 3000>
    componentLinearAccScale=<0.3 0.4 0.3>
    componentLinearVelScale=<0.3 0.4 0.3>
    gravityOverride {z=-980}
    physicsBodyDefinitions [{
      boundBone {boneName=dyn_L_hair_4}
      boxExtents=<25 25 25>
      constraintSetup {angularLimitsMax=<0 15 2>, angularLimitsMin=<0 0 -2>, angularTarget=<1 0 0>}
      localJointOffset=<2 0 0>
     }]
    relativeSpaceBone {boneName=root}
   }
   n20 = Anim Dynamics - Bone: dyn_C_hair_4!(ComponentPose=@n19, bAlphaBoolEnabled:bool=True, GravityOverride:Vector="0,0,-980", AngularSpringConstant:float=20000, ExternalForce="0,0,0") -> (Pose) {
    alphaInputType=Bool
    angularDampingOverride=0.9999
    angularSpringConstant=20000
    bAngularSpring=true
    bOverrideAngularDamping=true
    bUseGravityOverride=true
    gravityScale=0.35
    simulationSpace=RootRelative
    alphaBoolBlend {blendInTime=2, blendOption=Cubic}
    boundBone {boneName=dyn_C_hair_4}
    componentAppliedLinearAccClamp=<3000 3000 3000>
    componentLinearAccScale=<0.3 0.4 0.3>
    componentLinearVelScale=<0.3 0.4 0.3>
    gravityOverride {z=-980}
    physicsBodyDefinitions [{
      boundBone {boneName=dyn_C_hair_4}
      boxExtents=<25 25 25>
      constraintSetup {angularLimitsMax=<0 15 2>, angularLimitsMin=<0 0 -2>, angularTarget=<1 0 0>}
      localJointOffset=<2 0 0>
     }]
    relativeSpaceBone {boneName=root}
   }
   n21 = Anim Dynamics - Bone: dyn_R_hair_4!(ComponentPose=@n20, bAlphaBoolEnabled:bool=True, GravityOverride:Vector="0,0,-980", AngularSpringConstant:float=20000, ExternalForce="0,0,0") -> (Pose) {
    alphaInputType=Bool
    angularDampingOverride=0.9999
    angularSpringConstant=20000
    bAngularSpring=true
    bOverrideAngularDamping=true
    bUseGravityOverride=true
    gravityScale=0.35
    simulationSpace=RootRelative
    alphaBoolBlend {blendInTime=2, blendOption=Cubic}
    boundBone {boneName=dyn_R_hair_4}
    componentAppliedLinearAccClamp=<3000 3000 3000>
    componentLinearAccScale=<0.3 0.4 0.3>
    componentLinearVelScale=<0.3 0.4 0.3>
    gravityOverride {z=-980}
    physicsBodyDefinitions [{
      boundBone {boneName=dyn_R_hair_4}
      boxExtents=<25 25 25>
      constraintSetup {angularLimitsMax=<0 15 2>, angularLimitsMin=<0 0 -2>, angularTarget=<1 0 0>}
      localJointOffset=<2 0 0>
     }]
    relativeSpaceBone {boneName=root}
   }
   n22 = Anim Dynamics - Bone: dyn_L_hair_5!(ComponentPose=@n21, bAlphaBoolEnabled:bool=True, GravityOverride:Vector="0,0,-980", AngularSpringConstant:float=20000, ExternalForce="0,0,0") -> (Pose) {
    alphaInputType=Bool
    angularDampingOverride=0.9999
    angularSpringConstant=20000
    bAngularSpring=true
    bOverrideAngularDamping=true
    bUseGravityOverride=true
    gravityScale=0.35
    simulationSpace=RootRelative
    alphaBoolBlend {blendInTime=2, blendOption=Cubic}
    boundBone {boneName=dyn_L_hair_5}
    componentAppliedLinearAccClamp=<3000 3000 3000>
    componentLinearAccScale=<0.3 0.4 0.3>
    componentLinearVelScale=<0.3 0.4 0.3>
    gravityOverride {z=-980}
    physicsBodyDefinitions [{
      boundBone {boneName=dyn_L_hair_5}
      boxExtents=<25 25 25>
      constraintSetup {angularLimitsMax=<0 20 5>, angularLimitsMin=<0 -10 -5>, angularTarget=<1 0 0>}
      localJointOffset=<2 0 0>
     }]
    relativeSpaceBone {boneName=root}
   }
   n23 = Anim Dynamics - Bone: dyn_C_hair_5!(ComponentPose=@n22, bAlphaBoolEnabled:bool=True, GravityOverride:Vector="0,0,-980", AngularSpringConstant:float=20000, ExternalForce="0,0,0") -> (Pose) {
    alphaInputType=Bool
    angularDampingOverride=0.9999
    angularSpringConstant=20000
    bAngularSpring=true
    bOverrideAngularDamping=true
    bUseGravityOverride=true
    gravityScale=0.35
    simulationSpace=RootRelative
    alphaBoolBlend {blendInTime=2, blendOption=Cubic}
    boundBone {boneName=dyn_C_hair_5}
    componentAppliedLinearAccClamp=<3000 3000 3000>
    componentLinearAccScale=<0.3 0.4 0.3>
    componentLinearVelScale=<0.3 0.4 0.3>
    gravityOverride {z=-980}
    physicsBodyDefinitions [{
      boundBone {boneName=dyn_C_hair_5}
      boxExtents=<25 25 25>
      constraintSetup {angularLimitsMax=<0 20 5>, angularLimitsMin=<0 -10 -5>, angularTarget=<1 0 0>}
      localJointOffset=<2 0 0>
     }]
    relativeSpaceBone {boneName=root}
   }
   n24 = Anim Dynamics - Bone: dyn_R_hair_5!(ComponentPose=@n23, bAlphaBoolEnabled:bool=True, GravityOverride:Vector="0,0,-980", AngularSpringConstant:float=20000, ExternalForce="0,0,0") -> (Pose) {
    alphaInputType=Bool
    angularDampingOverride=0.9999
    angularSpringConstant=20000
    bAngularSpring=true
    bOverrideAngularDamping=true
    bUseGravityOverride=true
    gravityScale=0.35
    simulationSpace=RootRelative
    alphaBoolBlend {blendInTime=2, blendOption=Cubic}
    boundBone {boneName=dyn_R_hair_5}
    componentAppliedLinearAccClamp=<3000 3000 3000>
    componentLinearAccScale=<0.3 0.4 0.3>
    componentLinearVelScale=<0.3 0.4 0.3>
    gravityOverride {z=-980}
    physicsBodyDefinitions [{
      boundBone {boneName=dyn_R_hair_5}
      boxExtents=<25 25 25>
      constraintSetup {angularLimitsMax=<0 20 5>, angularLimitsMin=<0 -10 -5>, angularTarget=<1 0 0>}
      localJointOffset=<2 0 0>
     }]
    relativeSpaceBone {boneName=root}
   }
   n25 = Anim Dynamics - Bone: dyn_L_hair_6!(ComponentPose=@n24, bAlphaBoolEnabled:bool=True, GravityOverride:Vector="0,0,-980", AngularSpringConstant:float=20000, ExternalForce="0,0,0") -> (Pose) {
    alphaInputType=Bool
    angularDampingOverride=0.9999
    angularSpringConstant=20000
    bAngularSpring=true
    bOverrideAngularDamping=true
    bUseGravityOverride=true
    bUseSphericalLimits=true
    gravityScale=0.35
    simulationSpace=RootRelative
    alphaBoolBlend {blendInTime=2, blendOption=Cubic}
    boundBone {boneName=dyn_L_hair_6}
    componentAppliedLinearAccClamp=<3000 3000 3000>
    componentLinearAccScale=<0.3 0.4 0.3>
    componentLinearVelScale=<0.3 0.4 0.3>
    gravityOverride {z=-980}
    physicsBodyDefinitions [{
      boundBone {boneName=dyn_L_hair_6}
      boxExtents=<25 25 25>
      constraintSetup {angularLimitsMax=<0 20 6>, angularLimitsMin=<0 0 -6>, angularTarget=<1 0 0>}
      localJointOffset=<2 0 0>
     }]
    relativeSpaceBone {boneName=root}
    sphericalLimits [{limitRadius=15, drivingBone {boneName=pelvis}, sphereLocalOffset=<0 0 0>}]
   }
   n26 = Anim Dynamics - Bone: dyn_C_hair_6!(ComponentPose=@n25, bAlphaBoolEnabled:bool=True, GravityOverride:Vector="0,0,-980", AngularSpringConstant:float=20000, ExternalForce="0,0,0") -> (Pose) {
    alphaInputType=Bool
    angularDampingOverride=0.9999
    angularSpringConstant=20000
    bAngularSpring=true
    bOverrideAngularDamping=true
    bUseGravityOverride=true
    bUseSphericalLimits=true
    gravityScale=0.35
    simulationSpace=RootRelative
    alphaBoolBlend {blendInTime=2, blendOption=Cubic}
    boundBone {boneName=dyn_C_hair_6}
    componentAppliedLinearAccClamp=<3000 3000 3000>
    componentLinearAccScale=<0.3 0.4 0.3>
    componentLinearVelScale=<0.3 0.4 0.3>
    gravityOverride {z=-980}
    physicsBodyDefinitions [{
      boundBone {boneName=dyn_C_hair_6}
      boxExtents=<25 25 25>
      constraintSetup {angularLimitsMax=<0 20 6>, angularLimitsMin=<0 0 -6>, angularTarget=<1 0 0>}
      localJointOffset=<2 0 0>
     }]
    relativeSpaceBone {boneName=root}
    sphericalLimits [{limitRadius=15, drivingBone {boneName=pelvis}, sphereLocalOffset=<0 0 0>}]
   }
   n27 = Anim Dynamics - Bone: dyn_R_hair_6!(ComponentPose=@n26, bAlphaBoolEnabled:bool=True, GravityOverride:Vector="0,0,-980", AngularSpringConstant:float=20000, ExternalForce="0,0,0") -> (Pose) {
    alphaInputType=Bool
    angularDampingOverride=0.9999
    angularSpringConstant=20000
    bAngularSpring=true
    bOverrideAngularDamping=true
    bUseGravityOverride=true
    bUseSphericalLimits=true
    gravityScale=0.35
    simulationSpace=RootRelative
    alphaBoolBlend {blendInTime=2, blendOption=Cubic}
    boundBone {boneName=dyn_R_hair_6}
    componentAppliedLinearAccClamp=<3000 3000 3000>
    componentLinearAccScale=<0.3 0.4 0.3>
    componentLinearVelScale=<0.3 0.4 0.3>
    gravityOverride {z=-980}
    physicsBodyDefinitions [{
      boundBone {boneName=dyn_R_hair_6}
      boxExtents=<25 25 25>
      constraintSetup {angularLimitsMax=<0 20 6>, angularLimitsMin=<0 0 -6>, angularTarget=<1 0 0>}
      localJointOffset=<2 0 0>
     }]
    relativeSpaceBone {boneName=root}
    sphericalLimits [{limitRadius=15, drivingBone {boneName=pelvis}, sphereLocalOffset=<0 0 0>}]
   }
   n28 = Anim Dynamics - Bone: dyn_L_hair_7!(ComponentPose=@n27, bAlphaBoolEnabled:bool=True, GravityOverride:Vector="0,0,-980", AngularSpringConstant:float=20000, ExternalForce="0,0,0") -> (Pose) {
    alphaInputType=Bool
    angularDampingOverride=0.9999
    angularSpringConstant=20000
    bAngularSpring=true
    bOverrideAngularDamping=true
    bUseGravityOverride=true
    bUseSphericalLimits=true
    gravityScale=0.35
    simulationSpace=RootRelative
    alphaBoolBlend {blendInTime=2, blendOption=Cubic}
    boundBone {boneName=dyn_L_hair_7}
    componentAppliedLinearAccClamp=<3000 3000 3000>
    componentLinearAccScale=<0.3 0.4 0.3>
    componentLinearVelScale=<0.3 0.4 0.3>
    gravityOverride {z=-980}
    physicsBodyDefinitions [{
      boundBone {boneName=dyn_L_hair_7}
      boxExtents=<25 25 25>
      constraintSetup {angularLimitsMax=<0 10 5>, angularLimitsMin=<0 -10 -5>, angularTarget=<1 0 0>}
      localJointOffset=<2 0 0>
     }]
    relativeSpaceBone {boneName=root}
    sphericalLimits [{limitRadius=15, drivingBone {boneName=pelvis}, sphereLocalOffset=<0 0 0>}]
   }
   n29 = Anim Dynamics - Bone: dyn_C_hair_7!(ComponentPose=@n28, bAlphaBoolEnabled:bool=True, GravityOverride:Vector="0,0,-980", AngularSpringConstant:float=20000, ExternalForce="0,0,0") -> (Pose) {
    alphaInputType=Bool
    angularDampingOverride=0.9999
    angularSpringConstant=20000
    bAngularSpring=true
    bOverrideAngularDamping=true
    bUseGravityOverride=true
    bUseSphericalLimits=true
    gravityScale=0.35
    simulationSpace=RootRelative
    alphaBoolBlend {blendInTime=2, blendOption=Cubic}
    boundBone {boneName=dyn_C_hair_7}
    componentAppliedLinearAccClamp=<3000 3000 3000>
    componentLinearAccScale=<0.3 0.4 0.3>
    componentLinearVelScale=<0.3 0.4 0.3>
    gravityOverride {z=-980}
    physicsBodyDefinitions [{
      boundBone {boneName=dyn_C_hair_7}
      boxExtents=<25 25 25>
      constraintSetup {angularLimitsMax=<0 10 5>, angularLimitsMin=<0 -10 -5>, angularTarget=<1 0 0>}
      localJointOffset=<2 0 0>
     }]
    relativeSpaceBone {boneName=root}
    sphericalLimits [{limitRadius=15, drivingBone {boneName=pelvis}, sphereLocalOffset=<0 0 0>}]
   }
   n30 = Anim Dynamics - Bone: dyn_R_hair_7!(ComponentPose=@n29, bAlphaBoolEnabled:bool=True, GravityOverride:Vector="0,0,-980", AngularSpringConstant:float=20000, ExternalForce="0,0,0") -> (Pose) {
    alphaInputType=Bool
    angularDampingOverride=0.9999
    angularSpringConstant=20000
    bAngularSpring=true
    bOverrideAngularDamping=true
    bUseGravityOverride=true
    bUseSphericalLimits=true
    gravityScale=0.35
    simulationSpace=RootRelative
    alphaBoolBlend {blendInTime=2, blendOption=Cubic}
    boundBone {boneName=dyn_R_hair_7}
    componentAppliedLinearAccClamp=<3000 3000 3000>
    componentLinearAccScale=<0.3 0.4 0.3>
    componentLinearVelScale=<0.3 0.4 0.3>
    gravityOverride {z=-980}
    physicsBodyDefinitions [{
      boundBone {boneName=dyn_R_hair_7}
      boxExtents=<25 25 25>
      constraintSetup {angularLimitsMax=<0 10 5>, angularLimitsMin=<0 -10 -5>, angularTarget=<1 0 0>}
      localJointOffset=<2 0 0>
     }]
    relativeSpaceBone {boneName=root}
    sphericalLimits [{limitRadius=15, drivingBone {boneName=pelvis}, sphereLocalOffset=<0 0 0>}]
   }
   n31 = "Transform (Modify) Bone - Bone: base_C_hair_1"!(ComponentPose=@n30, bAlphaBoolEnabled:bool=True, Translation="0,0,-8", Rotation="0,0,0") -> (Pose) {
    alphaInputType=Bool
    rotationMode=BMM_Additive
    rotationSpace=BCS_BoneSpace
    translationMode=BMM_Additive
    translationSpace=BCS_BoneSpace
    alphaBoolBlend {blendInTime=0.2, blendOption=Linear, blendOutTime=0.5}
    alphaScaleBiasClamp {interpSpeedDecreasing=1, interpSpeedIncreasing=1}
    boneToModify {boneName=base_C_hair_1}
    translation {z=-8}
   }
   n32 = "Transform (Modify) Bone - Bone: base_L_hair_1"!(ComponentPose=@n31, bAlphaBoolEnabled:bool=True, Translation="0,0,-8", Rotation="0,0,0") -> (Pose) {
    alphaInputType=Bool
    rotationMode=BMM_Additive
    rotationSpace=BCS_BoneSpace
    translationMode=BMM_Additive
    translationSpace=BCS_BoneSpace
    alphaBoolBlend {blendInTime=0.2, blendOption=Linear, blendOutTime=0.5}
    alphaScaleBiasClamp {interpSpeedDecreasing=1, interpSpeedIncreasing=1}
    boneToModify {boneName=base_L_hair_1}
    translation {z=-8}
   }
   n33 = "Transform (Modify) Bone - Bone: base_R_hair_1"!(ComponentPose=@n32, bAlphaBoolEnabled:bool=True, Translation="0,0,-8", Rotation="0,0,0") -> (Pose) {
    alphaInputType=Bool
    rotationMode=BMM_Additive
    rotationSpace=BCS_BoneSpace
    translationMode=BMM_Additive
    translationSpace=BCS_BoneSpace
    alphaBoolBlend {blendInTime=0.2, blendOption=Linear, blendOutTime=0.5}
    alphaScaleBiasClamp {interpSpeedDecreasing=1, interpSpeedIncreasing=1}
    boneToModify {boneName=base_R_hair_1}
    translation {z=-8}
   }
   n34 = "Transform (Modify) Bone - Bone: base_C_hair_1"(ComponentPose=@n33, bAlphaBoolEnabled=True, Translation="0,0,0", Rotation="-5,0,0") {
    alphaInputType=Bool
    rotationMode=BMM_Additive
    rotationSpace=BCS_BoneSpace
    translationMode=BMM_Additive
    translationSpace=BCS_BoneSpace
    alphaBoolBlend {blendInTime=0.2, blendOption=Linear, blendOutTime=0.5}
    alphaScaleBiasClamp {interpSpeedDecreasing=1, interpSpeedIncreasing=1}
    boneToModify {boneName=base_C_hair_1}
    rotation {pitch=-5}
   }
   n35 = "Transform (Modify) Bone - Bone: base_L_hair_1"(ComponentPose=@n34, bAlphaBoolEnabled=True, Translation="0,0,-5", Rotation="-5,0,0") {
    alphaInputType=Bool
    rotationMode=BMM_Additive
    rotationSpace=BCS_BoneSpace
    translationMode=BMM_Additive
    translationSpace=BCS_BoneSpace
    alphaBoolBlend {blendInTime=0.2, blendOption=Linear, blendOutTime=0.5}
    alphaScaleBiasClamp {interpSpeedDecreasing=1, interpSpeedIncreasing=1}
    boneToModify {boneName=base_L_hair_1}
    rotation {pitch=-5}
    translation {z=-5}
   }
   n36 = "Transform (Modify) Bone - Bone: base_R_hair_1"(ComponentPose=@n35, bAlphaBoolEnabled=True, Translation="-2,0,-2", Rotation="-7,0,0") {
    alphaInputType=Bool
    rotationMode=BMM_Additive
    rotationSpace=BCS_BoneSpace
    translationMode=BMM_Additive
    translationSpace=BCS_BoneSpace
    alphaBoolBlend {blendInTime=0.2, blendOption=Linear, blendOutTime=0.5}
    alphaScaleBiasClamp {interpSpeedDecreasing=1, interpSpeedIncreasing=1}
    boneToModify {boneName=base_R_hair_1}
    rotation {pitch=-7}
    translation {x=-2, z=-2}
   }
   n37 = Get Hang Gliding Alpha()
   n38 = "Transform (Modify) Bone - Bone: base_L_hair_1"!(ComponentPose=@n36, Alpha:float=@n37, Translation="-6,4,0", Rotation="0,0,0") -> (Pose) {
    rotationMode=BMM_Additive
    rotationSpace=BCS_BoneSpace
    translationMode=BMM_Additive
    translationSpace=BCS_BoneSpace
    alphaBoolBlend {blendInTime=0.2, blendOption=Linear, blendOutTime=0.5}
    alphaScaleBiasClamp {bInterpResult=true, interpSpeedDecreasing=1, interpSpeedIncreasing=1}
    boneToModify {boneName=base_L_hair_1}
    translation=<-6 4>
   }
   n39 = "Transform (Modify) Bone - Bone: base_C_hair_1"!(ComponentPose=@n38, Alpha:float=@n37, Translation="-5,0,0", Rotation="0,0,0") -> (Pose) {
    rotationMode=BMM_Additive
    rotationSpace=BCS_BoneSpace
    translationMode=BMM_Additive
    translationSpace=BCS_BoneSpace
    alphaBoolBlend {blendInTime=0.2, blendOption=Linear, blendOutTime=0.5}
    alphaScaleBiasClamp {bInterpResult=true, interpSpeedDecreasing=1, interpSpeedIncreasing=1}
    boneToModify {boneName=base_C_hair_1}
    translation {x=-5}
   }
   n40 = "Transform (Modify) Bone - Bone: base_R_hair_1"!(ComponentPose=@n39, Alpha:float=@n37, Translation="-5,-2,0", Rotation="0,0,0") -> (Pose) {
    rotationMode=BMM_Additive
    rotationSpace=BCS_BoneSpace
    translationMode=BMM_Additive
    translationSpace=BCS_BoneSpace
    alphaBoolBlend {blendInTime=0.2, blendOption=Linear, blendOutTime=0.5}
    alphaScaleBiasClamp {bInterpResult=true, interpSpeedDecreasing=1, interpSpeedIncreasing=1}
    boneToModify {boneName=base_R_hair_1}
    translation=<-5 -2>
   }
   n41 = Get bIsSkydiveFloating()
   n42 = "Transform (Modify) Bone - Bone: base_shoulder_r_hair"!(ComponentPose=@n40, bAlphaBoolEnabled:bool=@n41, Translation="-3,3,0", Rotation="0,0,0") -> (Pose) {
    alphaInputType=Bool
    rotationMode=BMM_Additive
    rotationSpace=BCS_BoneSpace
    translationMode=BMM_Additive
    translationSpace=BCS_BoneSpace
    alphaBoolBlend {blendInTime=0.2, blendOption=Linear, blendOutTime=0.5}
    alphaScaleBiasClamp {bInterpResult=true, interpSpeedDecreasing=1, interpSpeedIncreasing=1}
    boneToModify {boneName=base_shoulder_r_hair}
    translation=<-3 3>
   }
   n43 = "Transform (Modify) Bone - Bone: base_R_hair_1"!(ComponentPose=@n42, bAlphaBoolEnabled:bool=@n41, Translation="-1,0,0", Rotation="0,0,0") -> (Pose) {
    alphaInputType=Bool
    rotationMode=BMM_Additive
    rotationSpace=BCS_BoneSpace
    translationMode=BMM_Additive
    translationSpace=BCS_BoneSpace
    alphaBoolBlend {blendInTime=0.2, blendOption=Linear, blendOutTime=0.5}
    alphaScaleBiasClamp {bInterpResult=true, interpSpeedDecreasing=1, interpSpeedIncreasing=1}
    boneToModify {boneName=base_R_hair_1}
    translation {x=-1}
   }
   n44 = RigidBody!(
    ComponentPose=@n43
    Alpha=Get alphaHairRigidBody float() -> (alphaHairRigidBody float:double)
    OverrideWorldGravity:Vector=Get GravityOverride()
    ComponentLinearAccScale:Vector="0.3,0.4,0.3"
    ComponentLinearVelScale:Vector="0.3,0.4,0.3"
   ) -> (Pose) {
    bOverrideWorldGravity=true
    overridePhysicsAsset=&/F_MED_CajunTaco_FaceAcc_Physics
    simulationSpace=BaseBoneSpace
    alphaBoolBlend {blendOption=Linear}
    alphaScaleBiasClamp {bInterpResult=true, interpSpeedDecreasing=3, interpSpeedIncreasing=3}
    baseBoneRef {boneName=root}
    componentAppliedLinearAccClamp=<3000 3000 3000>
    componentLinearAccScale=<0.06 0.06 0.1>
    componentLinearVelScale=<0.06 0.06 0.1>
    simSpaceSettings {maxAngularVelocity=3, maxLinearAcceleration=700, maxLinearVelocity=500, worldAlpha=0.2, externalLinearDragV=<1 1 1>}
   }
   Save cached pose 'CRCache_HairDynPose'(Pose=Component To Local(ComponentPose=@n44)) {cacheName=CRCache_HairDynPose}
   n45 = Anim Dynamics - Bone: dyn_C_hair_6(
    ComponentPose=Local To Component(LocalPose=Use cached pose 'CRCache_HairPose'())
    bAlphaBoolEnabled=True
    GravityOverride="0,0,-980"
    AngularSpringConstant=20000
    ExternalForce="0,0,0"
   ) {
    alphaInputType=Bool
    angularDampingOverride=1
    angularSpringConstant=20000
    bAngularSpring=true
    bOverrideAngularDamping=true
    bUseGravityOverride=true
    bUseSphericalLimits=true
    gravityScale=0.35
    simulationSpace=RootRelative
    alphaBoolBlend {blendInTime=2, blendOption=Cubic}
    boundBone {boneName=dyn_C_hair_6}
    componentAppliedLinearAccClamp=<400000 400000 400000>
    componentLinearAccScale=<0.3 0.4 0.3>
    componentLinearVelScale=<0.3 0.4 3>
    gravityOverride {z=-980}
    physicsBodyDefinitions [{
      boundBone {boneName=dyn_C_hair_6}
      boxExtents=<25 25 25>
      constraintSetup {angularLimitsMax=<0 45 10>, angularLimitsMin=<0 0 -10>, angularTarget=<1 0 0>}
      localJointOffset=<7 0 0>
     }]
    relativeSpaceBone {boneName=root}
    sphericalLimits [{limitRadius=15, drivingBone {boneName=pelvis}, sphereLocalOffset=<0 0 0>}]
   }
   n46 = Anim Dynamics - Bone: dyn_L_hair_6(ComponentPose=@n45, bAlphaBoolEnabled=True, GravityOverride="0,0,-980", AngularSpringConstant=20000, ExternalForce="0,0,0") {
    alphaInputType=Bool
    angularDampingOverride=1
    angularSpringConstant=20000
    bAngularSpring=true
    bOverrideAngularDamping=true
    bUseGravityOverride=true
    bUseSphericalLimits=true
    gravityScale=0.35
    simulationSpace=RootRelative
    alphaBoolBlend {blendInTime=2, blendOption=Cubic}
    boundBone {boneName=dyn_L_hair_6}
    componentAppliedLinearAccClamp=<400000 400000 400000>
    componentLinearAccScale=<0.3 0.4 0.3>
    componentLinearVelScale=<0.3 0.4 3>
    gravityOverride {z=-980}
    physicsBodyDefinitions [{
      boundBone {boneName=dyn_L_hair_6}
      boxExtents=<25 25 25>
      constraintSetup {angularLimitsMax=<0 45 10>, angularLimitsMin=<0 0 -10>, angularTarget=<1 0 0>}
      localJointOffset=<7 0 0>
     }]
    relativeSpaceBone {boneName=root}
    sphericalLimits [{limitRadius=15, drivingBone {boneName=pelvis}, sphereLocalOffset=<0 0 0>}]
   }
   n47 = Anim Dynamics - Bone: dyn_R_hair_6(ComponentPose=@n46, bAlphaBoolEnabled=True, GravityOverride="0,0,-980", AngularSpringConstant=20000, ExternalForce="0,0,0") {
    alphaInputType=Bool
    angularDampingOverride=1
    angularSpringConstant=20000
    bAngularSpring=true
    bOverrideAngularDamping=true
    bUseGravityOverride=true
    bUseSphericalLimits=true
    gravityScale=0.35
    simulationSpace=RootRelative
    alphaBoolBlend {blendInTime=2, blendOption=Cubic}
    boundBone {boneName=dyn_R_hair_6}
    componentAppliedLinearAccClamp=<400000 400000 400000>
    componentLinearAccScale=<0.3 0.4 0.3>
    componentLinearVelScale=<0.3 0.4 3>
    gravityOverride {z=-980}
    physicsBodyDefinitions [{
      boundBone {boneName=dyn_R_hair_6}
      boxExtents=<25 25 25>
      constraintSetup {angularLimitsMax=<0 45 10>, angularLimitsMin=<0 0 -10>, angularTarget=<1 0 0>}
      localJointOffset=<7 0 0>
     }]
    relativeSpaceBone {boneName=root}
    sphericalLimits [{limitRadius=15, drivingBone {boneName=pelvis}, sphereLocalOffset=<0 0 0>}]
   }
   n48 = Anim Dynamics - Bone: dyn_C_hair_7(ComponentPose=@n47, bAlphaBoolEnabled=True, GravityOverride="0,0,-980", AngularSpringConstant=20000, ExternalForce="0,0,0") {
    alphaInputType=Bool
    angularDampingOverride=1
    angularSpringConstant=20000
    bAngularSpring=true
    bOverrideAngularDamping=true
    bUseGravityOverride=true
    bUseSphericalLimits=true
    gravityScale=0.35
    alphaBoolBlend {blendInTime=2, blendOption=Cubic}
    boundBone {boneName=dyn_C_hair_7}
    componentAppliedLinearAccClamp=<400000 400000 400000>
    componentLinearAccScale=<0.3 0.4 0.3>
    componentLinearVelScale=<0.3 0.4 3>
    gravityOverride {z=-980}
    physicsBodyDefinitions [{
      boundBone {boneName=dyn_C_hair_7}
      boxExtents=<25 25 25>
      constraintSetup {angularLimitsMax=<0 25 15>, angularLimitsMin=<0 -10 -15>, angularTarget=<1 0 0>}
      localJointOffset=<7 0 0>
     }]
    relativeSpaceBone {boneName=root}
    sphericalLimits [{limitRadius=15, drivingBone {boneName=pelvis}, sphereLocalOffset=<0 0 0>}]
   }
   n49 = Anim Dynamics - Bone: dyn_L_hair_7(ComponentPose=@n48, bAlphaBoolEnabled=True, GravityOverride="0,0,-980", AngularSpringConstant=20000, ExternalForce="0,0,0") {
    alphaInputType=Bool
    angularDampingOverride=1
    angularSpringConstant=20000
    bAngularSpring=true
    bOverrideAngularDamping=true
    bUseGravityOverride=true
    bUseSphericalLimits=true
    gravityScale=0.35
    alphaBoolBlend {blendInTime=2, blendOption=Cubic}
    boundBone {boneName=dyn_L_hair_7}
    componentAppliedLinearAccClamp=<400000 400000 400000>
    componentLinearAccScale=<0.3 0.4 0.3>
    componentLinearVelScale=<0.3 0.4 3>
    gravityOverride {z=-980}
    physicsBodyDefinitions [{
      boundBone {boneName=dyn_L_hair_7}
      boxExtents=<25 25 25>
      constraintSetup {angularLimitsMax=<0 25 15>, angularLimitsMin=<0 -10 -15>, angularTarget=<1 0 0>}
      localJointOffset=<7 0 0>
     }]
    relativeSpaceBone {boneName=root}
    sphericalLimits [{limitRadius=15, drivingBone {boneName=pelvis}, sphereLocalOffset=<0 0 0>}]
   }
   n50 = Anim Dynamics - Bone: dyn_R_hair_7(ComponentPose=@n49, bAlphaBoolEnabled=True, GravityOverride="0,0,-980", AngularSpringConstant=20000, ExternalForce="0,0,0") {
    alphaInputType=Bool
    angularDampingOverride=1
    angularSpringConstant=20000
    bAngularSpring=true
    bOverrideAngularDamping=true
    bUseGravityOverride=true
    bUseSphericalLimits=true
    gravityScale=0.35
    alphaBoolBlend {blendInTime=2, blendOption=Cubic}
    boundBone {boneName=dyn_R_hair_7}
    componentAppliedLinearAccClamp=<400000 400000 400000>
    componentLinearAccScale=<0.3 0.4 0.3>
    componentLinearVelScale=<0.3 0.4 3>
    gravityOverride {z=-980}
    physicsBodyDefinitions [{
      boundBone {boneName=dyn_R_hair_7}
      boxExtents=<25 25 25>
      constraintSetup {angularLimitsMax=<0 25 15>, angularLimitsMin=<0 -10 -15>, angularTarget=<1 0 0>}
      localJointOffset=<7 0 0>
     }]
    relativeSpaceBone {boneName=root}
    sphericalLimits [{limitRadius=15, drivingBone {boneName=pelvis}, sphereLocalOffset=<0 0 0>}]
   }
   Save cached pose 'CRCache_HairDynBackPackPose'(Pose=Component To Local(ComponentPose=@n50)) {cacheName=CRCache_HairDynBackPackPose}
  ]}
 EventGraph {nodes [
   Event BlueprintUpdateAnimation()
   > n1 = Sequence!(execute) -> (then_0, then_1, then_2:Exec, then_3:Exec, then_4:Exec, then_5:Exec)
   > n2 = Gravity Override State Function DEPRECATED(CharacterPartAnimInstance=Self-Reference(), GameTimeInSeconds=GetGameTimeInSeconds(), GravityOverrideParams=Get Gravity Override Params())
   > Set GravityOverride(GravityOverride=@n2)
   > n3 = Set GravityOverride AnimDynamics(GravityOverride AnimDynamics:Vector="0, 0, 0") -> (Output_Get:Vector)
   n4 = Get bIsSkyDiving()
   n5 = Get bIsJumping()
   n6 = Get bIsFalling()
   > Branch(Condition=AND Boolean(A=OR Boolean(A=@n5, B=@n6), B=NOT Boolean(A=@n4)))
   n7 = Remap Float(Input=Get PawnSpeed2D(), OldRange=(X=0,Y=550), NewRange=(X=0,Y=1), Clamp=true) -> (Output=~)
   n8 = "Clamp (Float)"(Value=Get PawnVelocityZ(), Min=-980, Max=0)
   n9 = float + float(A=Lerp(A=1300, B=float * float(A=@n8, B=-2) {is=title K2Node_PromotableOperator}, Alpha=@n7), B=Lerp(A=0, B=800, Alpha=@n7))
   > Set GravityOverride AnimDynamics(GravityOverride AnimDynamics=vector + vector(A=@n3, B=MakeVector(Y=0, Z=@n9)))
   n10 = Branch(execute=@n1.then_3, Condition=NOT Boolean(A=Get bIsDBNO()))
   > Set Not DBNO Alpha(Not DBNO Alpha:double=1) -> (Output_Get:double)
   Set Not DBNO Alpha(execute=@n10.else, Not DBNO Alpha=0)
   n11 = GetAnimInstance(self=GetOwnerSkeletalMeshComponent())
   Is Valid(exec=@n1.then_4, InputObject=@n11)
   > Set isCajunTacoEmotePlaying(isCajunTacoEmotePlaying:bool=OR Boolean!(
     A=Montage_IsPlaying(self=@n11, Montage=/BRCosmetics/Animation/Game/MainPlayer/Emotes/ChiveFlake/CMF/Emote_ChiveFlake_CMF_M)
     B=Montage_IsPlaying(self=@n11, Montage=/BRCosmetics/Animation/Game/MainPlayer/Emotes/ChiveFlake/Frontend/Emote_ChiveFlake_CMF_Frontend_M)
     C:bool=Montage_IsPlaying(self=@n11, Montage=/BRCosmetics/Animation/Game/MainPlayer/Emotes/CajunTaco/Followers/CLF/Emote_CajunTaco_CLF_Follower_M)
     D:bool=Montage_IsPlaying(self=@n11, Montage=/BRCosmetics/Animation/Game/MainPlayer/Emotes/CajunTaco/CLF/Emote_CajunTaco_CLF_M)
     E:bool=Montage_IsPlaying(self=@n11, Montage=/BRCosmetics/Animation/Game/MainPlayer/Emotes/CajunTaco/CLF/Emote_CajunTaco_CLF_M_Hide)
     F:bool=Montage_IsPlaying(self=@n11, Montage=/BRCosmetics/Animation/Game/MainPlayer/Emotes/CajunTaco/Followers/CLM/Emote_CajunTaco_CLM_Follower_M)
     G:bool=Montage_IsPlaying(self=@n11, Montage=/BRCosmetics/Animation/Game/MainPlayer/Emotes/CajunTaco/CLM/Emote_CajunTaco_CLM_M)
     H:bool=Montage_IsPlaying(self=@n11, Montage=/BRCosmetics/Animation/Game/MainPlayer/Emotes/CajunTaco/CLM/Emote_CajunTaco_CLM_M_Hide)
     I:bool=Montage_IsPlaying(self=@n11, Montage=/BRCosmetics/Animation/Game/MainPlayer/Emotes/CajunTaco/Followers/CMF/Emote_CajunTaco_CMF_Follower_M)
     J:bool=Montage_IsPlaying(self=@n11, Montage=/BRCosmetics/Animation/Game/MainPlayer/Emotes/CajunTaco/CMF/Emote_CajunTaco_CMF_M)
     K:bool=Montage_IsPlaying(self=@n11, Montage=/BRCosmetics/Animation/Game/MainPlayer/Emotes/CajunTaco/CMF/Emote_CajunTaco_CMF_M_Hide)
     L:bool=Montage_IsPlaying(self=@n11, Montage=/BRCosmetics/Animation/Game/MainPlayer/Emotes/CajunTaco/Followers/Emote_CajunTaco_CMM_Follower_M)
     M:bool=Montage_IsPlaying(self=@n11, Montage=/BRCosmetics/Animation/Game/MainPlayer/Emotes/CajunTaco/Emote_CajunTaco_CMM_M)
     N:bool=Montage_IsPlaying(self=@n11, Montage=/BRCosmetics/Animation/Game/MainPlayer/Emotes/CajunTaco/Emote_CajunTaco_CMM_M_Hide)
     O:bool=Montage_IsPlaying(self=@n11, Montage=/BRCosmetics/Animation/Game/MainPlayer/Emotes/CajunTaco/Followers/Hero/Emote_CajunTaco_Follower_Hero_M)
     P:bool=Montage_IsPlaying(self=@n11, Montage=/BRCosmetics/Animation/Game/MainPlayer/Emotes/CajunTaco/Hero/Emote_CajunTaco_Hero_M)
    ) -> (ReturnValue)) -> (Output_Get:bool)
   > Set IsPlayingCelebration(IsPlayingCelebration:bool=Montage_IsPlaying(self=@n11, Montage=/BRCosmetics/Animation/Game/MainPlayer/Emotes/Emote_Celebration_Female_M)) -> (Output_Get:bool)
   Set disableUpperLimbsCorrection(
    execute=@n1.then_1
    disableUpperLimbsCorrection:bool=OR Boolean!(A=Get bIsSkydiveDiveMode(), B=Get bIsSkydiveFloating(), C=Get bIsParachuteOpened()) -> (ReturnValue)
   ) -> (Output_Get:bool)
   > Set toggleAnimDynamics(toggleAnimDynamics:bool=NOT Boolean(A=Get isOnZipline() -> (isOnZipline:bool))) -> (Output_Get:bool)
   n12 = Get bIsTacticalSprinting()
   n13 = AND Boolean(A=Get bIsJumping(), B=@n12)
   > Set isTacticalJumping(isTacticalJumping:bool=@n13) -> (Output_Get:bool)
   n14 = Get bIsFalling()
   > Set isTacticalJumpingFalling(isTacticalJumpingFalling:bool=OR Boolean!(A=@n13, B=AND Boolean(A=@n12, B=@n14)) -> (ReturnValue)) -> (Output_Get:bool)
   > Set Skydive Alpha(Skydive Alpha:double="To Float (Boolean)"(InBool=Get bIsSkyDiving())) -> (Output_Get:double)
   > Set Hang Gliding Alpha(Hang Gliding Alpha:double="To Float (Boolean)"(InBool=Get bIsParachuteOpened())) -> (Output_Get:double)
   n15 = Branch(execute=@n1.then_2, Condition=OR Boolean(A=Get bIsSkydiveDiveMode(), B=Get bIsSkydiveFloating()))
   > Set angularSpringConstant(angularSpringConstant:double=10000) -> (Output_Get:double)
   n16 = Branch(execute=@n15.else, Condition=Get bIsSprinting())
   > n17 = Branch(Condition=Get bIsTacticalSprinting())
   > Set angularSpringConstant(angularSpringConstant=15000)
   Set angularSpringConstant(execute=@n17.else, angularSpringConstant=17000)
   Set angularSpringConstant(execute=@n16.else, angularSpringConstant=8000)
   n18 = Branch(execute=@n1.then_5, Condition=Get bIsInFrontEnd())
   > Set alphaHairRigidBody float(alphaHairRigidBody float:double=0.45) -> (Output_Get:double)
   n19 = Branch(execute=@n18.else, Condition=Get bIsPlayingMeleeAnim())
   Set alphaHairRigidBody float(execute=[@n19.then, @n21.then], alphaHairRigidBody float=0.4)
   n20 = Branch(execute=@n19.else, Condition=Get bIsTargeting())
   > Set alphaHairRigidBody float(alphaHairRigidBody float=0.6)
   n21 = Branch(execute=@n20.else, Condition=Get isTacticalJumpingFalling() -> (isTacticalJumpingFalling:bool))
   Set alphaHairRigidBody float(execute=@n21.else, alphaHairRigidBody float=1)
   > Branch(Condition=Get bIsSkyDiving())
   > Set alphaHairRigidBody float(alphaHairRigidBody float=0.7)
   Event BlueprintBeginPlay() {is=title K2Node_Event}
   > Set disableControlRig(disableControlRig:bool=false) -> (Output_Get:bool)
   Event OnCharacterCustomizationLoaded()
   n22 = CheckForBackbling()
   > Set isWearingPack(isWearingPack:bool=OR Boolean!(A=@n22, B=@n22.bIsBackblingACape, C:bool=@n22.bIsBackblingLong) -> (ReturnValue)) -> (Output_Get:bool)
   > Set isWearingCape(isWearingCape=@n22.bIsBackblingLong)
   Event OnWeaponChanged()
   > Set isAssaultRifleEquipped float(isAssaultRifleEquipped float:double=0) -> (Output_Get:double)
   > n23 = Switch on EFortWeaponCoreAnimation!(execute, Selection:EFortWeaponCoreAnimation Enum=Get WeaponCoreAnim!() -> (WeaponCoreAnim)) -> (
    Melee:Exec
    Pistol:Exec
    Shotgun:Exec
    PaperBlueprint:Exec
    Rifle:Exec
    MeleeOneHand:Exec
    MachinePistol:Exec
    RocketLauncher:Exec
    GrenadeLauncher:Exec
    GoingCommando:Exec
    AssaultRifle:Exec
    TacticalShotgun:Exec
    SniperRifle:Exec
    TrapPlacement:Exec
    ShoulderLauncher:Exec
    AbilityDecoTool:Exec
    Crossbow:Exec
    C4:Exec
    RemoteControl:Exec
    DualWield:Exec
    AR_BullPup:Exec
    AR_ForwardGrip:Exec
    MedPackPaddles:Exec
    SMG_P90:Exec
    AR_DrumGun:Exec
    Consumable_Small:Exec
    Consumable_Large:Exec
    Balloon:Exec
    MountedTurret:Exec
    CreativeTool:Exec
    ExplosiveBow:Exec
    AshtonIndigo:Exec
    AshtonChicago:Exec
    MeleeDualWield:Exec
    Unarmed:Exec
   )
   Set isAssaultRifleEquipped float(execute=@n23.AssaultRifle, isAssaultRifleEquipped float=1)
   Event BlueprintInitializeAnimation()
   > Set Skeletal_Mesh(Skeletal_Mesh:Skeletal Mesh Component Object Reference=GetOwnerSkeletalMeshComponent()) -> (Output_Get:Skeletal Mesh Component Object Reference)
   n24 = Event BlueprintProcessAnimInputEvent()
   > n25 = Branch(Condition=IsNamed(self=@n24.AnimInputEvent, InEventName=aie.Player.Zipline.Entered))
   > Set isOnZipline(isOnZipline:bool=1) -> (Output_Get:bool)
   Branch(execute=@n25.else, Condition=IsNamed(self=@n24.AnimInputEvent, InEventName=aie.Player.Zipline.Exited))
   > Set isOnZipline(isOnZipline=false)
   Get bIsMoving2D()
   Get bIsSkydiveFloating()
   Get Skeletal_Mesh()
  ]}
]
variables [
 alphaHairRigidBody float {type=number}
 angularSpringConstant {type=number}
 disableControlRig {type=boolean}
 disableUpperLimbsCorrection {type=boolean}
 Gravity Override Params {type=GravityOverrideParamsStruct}
 GravityOverride {type=Vector}
 GravityOverride AnimDynamics {type=Vector}
 Hang Gliding Alpha {type=number}
 isAssaultRifleEquipped float {type=number}
 isCajunTacoEmotePlaying {type=boolean}
 isOnZipline {type=boolean}
 IsPlayingCelebration {type=boolean}
 isTacticalJumping {type=boolean}
 isTacticalJumpingFalling {type=boolean}
 isWearingCape {type=boolean}
 isWearingPack {type=boolean}
 Not DBNO Alpha {type=number}
 Skeletal_Mesh {type=SkeletalMeshComponent}
 Skydive Alpha {type=number}
 toggleAnimDynamics {type=boolean}
]
